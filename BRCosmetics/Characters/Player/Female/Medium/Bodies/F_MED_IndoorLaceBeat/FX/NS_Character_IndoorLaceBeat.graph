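# Niagara character effect for the IndoorLaceBeat cosmetic body, emitting golden pixie-dust sprites from skeletal mesh bones.
# A fake audio-reactivity loop drives spawn probability and alpha, while time of day, distance and front-end modifiers adjust the look.
# PixieDust spawns around selected bones, while Pixie_Dust picks bones by random group and partition index, using baked per-bone offset arrays.
# PixieDust scales inherited bone movement by skydiving state, owner velocity and distance.
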NiagaraSystem /BRCosmetics/Characters/Player/Female/Medium/Bodies/F_MED_IndoorLaceBeat/FX/NS_Character_IndoorLaceBeat
dependencies {
 usedDataInterfaces [
  &NiagaraDataInterfaceArrayFloat3
  &NiagaraDataInterfaceCurve
  &NiagaraDataInterfaceVector2DCurve
  &NiagaraDataInterfaceArrayInt32
  &NiagaraDataInterfaceArrayFloat
  &NiagaraDataInterfaceSkeletalMesh
  &NiagaraDataInterfaceColorCurve
 ]
 usedDynamicInputs [
  &RandomRangeLinearColor
  &MakeVector2D
  /Niagara/DynamicInputs/Lerp/SmoothLerpOverTime_Float
  &ColorFromCurve
  &FNTimeOfDayScalar
 ]
 usedModules [
  &self:SystemSpawnScript.SystemScriptSource.SystemScriptGraph.NiagaraNodeAssignment_3.SetVariables_#1
  &self:SystemSpawnScript.SystemScriptSource.SystemScriptGraph.NiagaraNodeAssignment_4.SetVariables_#2
  &SystemState
  &Timeline
  /Game/Effects/Niagara/Modules/Audio/FN_FakeAudioReactivityLoop
  &self:SystemSpawnScript.SystemScriptSource.SystemScriptGraph.NiagaraNodeAssignment_1.SetVariables_#3
  &self:SystemSpawnScript.SystemScriptSource.SystemScriptGraph.NiagaraNodeAssignment_10.SetVariables_#4
  &self:SystemSpawnScript.SystemScriptSource.SystemScriptGraph.NiagaraNodeAssignment_12.SetVariables_#5
 ]
}
emitters {
 PixieDust {
  emitterUpdateScript [
   &EmitterState(
    Inactive Response=::NewEnumerator1
    Life Cycle Mode=::NewEnumerator1
    Loop Behavior=::NewEnumerator1
    Loop Duration Mode=::NewEnumerator1
    UseLoopDelay=false
   )
   &SpawnRate(
    Spawn Probability=&Lerp_Float(A=0, B=0.7, Alpha=$System.AudioReactivity)
    SpawnRate=18
    Use Spawn Probability=true
   )
  ]
  particleSpawnScript [
   &InitializeParticle(
    AdjustAlpha=false
    AdjustHue=true
    AdjustSaturation=true
    AdjustValue=true
    Color=<rgba 2 1.7198 0.3188 0.03>
    Color Mode=::NewEnumerator3
    Hue Shift Range:Vector2f=<-0.02 0.02>
    Lifetime=2.5
    Saturation Range:Vector2f=<0.9 1>
    Sprite Rotation Mode=::NewEnumerator2
    Sprite Size Mode=::NewEnumerator1
    Sprite UV Mode=::NewEnumerator3
    Uniform Sprite Size Max=130
    Uniform Sprite Size Min=120
    UsePositionOffset=false
    Write Material Random=true
   )
   &CameraOffset(Camera Offset Amount=5)
   &AddVelocity(Velocity=<0 -15 20>)
   &DynamicMaterialParameters(
    Index 0 Param 1=&RandomRangeFloat(Minimum=-0.05, Maximum=0.05)
    Index 0 Param 2=&RandomRangeFloat(Minimum=0, Maximum=32)
    Module.Index0Float0Write=true
    Module.Index0Float1Write=true
    Module.Index0Float2Write=true
    Module.Index0Float3Write=true
   )
   &ShapeLocation(MaskSpawnGroup=false, Override Seed=false, Sphere Radius=7, UseOwnerScale=true)
  ]
  particleUpdateScript [
   &ParticleState()
   &SkeletalMeshLocation(
    Bone / Socket Index=&SelectIntFromArray(
     Array Sampling Mode=::NewEnumerator1
     Direct Array Index=?
     Int Selection Array=DataInterface() {propertyValues {IntData [32,33,34,35,36]}}
     Random Seed=0
    )
    Bone Sampling Mode=::NewEnumerator4
    MaskSpawnGroup=false
    Position Sampling=::NewEnumerator1
    Sampled Position Offset=$Particles.Initial.ShapeLocation.ShapeVector
    Skeletal Mesh=$System.Skeletal Mesh_Dyn
    Source Movement Scale:Vector3f=&VectorFromFloat(Value=&MakeCustomFloatFromBool(
      Bool=$User.IsSkydiving
      True Float=&MakeCustomFloatFromBool(
       Bool=&SetBoolByFloatComparison(
        A=&VectorLength(Vector=&Multiply_Vector(A=$Engine.Owner.Velocity, B=<1 1 0>))
        B=50
        Comparison Type=::NewEnumerator6
       )
       True Float=&FloatFromCurve(FloatCurve=DataInterface() {propertyValues {Curve {
           defaultValue=1
           keys [
            {arriveTangent=0.0754, interpMode=RCIM_Cubic, leaveTangent=0.0754, tangentMode=RCTM_User, time=0, value=0.9}
            {arriveTangent=-0.0997, interpMode=RCIM_Cubic, leaveTangent=-0.0997, tangentMode=RCTM_User, time=1, value=0.5}
           ]
          }}})
       False Float=&FloatFromCurve(FloatCurve=DataInterface() {propertyValues {Curve {
           defaultValue=1
           keys [
            {arriveTangent=0.0115, interpMode=RCIM_Cubic, leaveTangent=0.0115, tangentMode=RCTM_User, time=0, value=0.9}
            {arriveTangent=-0.0997, interpMode=RCIM_Cubic, leaveTangent=-0.0997, tangentMode=RCTM_User, time=1, value=0.7}
           ]
          }}})
      )
      False Float=&Multiply_Float(
       A=&FloatFromCurve(
        FloatCurve=DataInterface() {propertyValues {Curve {
           defaultValue=1
           keys [
            {arriveTangent=-1.7317, interpMode=RCIM_Cubic, leaveTangent=-1.7317, tangentMode=RCTM_User, time=0, value=0.9}
            {interpMode=RCIM_Cubic, tangentMode=RCTM_User, time=1, value=0}
           ]
          }}}
        CurveIndex=&RemapRange(
         Input Value To Remap=&DistanceBetweenPositions(A=$Engine.Emitter.SimulationPosition, B=$Particles.Position)
         Input Min=0
         Input Max=400
         Output Min=0
         Output Max=1
        )
       )
       B=&Lerp_Float(
        A=0
        B=1
        Alpha=&RemapRange(
         Input Value To Remap=&VectorLength(Vector=&Multiply_Vector(A=$Engine.Owner.Velocity, B=<1 1 1>))
         Input Min=0
         Input Max=200
         Output Min=0
         Output Max=1
        )
       )
      )
     ))
   )
   &ScaleColor(
    Scale Alpha=&Multiply_Float(
     A=$System.TODScalar
     B=&Multiply_Float(
      A=&FloatFromCurve(
       FloatCurve=DataInterface() {propertyValues {Curve {
          defaultValue=1
          keys [
           {arriveTangent=4.0437, interpMode=RCIM_Cubic, leaveTangent=4.0437, tangentMode=RCTM_User, time=0, value=0.1}
           {arriveTangent=-0.8452, interpMode=RCIM_Cubic, leaveTangent=-0.8452, tangentMode=RCTM_User, time=0.1418, value=0.9847}
           {arriveTangent=-0.8544, interpMode=RCIM_Cubic, leaveTangent=-0.8544, tangentMode=RCTM_User, time=1, value=0}
          ]
         }}}
       Scale Curve=2
      )
      B=&RemapRange(
       Input Value To Remap=&DistanceBetweenPositions(A=$Engine.Emitter.SimulationPosition, B=$Particles.Position)
       Input Min=0
       Input Max=300
       Output Min=1
       Output Max=0
      )
     )
    )
    Scale RGB=&MakeVectorFromLinearColorRGB(Color=$System.TODColor)
    ScaleA=true
    ScaleRGB=true
   ) {module=ScaleColor001}
   &ScaleSpriteSize(Uniform Curve Sprite Scale=DataInterface() {propertyValues {Curve {
       defaultValue=1
       keys [
        {arriveTangent=-0.2039, interpMode=RCIM_Cubic, leaveTangent=-0.2039, tangentMode=RCTM_User, time=0.0017, value=0.5}
        {interpMode=RCIM_Cubic, tangentMode=RCTM_User, time=1, value=1}
       ]
      }}})
   &Drag(Drag=25, Use Linear Drag=true, Use Rotational Drag=true)
   &SolveForcesAndVelocity(Clamp Velocity=false, Limit Acceleration=false)
   &ScaleColor(
    Scale Alpha=&Multiply_Float(
     A=&Lerp_Float(A=0.5, B=1.3, Alpha=$System.AudioReactivity)
     B=&FN_FrontEndFloatModifier(Gameplay=&Lerp_Float(A=1, B=0.4, Alpha=$System.Timeline.NormalizedTime), Front End=0.8)
    )
    ScaleA=true
    ScaleRGB=true
   )
   &KillParticles(Kill Particles=&SetBoolByFloatComparison(
     A=&MakeFloatFromLinearColor(LinearColor=$Particles.Color, Channel:ENiagaraFloatFromLinearColorOptions=::NewEnumerator3)
     B=0
     Comparison Type=::NewEnumerator7
    ))
  ]
  rendererClasses [&NiagaraSpriteRendererProperties]
  renderers [{rendererClass=&NiagaraSpriteRendererProperties}]
 }
 Pixie_Dust {
  emitterUpdateScript [
   &EmitterState(
    Inactive Response=::NewEnumerator1
    Life Cycle Mode=::NewEnumerator1
    Loop Behavior=::NewEnumerator1
    Loop Duration Mode=::NewEnumerator1
    UseLoopDelay=false
   )
   &SpawnRate(
    Spawn Probability=&Lerp_Float(A=0, B=0.7, Alpha=$System.AudioReactivity)
    SpawnRate=15
    Use Spawn Probability=true
   )
  ]
  particleSpawnScript [
   /SetVariables(Particles.RandomInt:NiagaraInt32=&RandomRangeInt(Minimum=0, Maximum=7)) {isSetParameters=true, module=SetVariables_#6}
   &PartitionParticles(Number Of Partitions=8, Particles Per Partition=4, Partition Mode:ENiagara_PartitionMode_v2=::NewEnumerator1)
   &InitializeParticle(
    AdjustAlpha=false
    AdjustHue=true
    AdjustSaturation=true
    AdjustValue=true
    Color=<rgba 2 1.7198 0.3188 0.03>
    Color Mode=::NewEnumerator3
    Hue Shift Range:Vector2f=<-0.02 0.02>
    Lifetime=2.5
    Saturation Range:Vector2f=<0.9 1>
    Sprite Rotation Mode=::NewEnumerator2
    Sprite Size Mode=::NewEnumerator3
    Sprite UV Mode=::NewEnumerator3
    Uniform Sprite Size=&Multiply_Float(
     A=&RandomRangeFloat(Minimum=120, Maximum=130)
     B=&FN_FrontEndFloatModifier(Gameplay=1, Front End=0.8)
    )
    UsePositionOffset=false
    Write Material Random=true
   )
   &CameraOffset(Camera Offset Amount=5)
   &AddVelocity(Velocity=<0 -15 20>)
   &DynamicMaterialParameters(
    Index 0 Param 1=&RandomRangeFloat(Minimum=-0.05, Maximum=0.05)
    Index 0 Param 2=&RandomRangeFloat(Minimum=0, Maximum=32)
    Module.Index0Float0Write=true
    Module.Index0Float1Write=true
    Module.Index0Float2Write=true
    Module.Index0Float3Write=true
   )
  ]
  particleUpdateScript [
   &ParticleState()
   &SkeletalMeshLocation(
    Bone / Socket Index=&Add_Integer(A=&Multiply_Int(A=$Particles.RandomInt, B=4), B=$Particles.PartitionParticles.Index In Partition)
    Bone Sampling Mode=::NewEnumerator4
    MaskSpawnGroup=false
    Position Sampling=::NewEnumerator1
    Sampled Position Offset=&MergeHelper_VectorSwitch(
     0=&MergeHelper_VectorSwitch(
      0=&SelectVectorFromArray(
       Array Sampling Mode=::NewEnumerator0
       Direct Array Index=?
       Vector Selection Array=DataInterface() {propertyValues {FloatData [
          <12.1043 -2.1186 -4.8536>
          <7.7075 -1.2879 -6.6743>
          <10.374 -1.9803 -4.7231>
          <4.5107 -1.3436 -4.0531>
          <1.0578 -1.1923 -2.4149>
          <8.5727 -2.8564 -0.6967>
          <8.2147 -2.5855 0.117>
          <1.9331 -0.9532 3.66>
          <11.8682 -1.422 6.0568>
          <-2.1783 -0.0616 1.7825>
          <-6.0713 3.4174 0.6358>
         ]}}
       Random Seed=0
      )
      1=&SelectVectorFromArray(
       Array Sampling Mode=::NewEnumerator0
       Direct Array Index=?
       Vector Selection Array=DataInterface() {propertyValues {FloatData [
          <13.1997 -2.1349 -8.4107>
          <12.5784 -2.9523 -4.639>
          <4.9187 -2.1511 -7.9891>
          <3.7365 -1.9257 -8.3032>
          <-0.1422 -2.6159 -4.3811>
          <7.5273 -3.3673 -3.519>
          <-0.6594 -3.258 -3.1036>
          <3.7533 -3.7299 -2.7185>
          <2.2406 -4.1329 -1.2617>
          <1.7552 -3.3102 1.5707>
          <6.1943 -4.188 -0.2337>
          <7.97 -3.1401 2.5243>
          <-1.8418 -2.9285 4.3853>
          <-6.0999 -2.4904 2.6225>
          <4.3014 -2.8786 5.4921>
          <-8.2397 -2.5249 4.0168>
          <1.5502 -1.4877 7.4717>
          <6.9471 -1.4404 7.983>
          <10.9996 -2.1045 7.6277>
         ]}}
       Random Seed=0
      )
      2=&SelectVectorFromArray(
       Array Sampling Mode=::NewEnumerator0
       Direct Array Index=?
       Vector Selection Array=DataInterface() {propertyValues {FloatData [
          <13.5835 -2.7239 -8.1627>
          <8.5022 -1.9974 -9.4686>
          <3.1847 -2.132 -8.996>
          <11.0229 -2.9949 -6.5318>
          <0.9577 -2.9768 -6.5084>
          <-2.7889 -3.2198 -4.2949>
          <-6.0514 -4.0949 -1.9873>
          <-3.366 -4.3479 -1.0844>
          <-0.8601 -3.9585 -2.4468>
          <-0.5962 -3.6058 4.5014>
          <-5.158 -2.0363 8.2433>
          <1.338 -3.8928 5.4209>
          <7.1802 -4.4771 -0.0398>
          <9.7493 -3.4028 -4.203>
          <14.3388 -4.1935 -1.8803>
          <13.3693 -4.3553 0.5492>
          <11.3134 -3.3555 3.7878>
          <9.5748 -3.7437 4.7187>
          <11.6098 -3.6144 7.8542>
          <13.7223 -2.7235 8.9223>
          <5.4559 -2.6173 8.5364>
          <2.4703 -1.8967 9.0377>
         ]}}
       Random Seed=0
      )
      3=&SelectVectorFromArray(
       Array Sampling Mode=::NewEnumerator0
       Direct Array Index=?
       Vector Selection Array=DataInterface() {propertyValues {FloatData [
          <-4.88 0.044 -10.9935>
          <3.2899 0.23 -11.2833>
          <6.2185 -0.1043 -11.2077>
          <4.2649 -1.8626 -9.4559>
          <-6.4952 -2.3322 -8.9241>
          <-7.5242 -3.1069 -4.9835>
          <-5.8895 -4.2306 -1.6552>
          <-0.9618 -4.4573 -0.4779>
          <-2.7323 -4.2396 0.9032>
          <0.2021 -4.0856 -2.354>
          <3.0944 -4.2641 -1.359>
          <-5.0311 -2.7573 9.0379>
          <-2.3219 -3.0952 8.8189>
          <2.3484 -2.3081 9.9155>
          <10.0039 -1.1324 11.1644>
          <6.0605 -3.4446 8.9139>
          <9.9069 -3.2898 9.3914>
          <5.8876 -3.9364 8.2355>
          <12.495 -4.3043 6.8522>
          <5.2545 -3.7289 2.7196>
          <4.0198 -4.215 0.9987>
          <6.1638 -3.5996 -3.7839>
          <8.1976 -4.2307 0.7947>
          <8.8474 -4.3606 -0.5746>
          <8.3734 -3.537 -3.8766>
          <8.5305 -3.9092 -2.9133>
          <8.3126 -2.6832 -6.3893>
          <9.8983 -2.6578 -6.0268>
          <11.5319 -3.3281 -4.2075>
          <11.684 -2.7488 -5.3889>
          <12.2876 -2.4244 -8.2869>
          <10.9829 -2.2151 -8.791>
          <11.614 -0.8961 -10.9645>
          <13.4004 -4.2027 0.6647>
          <11.0466 -3.3301 3.7146>
         ]}}
       Random Seed=0
      )
      Index=$Particles.PartitionParticles.Index In Partition
      Num=::NewEnumerator2
     )
     1=&MergeHelper_VectorSwitch(
      0=&SelectVectorFromArray(
       Array Sampling Mode=::NewEnumerator0
       Direct Array Index=?
       Vector Selection Array=DataInterface() {propertyValues {FloatData [
          <11.7658 -3.598 -0.4081>
          <7.8147 -3.5269 0.6222>
          <-4.3956 1.1866 4.6141>
          <0.4923 -0.6825 5.6672>
          <1.0913 -2.4977 2.0328>
          <6.569 -4.062 2.3387>
          <5.2682 -2.7674 3.5851>
          <11.2378 -3.8772 4.1832>
          <10.65 -3.0196 4.9513>
          <11.62 -2.891 5.6084>
          <1.1841 -1.0771 4.8551>
          <6.9081 -2.2376 5.9513>
          <9.7535 -1.3436 -6.0433>
          <11.6305 -2.808 -5.1278>
          <6.7995 -2.9427 -4.8478>
          <5.6343 -2.6034 -4.8821>
          <-5.4172 2.2125 -1.7372>
          <-4.8795 0.6741 -2.6068>
          <-6.2558 3.5404 -3.0367>
          <-6.0454 2.7868 -3.6444>
         ]}}
       Random Seed=0
      )
      1=&SelectVectorFromArray(
       Array Sampling Mode=::NewEnumerator0
       Direct Array Index=?
       Vector Selection Array=DataInterface() {propertyValues {FloatData [
          <13.9398 -5.3035 0.3745>
          <9.8541 -6.0256 1.6295>
          <4.9227 -5.5203 1.1728>
          <-3.4644 -4.1535 -0.7604>
          <-3.4188 -5.8122 3.0433>
          <2.0445 -6.2985 3.1477>
          <6.5981 -6.8302 3.5926>
          <10.7185 -6.613 2.6297>
          <6.7068 -4.04 7.9918>
          <1.6413 -4.1334 6.3045>
          <-7.9982 -3.107 5.9216>
          <9.6113 -3.07 10.2768>
          <13.4732 -1.7432 -6.7286>
          <8.5275 -2.3305 -5.7308>
          <6.1477 -4.4459 -5.0881>
          <1.8096 -3.3875 -5.2337>
          <2.2904 -4.2631 -4.9943>
          <-1.9697 -4.166 -4.8857>
          <-4.4251 -3.7269 -4.9653>
          <-6.386 -3.6652 -4.9337>
          <-0.5166 -4.8484 -4.6476>
          <-7.1657 -5.2052 -3.3707>
          <-1.2893 -5.6232 -3.2452>
          <-6.9696 -3.9195 -1.6638>
          <11.6427 -4.9797 -0.8203>
         ]}}
       Random Seed=0
      )
      2=&SelectVectorFromArray(
       Array Sampling Mode=::NewEnumerator0
       Direct Array Index=?
       Vector Selection Array=DataInterface() {propertyValues {FloatData [
          <11.641 -5.6728 1.0917>
          <10.5424 -6.6954 2.5735>
          <5.9483 -6.3972 2.0113>
          <-2.0471 -5.1268 -0.196>
          <-5.6521 -5.8092 1.1555>
          <-6.5692 -7.1075 3.3024>
          <-3.0851 -5.5288 6.7308>
          <-0.7217 -7.3641 3.6122>
          <2.422 -7.6414 5.0376>
          <-1.2829 -5.6535 6.772>
          <6.1029 -5.8336 7.1393>
          <2.1951 -6.328 6.626>
          <3.3367 -4.762 7.9468>
          <-0.1721 -5.0425 7.3459>
          <-3.6088 -4.2428 8.875>
          <11.8749 -3.2069 11.8805>
          <13.6865 -1.612 -7.9008>
          <10.075 -1.6929 -7.7463>
          <0.2315 -1.9831 -7.8364>
          <2.7036 -1.5583 -6.7586>
          <-2.0812 -2.5538 -5.9072>
          <-5.7674 -5.6626 -1.5329>
          <-0.3899 -3.1092 -5.8358>
          <-4.3939 -6.0523 -4.2844>
          <-4.9464 -6.5225 -3.3622>
          <-0.9313 -4.3307 -5.5661>
          <-2.7343 -5.8549 -4.5366>
          <0.6389 -6.6813 -3.7188>
          <1.8613 -4.0525 -5.7092>
          <5.0787 -3.5626 -5.9212>
          <6.5801 -3.2552 -6.0342>
          <11.5875 -5.9147 -4.9222>
          <9.7935 -5.4003 -5.361>
          <13.8616 -7.1924 -3.2854>
          <12.5972 -6.1818 -0.8214>
         ]}}
       Random Seed=0
      )
      3=&SelectVectorFromArray(
       Array Sampling Mode=::NewEnumerator0
       Direct Array Index=?
       Vector Selection Array=DataInterface() {propertyValues {FloatData [
          <13.5131 -5.7583 1.4295>
          <7.2605 -5.8622 1.677>
          <4.4776 -6.6558 2.8925>
          <1.6866 -5.6577 0.7376>
          <-1.6198 -6.3878 2.3964>
          <-6.577 -8.1401 5.6945>
          <-1.4871 -8.2824 5.373>
          <3.6376 -8.1844 5.1915>
          <-0.489 -4.3727 10.3232>
          <2.1494 -3.994 11.5083>
          <-3.2046 -1.9845 13.0197>
          <4.3737 -2.2531 13.0521>
          <-6.2298 -2.0545 -6.5724>
          <-4.4524 -2.8303 -6.4319>
          <-3.4983 -5.7747 -5.1832>
          <-2.9351 -7.4859 -1.9071>
          <2.1201 -6.8309 -0.6298>
          <3.2178 -7.3425 -1.0302>
          <5.1423 -7.3351 -0.8961>
          <2.4541 -7.6752 -1.9195>
          <3.5209 -7.0318 -4.2472>
          <-2.3718 -5.7237 -5.2268>
          <-2.4962 -2.833 -6.455>
          <3.5565 -5.7296 -5.2477>
          <5.3577 -3.985 -6.2203>
          <4.3764 -6.5019 -4.7139>
          <5.6152 -7.7699 -2.2721>
          <8.8316 -5.7238 -5.2323>
          <12.1859 -7.0488 -4.3676>
          <12.0661 -1.9439 -6.9926>
          <9.1538 -1.1743 -8.0301>
          <9.969 -8.4099 7.0004>
          <11.1431 -3.3415 12.6745>
         ]}}
       Random Seed=0
      )
      Index=$Particles.PartitionParticles.Index In Partition
      Num=::NewEnumerator2
     )
     2=&MergeHelper_VectorSwitch(
      0=&SelectVectorFromArray(
       Array Sampling Mode=::NewEnumerator0
       Direct Array Index=?
       Vector Selection Array=DataInterface() {propertyValues {FloatData [
          <-5.9399 2.4876 -2.483>
          <1.3927 -0.2699 -4.8192>
          <4.3234 -0.5609 -5.3699>
          <11.8038 -1.7046 -6.1407>
          <5.2786 -3.6684 -0.8376>
          <9.4459 -2.1042 -4.4151>
          <1.4471 -0.7614 -3.7492>
          <6.9768 -2.4766 -2.8133>
          <3.5759 -1.5034 -3.2352>
          <4.1047 -2.8848 -1.8721>
          <0.8201 -2.4111 -0.4422>
          <9.07 -3.5758 0.5087>
          <0.4033 -1.8452 0.1267>
          <4.5094 -2.0277 2.2354>
          <11.4969 -2.9849 4.3867>
          <7.3333 -2.5212 4.523>
          <-0.2559 -0.7493 4.1809>
          <0.4402 -0.617 4.8044>
          <-4.5606 1.1089 2.7399>
          <-5.0443 1.9142 3.2012>
         ]}}
       Random Seed=0
      )
      1=&SelectVectorFromArray(
       Array Sampling Mode=::NewEnumerator0
       Direct Array Index=?
       Vector Selection Array=DataInterface() {propertyValues {FloatData [
          <3.3646 -1.0532 -9.1352>
          <4.8039 -1.3602 -9.0435>
          <0.4326 -2.7579 -6.3535>
          <3.7452 -3.1468 -6.0234>
          <6.3564 -3.65 -5.1381>
          <13.4828 -2.9968 -7.8299>
          <9.8335 -4.3439 -3.7535>
          <7.6125 -4.6599 -3.2974>
          <3.6911 -4.2338 -3.5495>
          <-6.7388 -4.219 0.5388>
          <10.0182 -3.8104 1.4943>
          <-9.6163 -3.0081 2.3103>
          <-1.2823 -3.543 2.8572>
          <0.4469 -3.5933 2.7843>
          <7.321 -3.5694 1.7831>
          <9.5626 -3.4941 4.3056>
          <6.9859 -3.5047 4.477>
          <3.016 -3.6311 3.7282>
          <3.3806 -2.6482 6.805>
          <-3.3884 -3.4047 4.3438>
          <-1.416 -2.9498 5.9555>
          <-6.1985 -3.2786 3.8874>
          <-2.4302 -2.597 6.4563>
          <-7.092 -2.9147 5.2939>
         ]}}
       Random Seed=0
      )
      2=&SelectVectorFromArray(
       Array Sampling Mode=::NewEnumerator0
       Direct Array Index=?
       Vector Selection Array=DataInterface() {propertyValues {FloatData [
          <-7.456 -1.0231 -10.5241>
          <-4.2 -3.2983 -7.7055>
          <-3.2176 -3.0004 -8.6432>
          <5.2729 -1.6967 -11.078>
          <8.956 -2.3102 -10.9651>
          <8.9057 -3.5169 -8.8347>
          <13.1999 -3.9403 -8.4831>
          <5.716 -4.1443 -5.2116>
          <3.0727 -3.9951 -6.009>
          <-0.1701 -3.8589 -6.6687>
          <0.1967 -5.105 -3.2144>
          <-7.5075 -4.4639 -3.7969>
          <-3.7179 -5.2313 -2.9375>
          <7.6246 -5.9122 -1.5863>
          <2.8896 -5.8026 0.9212>
          <7.6531 -4.5692 2.1964>
          <9.9763 -3.331 3.1678>
          <8.6903 -3.0759 3.759>
          <14.2663 -5.8825 -1.7514>
          <1.0856 -3.2246 3.2817>
          <-3.4101 -3.0515 4.4458>
          <-1.6187 -2.9622 4.4533>
          <0.5373 -2.2327 6.4225>
          <-2.0771 -2.1671 6.9585>
          <8.7728 -2.3072 5.5475>
          <6.9245 -1.8524 6.895>
          <8.4521 -1.8456 6.7797>
          <11.6126 -2.228 5.6094>
          <11.8803 -2.7994 4.4601>
          <11.0175 -1.6334 7.1706>
         ]}}
       Random Seed=0
      )
      3=&SelectVectorFromArray(
       Array Sampling Mode=::NewEnumerator0
       Direct Array Index=?
       Vector Selection Array=DataInterface() {propertyValues {FloatData [
          <-5.429 1.36 -13.8215>
          <-5.9982 -3.0291 -10.7798>
          <-6.2377 -4.791 2.4662>
          <-5.3736 -5.856 1.5637>
          <-4.3847 -4.2206 -7.6151>
          <-4.5102 -6.003 -1.3919>
          <-3.8091 -6.1455 0.1757>
          <2.8731 -4.323 -5.5709>
          <-1.7348 -5.9561 -1.6137>
          <2.8616 -5.9561 -1.7383>
          <3.5284 -5.9398 1.1858>
          <5.1226 -5.9824 0.9031>
          <7.932 -3.5378 3.2536>
          <9.8194 -5.5166 2.223>
          <3.2553 -4.3222 -7.4917>
          <7.0369 -4.3586 -7.1591>
          <6.0003 -3.1769 -11.6147>
          <8.3578 -4.3891 -8.9841>
          <-2.8972 -2.8144 -11.5036>
          <-0.9827 -1.5549 -12.628>
          <4.4991 -2.6082 -12.4733>
          <0.8117 1.7041 -14.419>
          <12.3988 -5.0475 2.9552>
          <10.7308 -5.2195 2.6859>
          <11.7094 -6.1045 -0.1857>
          <9.3034 -5.7891 -2.2668>
          <13.2722 -5.2864 -3.6934>
          <13.0805 -4.887 -4.8299>
          <12.7141 -4.628 -5.5518>
          <11.4309 -4.3931 -7.924>
          <12.8578 -3.9559 -10.7665>
          <12.1073 -3.3234 -11.9811>
          <10.0903 -0.5493 -14.1518>
          <10.0528 1.7838 -15.016>
          <-6.9942 -1.5206 7.1244>
         ]}}
       Random Seed=0
      )
      Index=$Particles.PartitionParticles.Index In Partition
      Num=::NewEnumerator2
     )
     3=&MergeHelper_VectorSwitch(
      0=&SelectVectorFromArray(
       Array Sampling Mode=::NewEnumerator0
       Direct Array Index=?
       Vector Selection Array=DataInterface() {propertyValues {FloatData [
          <-4.3516 1.8791 -2.5928>
          <0.2773 -0.1642 5.7218>
          <1.8137 -0.4164 5.672>
          <3.8953 -1.5305 -0.06>
          <1.1507 -0.0262 -3.7941>
          <3.0691 -0.196 -4.139>
          <2.3652 0.1254 -5.3059>
          <4.8556 -0.5573 -3.6843>
          <5.5916 -1.1 -2.3591>
          <5.7724 -0.3941 -4.4004>
          <7.686 -0.5557 -6.069>
          <8.2366 -0.6333 -4.2605>
          <12.2368 -0.5516 -5.8611>
          <12.9611 -1.4337 5.1475>
          <4.9374 -0.7863 4.5167>
          <8.4264 -1.9569 2.2314>
          <6.0278 -1.7344 -0.024>
         ]}}
       Random Seed=0
      )
      1=&SelectVectorFromArray(
       Array Sampling Mode=::NewEnumerator0
       Direct Array Index=?
       Vector Selection Array=DataInterface() {propertyValues {FloatData [
          <12.8365 -0.6775 -6.2608>
          <10.4235 -2.1158 7.7189>
          <11.5574 -3.0518 2.9138>
          <3.2858 -1.6914 9.5122>
          <4.1973 -1.7773 -3.3614>
          <-0.1739 -0.8037 -5.2139>
          <7.0392 -0.953 -6.4499>
          <12.936 -2.4902 -1.4546>
          <11.9473 -2.769 0.5989>
          <9.2288 -2.7886 -0.2875>
          <9.8262 -2.9326 2.1632>
          <2.0307 -1.928 5.6448>
          <-3.3671 -2.319 3.9217>
          <-1.4211 -2.7778 2.5841>
         ]}}
       Random Seed=0
      )
      2=&SelectVectorFromArray(
       Array Sampling Mode=::NewEnumerator0
       Direct Array Index=?
       Vector Selection Array=DataInterface() {propertyValues {FloatData [
          <-4.4711 -0.6173 -6.3131>
          <6.623 -1.2176 -7.0585>
          <11.743 -0.814 -6.7735>
          <4.2162 -0.3367 -5.3431>
          <10.8506 -1.6587 -2.4935>
          <9.0163 -1.6349 -2.6798>
          <0.6745 -1.7768 -2.8461>
          <7.6939 -2.5255 0.3525>
          <-0.6194 -2.6511 -0.4474>
          <-0.337 -3.2806 3.7298>
          <1.8761 -3.7015 4.998>
          <7.877 -2.7359 2.6185>
          <11.8046 -3.5613 6.071>
          <6.819 -2.2003 11.0462>
          <4.0493 -2.3355 8.9503>
          <-3.3545 -2.247 9.0877>
          <-4.1908 -2.2673 7.664>
          <-5.8859 -3.043 2.9149>
          <-8.0519 -2.3843 -1.6861>
         ]}}
       Random Seed=0
      )
      3=&SelectVectorFromArray(
       Array Sampling Mode=::NewEnumerator0
       Direct Array Index=?
       Vector Selection Array=DataInterface() {propertyValues {FloatData [
          <6.1685 0.3696 -6.2975>
          <-4.8639 -1.8373 -7.3626>
          <-2.0174 0.5331 -5.8391>
          <-3.1516 -0.5679 -4.1411>
          <-6.3018 -1.9311 -1.6156>
          <-6.5804 -2.3277 -0.403>
          <-6.8947 -3.7939 5.787>
          <-1.3032 -2.6788 3.5167>
          <-5.4932 -3.194 6.8793>
          <1.0961 -2.22 2.2465>
          <5.9961 -1.3285 -1.8281>
          <-0.7535 -3.7933 6.3305>
          <-6.039 -2.4903 7.7988>
          <1.4174 -2.0497 11.9384>
          <2.3615 -2.1639 10.0196>
          <12.9838 -1.8908 12.8261>
          <11.5712 -1.4234 -0.1103>
          <7.8089 -0.7844 -2.9083>
          <7.2847 -2.9679 5.2592>
          <7.6955 1.0736 -5.8073>
          <9.7901 0.0046 -6.5376>
          <12.407 -1.8103 3.5479>
          <9.2367 -2.6552 4.8918>
          <8.7347 -1.9103 9.029>
          <9.038 -1.89 9.3182>
          <11.4683 -1.813 9.8881>
          <12.4654 -1.8208 10.6203>
         ]}}
       Random Seed=0
      )
      Index=$Particles.PartitionParticles.Index In Partition
      Num=::NewEnumerator2
     )
     4=&MergeHelper_VectorSwitch(
      0=&SelectVectorFromArray(
       Array Sampling Mode=::NewEnumerator0
       Direct Array Index=?
       Vector Selection Array=DataInterface() {propertyValues {FloatData [
          <-3.6277 0.9295 -4.4257>
          <3.9914 -1.4836 -4.8789>
          <-1.7847 -0.9704 1.8692>
          <1.8967 -1.5917 3.09>
          <0.7674 -1.9866 -0.3333>
          <2.4262 -1.8776 2.4124>
          <1.3955 -2.3538 -2.6472>
          <2.5033 -2.6229 -0.6001>
          <6.5165 -2.7255 -3.9179>
          <8.7002 -3.0475 -0.5193>
          <11.975 -2.0392 -6.6812>
          <15.247 -4.4234 -2.9625>
          <12.8299 -2.2913 1.3619>
          <13.8333 -2.1719 5.3593>
          <10.0696 -1.6105 5.605>
         ]}}
       Random Seed=0
      )
      1=&SelectVectorFromArray(
       Array Sampling Mode=::NewEnumerator0
       Direct Array Index=?
       Vector Selection Array=DataInterface() {propertyValues {FloatData [
          <-4.3231 -2.7616 -0.3465>
          <-1.9158 -2.6951 -6.9306>
          <5.0894 -2.6172 -8.8124>
          <9.2933 -2.8244 -8.6546>
          <-0.8846 -4.5159 -4.6549>
          <9.8747 -5.3669 -5.0117>
          <7.2073 -5.3458 -3.9482>
          <-3.9114 -4.6073 -2.1205>
          <-0.819 -2.4114 -0.325>
          <11.0058 -1.3652 0.1338>
          <3.1952 -2.1787 2.4905>
          <-4.9162 -1.3256 6.8921>
          <6.0971 -2.4531 5.9588>
          <7.8905 -2.2874 3.966>
          <11.6165 -1.7322 2.1684>
          <13.2292 -1.7297 2.3598>
          <11.3897 -2.3399 4.7692>
          <9.786 -5.3439 -2.6604>
         ]}}
       Random Seed=0
      )
      2=&SelectVectorFromArray(
       Array Sampling Mode=::NewEnumerator0
       Direct Array Index=?
       Vector Selection Array=DataInterface() {propertyValues {FloatData [
          <-7.0278 -2.649 -9.9147>
          <-3.0515 -3.6446 -8.0444>
          <1.4428 -3.9307 -8.1461>
          <-3.8762 -5.0429 -6.6475>
          <-8.2772 -2.0958 8.0121>
          <-2.3431 -1.6196 9.0105>
          <1.0744 -2.4083 7.2567>
          <-0.6173 -1.7113 3.4053>
          <4.9128 -2.2958 8.7544>
          <3.1143 -1.9833 4.911>
          <-0.0756 -2.4849 -0.2038>
          <-0.8642 -4.9078 -1.4021>
          <-1.6859 -5.6979 -2.4735>
          <1.8346 -6.1346 -5.1335>
          <5.7576 -5.086 -1.4346>
          <11.2551 -3.028 -9.9363>
          <12.724 -3.6924 -9.032>
          <9.1609 -6.0913 -6.9122>
          <6.6065 -6.2863 -5.5805>
          <8.7485 -6.1861 -2.9457>
          <9.6759 -4.8302 -1.1682>
          <14.3157 -5.0739 -1.3502>
          <12.4968 -3.4863 -0.1716>
          <8.4883 -0.9547 1.8739>
          <12.4743 -0.9182 1.2302>
          <11.8195 -0.8785 2.2643>
          <11.2928 -1.1099 3.3139>
          <9.0939 -1.3232 3.6826>
          <12.3074 -1.3442 4.1329>
          <12.976 -1.8 5.6218>
          <7.6234 -2.3266 7.8405>
          <8.0079 -2.1671 9.6077>
          <5.2905 -1.3672 10.053>
          <8.1096 -1.1093 10.5365>
         ]}}
       Random Seed=0
      )
      3=&SelectVectorFromArray(
       Array Sampling Mode=::NewEnumerator0
       Direct Array Index=?
       Vector Selection Array=DataInterface() {propertyValues {FloatData [
          <-3.3902 -2.9333 -10.292>
          <-0.2034 -3.0304 -9.8568>
          <-6.3513 -6.2862 -2.8768>
          <-1.3116 -6.4448 -3.0494>
          <3.7604 -6.6673 -7.5697>
          <3.8266 -6.5441 -3.0932>
          <5.5951 -6.5434 -3.0294>
          <7.573 -6.575 -3.1117>
          <3.604 -5.1216 -1.0702>
          <4.3402 -2.3993 0.2925>
          <-2.35 -1.1909 4.29>
          <-6.9226 -1.943 6.4486>
          <-4.1577 -1.8481 10.6837>
          <1.2653 -0.7519 1.7511>
          <1.9979 -2.2531 9.1552>
          <5.7708 -1.7586 6.7871>
          <7.1062 -1.2791 5.4568>
          <6.447 -1.9008 7.7961>
          <-0.5993 -0.5178 12.0842>
          <11.6004 -1.6641 7.2494>
          <9.5539 -1.9651 8.8494>
          <7.6521 -1.4708 6.0838>
          <8.862 -1.8939 10.7816>
          <10.3346 -1.6254 11.7312>
          <7.8893 -1.6928 11.4459>
          <11.2463 -1.3228 12.5201>
          <10.5884 -0.0459 13.2321>
          <12.4944 -0.1564 13.4988>
          <10.9103 -0.5782 2.2549>
          <9.4402 -0.6447 2.9111>
          <8.8664 -1.104 0.6854>
          <9.4591 -3.5294 0.1145>
          <11.0986 -3.167 0.2352>
          <12.5312 -6.1735 -2.1822>
          <13.4159 -6.5529 -2.839>
          <7.9122 -7.0058 -5.6185>
          <8.4176 -5.909 -8.8174>
         ]}}
       Random Seed=0
      )
      Index=$Particles.PartitionParticles.Index In Partition
      Num=::NewEnumerator2
     )
     5=&MergeHelper_VectorSwitch(
      0=&SelectVectorFromArray(
       Array Sampling Mode=::NewEnumerator0
       Direct Array Index=?
       Vector Selection Array=DataInterface() {propertyValues {FloatData [
          <-5.1578 1.9054 -0.8273>
          <-3.9725 1.098 -1.5476>
          <-5.2462 1.5696 -0.4986>
          <-1.5438 -1.0028 1.1676>
          <-0.0204 -0.9908 1.6983>
          <0.4521 -1.3515 1.5526>
          <3.0397 -1.1901 2.3487>
          <-0.2117 -1.5702 0.1214>
          <-0.5793 -1.3477 -0.5654>
          <5.108 -1.1633 -1.288>
          <10.3958 -0.8387 -2.9407>
          <14.6994 -1.276 -1.7853>
          <8.6863 -3.0108 0.5665>
          <8.6114 -2.2845 2.4254>
          <12.8652 -1.2243 3.7266>
          <9.6086 -1.047 5.1863>
          <0.274 0.0091 4.8907>
          <12.5257 -2.6842 2.5819>
          <14.993 -2.6501 -0.483>
          <13.1035 -2.3405 -5.3676>
          <-3.856 1.4549 -3.5103>
         ]}}
       Random Seed=0
      )
      1=&SelectVectorFromArray(
       Array Sampling Mode=::NewEnumerator0
       Direct Array Index=?
       Vector Selection Array=DataInterface() {propertyValues {FloatData [
          <-1.7055 -1.1032 -2.859>
          <-0.1354 -0.8444 7.6557>
          <4.9793 -0.9123 7.8254>
          <13.0217 -1.0836 7.8683>
          <6.0195 -1.2375 5.4887>
          <3.285 -1.2344 6.2281>
          <4.5033 -1.5303 4.4788>
          <-3.8577 -3.1258 -0.2289>
          <-2.8189 -3.4562 0.0758>
          <11.6981 -2.5241 -1.2571>
          <12.5786 -5.1098 1.8912>
          <6.3832 -4.741 2.065>
          <10.2963 -3.0417 -5.9101>
          <5.8017 -2.7921 -5.5189>
          <-2.1148 -2.373 -4.7559>
         ]}}
       Random Seed=0
      )
      2=&SelectVectorFromArray(
       Array Sampling Mode=::NewEnumerator0
       Direct Array Index=?
       Vector Selection Array=DataInterface() {propertyValues {FloatData [
          <4.5869 -3.0303 -5.6283>
          <5.2636 -2.1312 -5.005>
          <-1.5778 -2.5643 -4.8688>
          <-3.1741 -1.4215 -4.5869>
          <6.8693 -0.8555 9.5713>
          <-7.2875 -0.8143 9.1467>
          <-3.6473 -1.2227 7.277>
          <-1.2155 -1.258 6.4993>
          <6.089 -1.5621 5.7697>
          <-1.8887 -1.7562 5.2235>
          <1.159 -4.8082 3.2881>
          <-2.6427 -4.3475 3.3985>
          <-5.6077 -2.4967 -1.298>
          <3.8034 -0.7952 -2.4906>
          <12.3866 -0.8882 -2.2261>
          <11.4616 -1.1098 6.4045>
          <6.8794 -2.8926 4.8565>
          <3.108 -5.126 3.1574>
          <9.9145 -4.7903 3.6518>
          <12.4826 -5.1769 3.4837>
          <12.464 -3.0153 -0.9904>
          <7.7653 -4.836 0.1871>
          <6.1273 -3.6699 -0.6122>
          <2.1146 -4.3199 -0.1632>
          <2.7544 -5.0259 0.5535>
          <3.3708 -3.4367 -0.7663>
          <0.6427 -5.1229 0.779>
          <-0.5874 -3.883 -0.4203>
          <-0.0345 -3.048 -1.0146>
          <-5.8764 -3.9505 -0.2313>
          <-5.4084 -2.2201 -4.6997>
         ]}}
       Random Seed=0
      )
      3=&SelectVectorFromArray(
       Array Sampling Mode=::NewEnumerator0
       Direct Array Index=?
       Vector Selection Array=DataInterface() {propertyValues {FloatData [
          <-4.7393 -0.9887 -5.0391>
          <-2.2928 -2.9104 -5.7796>
          <6.1548 -2.4149 -5.8731>
          <13.1554 -3.4873 -7.8814>
          <-4.7239 -1.3112 -1.9631>
          <12.5562 -1.2677 -2.4636>
          <12.6586 -5.3059 1.6287>
          <10.2859 -5.4994 2.8129>
          <-3.8334 -0.5616 10.9978>
          <-6.0557 -1.06 6.7307>
          <-0.6388 -0.9882 7.9205>
          <-7.2058 -2.0335 5.8789>
          <-0.8547 -1.1292 7.0045>
          <-5.1 -5.1761 3.6475>
          <2.3794 -0.9473 7.507>
          <3.6389 -0.9126 9.0675>
          <5.2491 -1.6964 6.9081>
          <9.3466 -0.3483 11.9851>
          <8.9132 0.446 13.3495>
          <8.7911 -0.7911 8.8764>
          <8.8193 -0.766 10.3132>
          <5.3365 -2.7805 5.8168>
          <4.5911 -4.0237 4.8133>
          <3.3158 -5.1599 3.9234>
          <5.4236 -5.0695 4.0445>
          <5.0902 -5.4027 3.4335>
          <9.9923 -3.8498 5.0754>
          <9.5982 -4.0249 -0.189>
          <10.9539 -4.5343 0.2528>
          <7.6953 -2.7814 -1.1807>
          <4.4857 -3.7829 -0.3758>
          <2.1466 -3.8906 -0.3099>
         ]}}
       Random Seed=0
      )
      Index=$Particles.PartitionParticles.Index In Partition
      Num=::NewEnumerator2
     )
     6=&MergeHelper_VectorSwitch(
      0=&SelectVectorFromArray(
       Array Sampling Mode=::NewEnumerator0
       Direct Array Index=?
       Vector Selection Array=DataInterface() {propertyValues {FloatData [
          <-3.761 1.1597 1.2311>
          <-1.3654 -1.3107 -1.3163>
          <-0.3763 -0.9872 -2.3599>
          <-0.1482 -0.4941 -2.8721>
          <2.3408 -1.271 -3.5664>
          <3.8169 -1.53 -3.3955>
          <6.2615 -2.3869 -4.8684>
          <10.9325 -2.7058 -3.5515>
          <9.8979 -3.0599 -5.3685>
          <8.8812 -2.7944 -3.0965>
          <8.9036 -3.8433 -1.3385>
          <7.2562 -1.2967 1.3498>
          <13.7167 -1.7698 1.5511>
          <6.5194 -0.4218 2.0331>
          <2.3818 -0.764 4.7861>
         ]}}
       Random Seed=0
      )
      1=&SelectVectorFromArray(
       Array Sampling Mode=::NewEnumerator0
       Direct Array Index=?
       Vector Selection Array=DataInterface() {propertyValues {FloatData [
          <-4.0817 -3.9455 -5.5213>
          <4.7583 -4.3306 -4.8289>
          <-0.3231 -5.0707 -3.1413>
          <5.5284 -4.4534 -5.1454>
          <4.9305 -4.4784 -4.028>
          <10.129 -4.5691 -4.9684>
          <12.0582 -5.5838 -3.9385>
          <-0.795 -2.8229 4.2663>
          <1.4047 -2.4699 3.5063>
          <-4.8 -2.3676 1.3135>
          <2.6311 -2.5668 3.5351>
          <2.4359 -3.345 5.3491>
          <8.0303 -3.4407 4.4223>
          <-3.8881 -3.5829 0.4695>
          <-3.7536 -3.9789 0.1879>
          <-0.2705 -5.2651 -1.1072>
          <3.7999 -4.9959 -0.3637>
          <5.1287 -2.5692 1.3638>
          <8.3426 -4.9458 -0.0875>
          <7.1737 -3.0868 0.9744>
          <11.0994 -2.2986 2.2758>
          <11.5646 -2.6167 2.9808>
          <7.656 -2.2563 2.3544>
          <6.7896 -2.2875 2.482>
          <8.0937 -2.6559 3.2572>
          <11.5693 -3.626 4.6252>
          <10.784 -3.5266 6.7521>
          <11.745 -3.4563 6.995>
         ]}}
       Random Seed=0
      )
      2=&SelectVectorFromArray(
       Array Sampling Mode=::NewEnumerator0
       Direct Array Index=?
       Vector Selection Array=DataInterface() {propertyValues {FloatData [
          <-4.303 -4.746 -5.2282>
          <4.6401 -5.2479 -7.2137>
          <11.3108 -5.7185 -5.6214>
          <1.4063 -5.1724 -5.1511>
          <-3.5512 -6.6473 -3.092>
          <4.0352 -6.1819 -1.0315>
          <2.5153 -3.5042 0.4908>
          <11.4815 -4.0918 0.2202>
          <7.1512 -2.5356 2.4442>
          <13.0211 -3.2721 3.7142>
          <-3.312 -2.6476 2.7436>
          <-3.6272 -3.5195 4.1478>
          <-5.7665 -4.0102 6.3304>
          <-4.1357 -3.6787 7.1776>
          <2.1676 -4.2937 6.2719>
          <7.021 -4.3164 5.6858>
          <-0.4263 -3.6142 7.6786>
          <7.7747 -4.4222 6.2278>
          <7.6336 -4.3318 7.2415>
          <10.4536 -4.5485 7.1658>
          <13.2536 -4.3625 5.2677>
          <11.9147 -3.9665 8.0727>
         ]}}
       Random Seed=0
      )
      3=&SelectVectorFromArray(
       Array Sampling Mode=::NewEnumerator0
       Direct Array Index=?
       Vector Selection Array=DataInterface() {propertyValues {FloatData [
          <-6.4844 -5.1564 -7.3043>
          <-2.2142 -5.4149 -6.3664>
          <3.6941 -6.4115 -6.3214>
          <0.6102 -7.5123 -4.7899>
          <-3.0589 -7.5836 -3.3224>
          <-6.3453 -3.4956 0.4339>
          <2.0756 -2.025 1.5667>
          <-3.0149 -3.2221 3.6485>
          <-6.9804 -4.0853 4.8624>
          <9.9745 -6.8582 -6.6178>
          <13.7093 -4.2371 -8.2069>
          <11.0953 -5.6976 -0.3027>
          <12.5871 -3.502 0.4424>
          <7.0921 -6.1476 -0.6477>
          <8.3702 -5.4546 -0.2939>
          <4.5976 -5.4959 -0.3346>
          <9.0181 -3.6333 0.4375>
          <7.6034 -4.0894 0.2985>
          <5.5929 -1.8237 1.9257>
          <2.5691 -3.8099 4.4581>
          <3.503 -4.6068 5.5355>
          <2.1185 -4.1599 8.1777>
          <2.151 -3.0839 9.1954>
         ]}}
       Random Seed=0
      )
      Index=$Particles.PartitionParticles.Index In Partition
      Num=::NewEnumerator2
     )
     7=&MergeHelper_VectorSwitch(
      0=&SelectVectorFromArray(
       Array Sampling Mode=::NewEnumerator0
       Direct Array Index=?
       Vector Selection Array=DataInterface() {propertyValues {FloatData [
          <-6.3138 2.6059 -0.8865>
          <1.4299 -1.4069 -3.0267>
          <-3.6126 1.4961 0.9269>
          <6.2303 -2.4933 -2.9358>
          <9.5919 -1.5987 -4.1948>
          <10.1598 -1.299 -4.8814>
          <8.8789 -3.4234 -1.6053>
          <13.6549 -1.3777 -5.2434>
          <14.6995 -2.3664 -3.8777>
          <14.6282 -1.7444 -4.6417>
          <12.732 -3.7277 -1.9665>
          <15.5552 -4.013 -1.8686>
          <8.3461 -2.6986 0.3931>
          <16.631 -4.1457 -0.1946>
          <10.5013 -1.5172 2.5817>
         ]}}
       Random Seed=0
      )
      1=&SelectVectorFromArray(
       Array Sampling Mode=::NewEnumerator0
       Direct Array Index=?
       Vector Selection Array=DataInterface() {propertyValues {FloatData [
          <-4.5714 -1.4207 -5.5158>
          <0.1081 -2.8225 -3.7909>
          <-4.4136 -4.2122 -0.1365>
          <-2.6189 -3.7005 1.0268>
          <1.7404 -4.8172 -0.1497>
          <1.0669 -4.0281 -2.843>
          <6.6757 -2.2236 -4.923>
          <8.1663 -3.6134 1.8617>
          <9.7084 -4.8378 -2.2276>
          <13.4025 -4.2282 -3.1273>
          <7.3098 -3.2729 5.3722>
         ]}}
       Random Seed=0
      )
      2=&SelectVectorFromArray(
       Array Sampling Mode=::NewEnumerator0
       Direct Array Index=?
       Vector Selection Array=DataInterface() {propertyValues {FloatData [
          <-7.0595 -4.9132 -2.3866>
          <-3.0653 -5.1886 -1.9869>
          <-2.8283 -5.6349 0.4002>
          <0.1439 -5.4999 1.1075>
          <-6.1169 -3.4131 4.3835>
          <1.9564 -3.9152 6.0011>
          <6.9929 -3.5731 3.842>
          <11.1577 -3.7518 4.9511>
          <12.6779 -4.7509 2.6513>
          <8.3524 -6.2053 -0.8529>
          <11.4717 -3.7291 -3.5226>
          <13.1501 -4.1712 -3.2218>
          <1.127 -2.9928 -4.3987>
          <-7.5842 -2.5293 -5.0678>
         ]}}
       Random Seed=0
      )
      3=&SelectVectorFromArray(
       Array Sampling Mode=::NewEnumerator0
       Direct Array Index=?
       Vector Selection Array=DataInterface() {propertyValues {FloatData [
          <-4.7817 -4.3752 3.3004>
          <0.9742 -4.366 3.6931>
          <1.5181 -4.8336 3.268>
          <2.8004 -6.1707 2.158>
          <13.3708 -6.012 3.2787>
          <12.2833 -5.7611 3.3446>
          <10.9562 -7.1959 1.0802>
          <6.6153 -7.0815 0.3367>
          <8.8292 -2.4325 -4.4459>
          <10.8354 -2.3201 -4.6793>
          <-0.3755 -6.8395 -0.9403>
          <-3.4378 -3.9636 -3.3355>
         ]}}
       Random Seed=0
      )
      Index=$Particles.PartitionParticles.Index In Partition
      Num=::NewEnumerator2
     )
     Index=$Particles.RandomInt
     Num=::NewEnumerator6
    )
    Skeletal Mesh=$System.Skeletal Mesh_Dyn
    Source Movement Scale:Vector3f=&VectorFromFloat(Value=&MakeCustomFloatFromBool(
      Bool=$User.IsSkydiving
      True Float=&MakeCustomFloatFromBool(
       Bool=&SetBoolByFloatComparison(
        A=&VectorLength(Vector=&Multiply_Vector(A=$Engine.Owner.Velocity, B=<1 1 0>))
        B=50
        Comparison Type=::NewEnumerator6
       )
       True Float=&FloatFromCurve(FloatCurve=DataInterface() {propertyValues {Curve {
           defaultValue=1
           keys [
            {arriveTangent=0.0754, interpMode=RCIM_Cubic, leaveTangent=0.0754, tangentMode=RCTM_User, time=0, value=0.9}
            {arriveTangent=-0.0997, interpMode=RCIM_Cubic, leaveTangent=-0.0997, tangentMode=RCTM_User, time=1, value=0.5}
           ]
          }}})
       False Float=&FloatFromCurve(FloatCurve=DataInterface() {propertyValues {Curve {
           defaultValue=1
           keys [
            {arriveTangent=0.0115, interpMode=RCIM_Cubic, leaveTangent=0.0115, tangentMode=RCTM_User, time=0, value=0.9}
            {arriveTangent=-0.0997, interpMode=RCIM_Cubic, leaveTangent=-0.0997, tangentMode=RCTM_User, time=1, value=0.7}
           ]
          }}})
      )
      False Float=&Multiply_Float(
       A=&FloatFromCurve(
        FloatCurve=DataInterface() {propertyValues {Curve {
           defaultValue=1
           keys [
            {arriveTangent=-1.7317, interpMode=RCIM_Cubic, leaveTangent=-1.7317, tangentMode=RCTM_User, time=0, value=0.9}
            {interpMode=RCIM_Cubic, tangentMode=RCTM_User, time=1, value=0}
           ]
          }}}
        CurveIndex=&RemapRange(
         Input Value To Remap=&DistanceBetweenPositions(A=$Engine.Emitter.SimulationPosition, B=$Particles.Position)
         Input Min=0
         Input Max=400
         Output Min=0
         Output Max=1
        )
       )
       B=&Lerp_Float(
        A=0
        B=1
        Alpha=&RemapRange(
         Input Value To Remap=&VectorLength(Vector=&Multiply_Vector(A=$Engine.Owner.Velocity, B=<1 1 1>))
         Input Min=0
         Input Max=200
         Output Min=0
         Output Max=1
        )
       )
      )
     ))
   )
   &ScaleColor(
    Scale Alpha=&Multiply_Float(
     A=$System.TODScalar
     B=&Multiply_Float(
      A=&FloatFromCurve(
       FloatCurve=DataInterface() {propertyValues {Curve {
          defaultValue=1
          keys [
           {arriveTangent=4.0437, interpMode=RCIM_Cubic, leaveTangent=4.0437, tangentMode=RCTM_User, time=0, value=0.1}
           {arriveTangent=-0.8452, interpMode=RCIM_Cubic, leaveTangent=-0.8452, tangentMode=RCTM_User, time=0.1418, value=0.9847}
           {arriveTangent=-0.8544, interpMode=RCIM_Cubic, leaveTangent=-0.8544, tangentMode=RCTM_User, time=1, value=0}
          ]
         }}}
       Scale Curve=2
      )
      B=&RemapRange(
       Input Value To Remap=&DistanceBetweenPositions(A=$Engine.Emitter.SimulationPosition, B=$Particles.Position)
       Input Min=0
       Input Max=300
       Output Min=1
       Output Max=0
      )
     )
    )
    Scale RGB=&MakeVectorFromLinearColorRGB(Color=$System.TODColor)
    ScaleA=true
    ScaleRGB=true
   ) {module=ScaleColor001}
   &ScaleSpriteSize(Uniform Curve Sprite Scale=DataInterface() {propertyValues {Curve {
       defaultValue=1
       keys [
        {arriveTangent=-0.2039, interpMode=RCIM_Cubic, leaveTangent=-0.2039, tangentMode=RCTM_User, time=0.0017, value=0.5}
        {interpMode=RCIM_Cubic, tangentMode=RCTM_User, time=1, value=1}
       ]
      }}})
   &Drag(Drag=25, Use Linear Drag=true, Use Rotational Drag=true)
   &SolveForcesAndVelocity(Clamp Velocity=false, Limit Acceleration=false)
   &ScaleColor(
    Scale Alpha=&Multiply_Float(
     A=&Lerp_Float(A=0.5, B=1.3, Alpha=$System.AudioReactivity)
     B=&FN_FrontEndFloatModifier(Gameplay=&Lerp_Float(A=1, B=0.4, Alpha=$System.Timeline.NormalizedTime), Front End=0.8)
    )
    ScaleA=true
    ScaleRGB=true
   )
   &KillParticles(Kill Particles=&SetBoolByFloatComparison(
     A=&MakeFloatFromLinearColor(LinearColor=$Particles.Color, Channel:ENiagaraFloatFromLinearColorOptions=::NewEnumerator3)
     B=0
     Comparison Type=::NewEnumerator7
    ))
  ]
  rendererClasses [&NiagaraSpriteRendererProperties]
  renderers [{rendererClass=&NiagaraSpriteRendererProperties}]
 }
 Stars_Packed_01 {
  emitterUpdateScript [
   &EmitterState(
    Inactive Response=::NewEnumerator1
    Life Cycle Mode=::NewEnumerator1
    Loop Behavior=::NewEnumerator1
    Loop Duration Mode=::NewEnumerator1
    UseLoopDelay=false
   )
   &SpawnRate(
    Spawn Probability=&Lerp_Float(A=0, B=0.7, Alpha=$System.AudioReactivity)
    SpawnRate=20
    Use Spawn Probability=true
   )
  ]
  particleSpawnScript [
   /SetVariables(Particles.RandomInt:NiagaraInt32=&RandomRangeInt(Minimum=0, Maximum=7)) {isSetParameters=true, module=SetVariables_#6}
   &PartitionParticles(Number Of Partitions=8, Particles Per Partition=4, Partition Mode:ENiagara_PartitionMode_v2=::NewEnumerator1)
   &InitializeParticle(
    AdjustAlpha=false
    AdjustHue=true
    AdjustSaturation=true
    AdjustValue=true
    Color=<rgba 5 4.0303 0.9457 1>
    Color Mode=::NewEnumerator3
    Hue Shift Range:Vector2f=<-0.05 0.05>
    Lifetime Max=1.2
    Lifetime Mode=::NewEnumerator1
    Sprite Size Mode=::NewEnumerator3
    Sprite UV Mode=::NewEnumerator3
    Uniform Sprite Size=&Multiply_Float(
     A=&RandomRangeFloat(Minimum=15, Maximum=20)
     B=&SelectFloatFromArray(Direct Array Index=$Engine.QualityLevel, Float Selection Array=DataInterface() {propertyValues {FloatData [0.5,0.6,0.8,1,1]}})
    )
    UsePositionOffset=false
    Write Material Random=true
   )
   &CameraOffset(Camera Offset Amount=5)
   &AddVelocity(Velocity=&FN_FrontEndVector3Modifier(Gameplay=<0 -20 30>, Front End=<0 0 30>))
  ]
  particleUpdateScript [
   &ParticleState()
   &SkeletalMeshLocation(
    Bone / Socket Index=&Add_Integer(A=&Multiply_Int(A=$Particles.RandomInt, B=4), B=$Particles.PartitionParticles.Index In Partition)
    Bone Sampling Mode=::NewEnumerator4
    MaskSpawnGroup=false
    Position Sampling=::NewEnumerator1
    Sampled Position Offset=&MergeHelper_VectorSwitch(
     0=&MergeHelper_VectorSwitch(
      0=&SelectVectorFromArray(
       Array Sampling Mode=::NewEnumerator0
       Direct Array Index=?
       Vector Selection Array=DataInterface() {propertyValues {FloatData [
          <12.1043 -2.1186 -4.8536>
          <7.7075 -1.2879 -6.6743>
          <10.374 -1.9803 -4.7231>
          <4.5107 -1.3436 -4.0531>
          <1.0578 -1.1923 -2.4149>
          <8.5727 -2.8564 -0.6967>
          <8.2147 -2.5855 0.117>
          <1.9331 -0.9532 3.66>
          <11.8682 -1.422 6.0568>
          <-2.1783 -0.0616 1.7825>
          <-6.0713 3.4174 0.6358>
         ]}}
       Random Seed=0
      )
      1=&SelectVectorFromArray(
       Array Sampling Mode=::NewEnumerator0
       Direct Array Index=?
       Vector Selection Array=DataInterface() {propertyValues {FloatData [
          <13.1997 -2.1349 -8.4107>
          <12.5784 -2.9523 -4.639>
          <4.9187 -2.1511 -7.9891>
          <3.7365 -1.9257 -8.3032>
          <-0.1422 -2.6159 -4.3811>
          <7.5273 -3.3673 -3.519>
          <-0.6594 -3.258 -3.1036>
          <3.7533 -3.7299 -2.7185>
          <2.2406 -4.1329 -1.2617>
          <1.7552 -3.3102 1.5707>
          <6.1943 -4.188 -0.2337>
          <7.97 -3.1401 2.5243>
          <-1.8418 -2.9285 4.3853>
          <-6.0999 -2.4904 2.6225>
          <4.3014 -2.8786 5.4921>
          <-8.2397 -2.5249 4.0168>
          <1.5502 -1.4877 7.4717>
          <6.9471 -1.4404 7.983>
          <10.9996 -2.1045 7.6277>
         ]}}
       Random Seed=0
      )
      2=&SelectVectorFromArray(
       Array Sampling Mode=::NewEnumerator0
       Direct Array Index=?
       Vector Selection Array=DataInterface() {propertyValues {FloatData [
          <13.5835 -2.7239 -8.1627>
          <8.5022 -1.9974 -9.4686>
          <3.1847 -2.132 -8.996>
          <11.0229 -2.9949 -6.5318>
          <0.9577 -2.9768 -6.5084>
          <-2.7889 -3.2198 -4.2949>
          <-6.0514 -4.0949 -1.9873>
          <-3.366 -4.3479 -1.0844>
          <-0.8601 -3.9585 -2.4468>
          <-0.5962 -3.6058 4.5014>
          <-5.158 -2.0363 8.2433>
          <1.338 -3.8928 5.4209>
          <7.1802 -4.4771 -0.0398>
          <9.7493 -3.4028 -4.203>
          <14.3388 -4.1935 -1.8803>
          <13.3693 -4.3553 0.5492>
          <11.3134 -3.3555 3.7878>
          <9.5748 -3.7437 4.7187>
          <11.6098 -3.6144 7.8542>
          <13.7223 -2.7235 8.9223>
          <5.4559 -2.6173 8.5364>
          <2.4703 -1.8967 9.0377>
         ]}}
       Random Seed=0
      )
      3=&SelectVectorFromArray(
       Array Sampling Mode=::NewEnumerator0
       Direct Array Index=?
       Vector Selection Array=DataInterface() {propertyValues {FloatData [
          <-4.88 0.044 -10.9935>
          <3.2899 0.23 -11.2833>
          <6.2185 -0.1043 -11.2077>
          <4.2649 -1.8626 -9.4559>
          <-6.4952 -2.3322 -8.9241>
          <-7.5242 -3.1069 -4.9835>
          <-5.8895 -4.2306 -1.6552>
          <-0.9618 -4.4573 -0.4779>
          <-2.7323 -4.2396 0.9032>
          <0.2021 -4.0856 -2.354>
          <3.0944 -4.2641 -1.359>
          <-5.0311 -2.7573 9.0379>
          <-2.3219 -3.0952 8.8189>
          <2.3484 -2.3081 9.9155>
          <10.0039 -1.1324 11.1644>
          <6.0605 -3.4446 8.9139>
          <9.9069 -3.2898 9.3914>
          <5.8876 -3.9364 8.2355>
          <12.495 -4.3043 6.8522>
          <5.2545 -3.7289 2.7196>
          <4.0198 -4.215 0.9987>
          <6.1638 -3.5996 -3.7839>
          <8.1976 -4.2307 0.7947>
          <8.8474 -4.3606 -0.5746>
          <8.3734 -3.537 -3.8766>
          <8.5305 -3.9092 -2.9133>
          <8.3126 -2.6832 -6.3893>
          <9.8983 -2.6578 -6.0268>
          <11.5319 -3.3281 -4.2075>
          <11.684 -2.7488 -5.3889>
          <12.2876 -2.4244 -8.2869>
          <10.9829 -2.2151 -8.791>
          <11.614 -0.8961 -10.9645>
          <13.4004 -4.2027 0.6647>
          <11.0466 -3.3301 3.7146>
         ]}}
       Random Seed=0
      )
      Index=$Particles.PartitionParticles.Index In Partition
      Num=::NewEnumerator2
     )
     1=&MergeHelper_VectorSwitch(
      0=&SelectVectorFromArray(
       Array Sampling Mode=::NewEnumerator0
       Direct Array Index=?
       Vector Selection Array=DataInterface() {propertyValues {FloatData [
          <11.7658 -3.598 -0.4081>
          <7.8147 -3.5269 0.6222>
          <-4.3956 1.1866 4.6141>
          <0.4923 -0.6825 5.6672>
          <1.0913 -2.4977 2.0328>
          <6.569 -4.062 2.3387>
          <5.2682 -2.7674 3.5851>
          <11.2378 -3.8772 4.1832>
          <10.65 -3.0196 4.9513>
          <11.62 -2.891 5.6084>
          <1.1841 -1.0771 4.8551>
          <6.9081 -2.2376 5.9513>
          <9.7535 -1.3436 -6.0433>
          <11.6305 -2.808 -5.1278>
          <6.7995 -2.9427 -4.8478>
          <5.6343 -2.6034 -4.8821>
          <-5.4172 2.2125 -1.7372>
          <-4.8795 0.6741 -2.6068>
          <-6.2558 3.5404 -3.0367>
          <-6.0454 2.7868 -3.6444>
         ]}}
       Random Seed=0
      )
      1=&SelectVectorFromArray(
       Array Sampling Mode=::NewEnumerator0
       Direct Array Index=?
       Vector Selection Array=DataInterface() {propertyValues {FloatData [
          <13.9398 -5.3035 0.3745>
          <9.8541 -6.0256 1.6295>
          <4.9227 -5.5203 1.1728>
          <-3.4644 -4.1535 -0.7604>
          <-3.4188 -5.8122 3.0433>
          <2.0445 -6.2985 3.1477>
          <6.5981 -6.8302 3.5926>
          <10.7185 -6.613 2.6297>
          <6.7068 -4.04 7.9918>
          <1.6413 -4.1334 6.3045>
          <-7.9982 -3.107 5.9216>
          <9.6113 -3.07 10.2768>
          <13.4732 -1.7432 -6.7286>
          <8.5275 -2.3305 -5.7308>
          <6.1477 -4.4459 -5.0881>
          <1.8096 -3.3875 -5.2337>
          <2.2904 -4.2631 -4.9943>
          <-1.9697 -4.166 -4.8857>
          <-4.4251 -3.7269 -4.9653>
          <-6.386 -3.6652 -4.9337>
          <-0.5166 -4.8484 -4.6476>
          <-7.1657 -5.2052 -3.3707>
          <-1.2893 -5.6232 -3.2452>
          <-6.9696 -3.9195 -1.6638>
          <11.6427 -4.9797 -0.8203>
         ]}}
       Random Seed=0
      )
      2=&SelectVectorFromArray(
       Array Sampling Mode=::NewEnumerator0
       Direct Array Index=?
       Vector Selection Array=DataInterface() {propertyValues {FloatData [
          <11.641 -5.6728 1.0917>
          <10.5424 -6.6954 2.5735>
          <5.9483 -6.3972 2.0113>
          <-2.0471 -5.1268 -0.196>
          <-5.6521 -5.8092 1.1555>
          <-6.5692 -7.1075 3.3024>
          <-3.0851 -5.5288 6.7308>
          <-0.7217 -7.3641 3.6122>
          <2.422 -7.6414 5.0376>
          <-1.2829 -5.6535 6.772>
          <6.1029 -5.8336 7.1393>
          <2.1951 -6.328 6.626>
          <3.3367 -4.762 7.9468>
          <-0.1721 -5.0425 7.3459>
          <-3.6088 -4.2428 8.875>
          <11.8749 -3.2069 11.8805>
          <13.6865 -1.612 -7.9008>
          <10.075 -1.6929 -7.7463>
          <0.2315 -1.9831 -7.8364>
          <2.7036 -1.5583 -6.7586>
          <-2.0812 -2.5538 -5.9072>
          <-5.7674 -5.6626 -1.5329>
          <-0.3899 -3.1092 -5.8358>
          <-4.3939 -6.0523 -4.2844>
          <-4.9464 -6.5225 -3.3622>
          <-0.9313 -4.3307 -5.5661>
          <-2.7343 -5.8549 -4.5366>
          <0.6389 -6.6813 -3.7188>
          <1.8613 -4.0525 -5.7092>
          <5.0787 -3.5626 -5.9212>
          <6.5801 -3.2552 -6.0342>
          <11.5875 -5.9147 -4.9222>
          <9.7935 -5.4003 -5.361>
          <13.8616 -7.1924 -3.2854>
          <12.5972 -6.1818 -0.8214>
         ]}}
       Random Seed=0
      )
      3=&SelectVectorFromArray(
       Array Sampling Mode=::NewEnumerator0
       Direct Array Index=?
       Vector Selection Array=DataInterface() {propertyValues {FloatData [
          <13.5131 -5.7583 1.4295>
          <7.2605 -5.8622 1.677>
          <4.4776 -6.6558 2.8925>
          <1.6866 -5.6577 0.7376>
          <-1.6198 -6.3878 2.3964>
          <-6.577 -8.1401 5.6945>
          <-1.4871 -8.2824 5.373>
          <3.6376 -8.1844 5.1915>
          <-0.489 -4.3727 10.3232>
          <2.1494 -3.994 11.5083>
          <-3.2046 -1.9845 13.0197>
          <4.3737 -2.2531 13.0521>
          <-6.2298 -2.0545 -6.5724>
          <-4.4524 -2.8303 -6.4319>
          <-3.4983 -5.7747 -5.1832>
          <-2.9351 -7.4859 -1.9071>
          <2.1201 -6.8309 -0.6298>
          <3.2178 -7.3425 -1.0302>
          <5.1423 -7.3351 -0.8961>
          <2.4541 -7.6752 -1.9195>
          <3.5209 -7.0318 -4.2472>
          <-2.3718 -5.7237 -5.2268>
          <-2.4962 -2.833 -6.455>
          <3.5565 -5.7296 -5.2477>
          <5.3577 -3.985 -6.2203>
          <4.3764 -6.5019 -4.7139>
          <5.6152 -7.7699 -2.2721>
          <8.8316 -5.7238 -5.2323>
          <12.1859 -7.0488 -4.3676>
          <12.0661 -1.9439 -6.9926>
          <9.1538 -1.1743 -8.0301>
          <9.969 -8.4099 7.0004>
          <11.1431 -3.3415 12.6745>
         ]}}
       Random Seed=0
      )
      Index=$Particles.PartitionParticles.Index In Partition
      Num=::NewEnumerator2
     )
     2=&MergeHelper_VectorSwitch(
      0=&SelectVectorFromArray(
       Array Sampling Mode=::NewEnumerator0
       Direct Array Index=?
       Vector Selection Array=DataInterface() {propertyValues {FloatData [
          <-5.9399 2.4876 -2.483>
          <1.3927 -0.2699 -4.8192>
          <4.3234 -0.5609 -5.3699>
          <11.8038 -1.7046 -6.1407>
          <5.2786 -3.6684 -0.8376>
          <9.4459 -2.1042 -4.4151>
          <1.4471 -0.7614 -3.7492>
          <6.9768 -2.4766 -2.8133>
          <3.5759 -1.5034 -3.2352>
          <4.1047 -2.8848 -1.8721>
          <0.8201 -2.4111 -0.4422>
          <9.07 -3.5758 0.5087>
          <0.4033 -1.8452 0.1267>
          <4.5094 -2.0277 2.2354>
          <11.4969 -2.9849 4.3867>
          <7.3333 -2.5212 4.523>
          <-0.2559 -0.7493 4.1809>
          <0.4402 -0.617 4.8044>
          <-4.5606 1.1089 2.7399>
          <-5.0443 1.9142 3.2012>
         ]}}
       Random Seed=0
      )
      1=&SelectVectorFromArray(
       Array Sampling Mode=::NewEnumerator0
       Direct Array Index=?
       Vector Selection Array=DataInterface() {propertyValues {FloatData [
          <3.3646 -1.0532 -9.1352>
          <4.8039 -1.3602 -9.0435>
          <0.4326 -2.7579 -6.3535>
          <3.7452 -3.1468 -6.0234>
          <6.3564 -3.65 -5.1381>
          <13.4828 -2.9968 -7.8299>
          <9.8335 -4.3439 -3.7535>
          <7.6125 -4.6599 -3.2974>
          <3.6911 -4.2338 -3.5495>
          <-6.7388 -4.219 0.5388>
          <10.0182 -3.8104 1.4943>
          <-9.6163 -3.0081 2.3103>
          <-1.2823 -3.543 2.8572>
          <0.4469 -3.5933 2.7843>
          <7.321 -3.5694 1.7831>
          <9.5626 -3.4941 4.3056>
          <6.9859 -3.5047 4.477>
          <3.016 -3.6311 3.7282>
          <3.3806 -2.6482 6.805>
          <-3.3884 -3.4047 4.3438>
          <-1.416 -2.9498 5.9555>
          <-6.1985 -3.2786 3.8874>
          <-2.4302 -2.597 6.4563>
          <-7.092 -2.9147 5.2939>
         ]}}
       Random Seed=0
      )
      2=&SelectVectorFromArray(
       Array Sampling Mode=::NewEnumerator0
       Direct Array Index=?
       Vector Selection Array=DataInterface() {propertyValues {FloatData [
          <-7.456 -1.0231 -10.5241>
          <-4.2 -3.2983 -7.7055>
          <-3.2176 -3.0004 -8.6432>
          <5.2729 -1.6967 -11.078>
          <8.956 -2.3102 -10.9651>
          <8.9057 -3.5169 -8.8347>
          <13.1999 -3.9403 -8.4831>
          <5.716 -4.1443 -5.2116>
          <3.0727 -3.9951 -6.009>
          <-0.1701 -3.8589 -6.6687>
          <0.1967 -5.105 -3.2144>
          <-7.5075 -4.4639 -3.7969>
          <-3.7179 -5.2313 -2.9375>
          <7.6246 -5.9122 -1.5863>
          <2.8896 -5.8026 0.9212>
          <7.6531 -4.5692 2.1964>
          <9.9763 -3.331 3.1678>
          <8.6903 -3.0759 3.759>
          <14.2663 -5.8825 -1.7514>
          <1.0856 -3.2246 3.2817>
          <-3.4101 -3.0515 4.4458>
          <-1.6187 -2.9622 4.4533>
          <0.5373 -2.2327 6.4225>
          <-2.0771 -2.1671 6.9585>
          <8.7728 -2.3072 5.5475>
          <6.9245 -1.8524 6.895>
          <8.4521 -1.8456 6.7797>
          <11.6126 -2.228 5.6094>
          <11.8803 -2.7994 4.4601>
          <11.0175 -1.6334 7.1706>
         ]}}
       Random Seed=0
      )
      3=&SelectVectorFromArray(
       Array Sampling Mode=::NewEnumerator0
       Direct Array Index=?
       Vector Selection Array=DataInterface() {propertyValues {FloatData [
          <-5.429 1.36 -13.8215>
          <-5.9982 -3.0291 -10.7798>
          <-6.2377 -4.791 2.4662>
          <-5.3736 -5.856 1.5637>
          <-4.3847 -4.2206 -7.6151>
          <-4.5102 -6.003 -1.3919>
          <-3.8091 -6.1455 0.1757>
          <2.8731 -4.323 -5.5709>
          <-1.7348 -5.9561 -1.6137>
          <2.8616 -5.9561 -1.7383>
          <3.5284 -5.9398 1.1858>
          <5.1226 -5.9824 0.9031>
          <7.932 -3.5378 3.2536>
          <9.8194 -5.5166 2.223>
          <3.2553 -4.3222 -7.4917>
          <7.0369 -4.3586 -7.1591>
          <6.0003 -3.1769 -11.6147>
          <8.3578 -4.3891 -8.9841>
          <-2.8972 -2.8144 -11.5036>
          <-0.9827 -1.5549 -12.628>
          <4.4991 -2.6082 -12.4733>
          <0.8117 1.7041 -14.419>
          <12.3988 -5.0475 2.9552>
          <10.7308 -5.2195 2.6859>
          <11.7094 -6.1045 -0.1857>
          <9.3034 -5.7891 -2.2668>
          <13.2722 -5.2864 -3.6934>
          <13.0805 -4.887 -4.8299>
          <12.7141 -4.628 -5.5518>
          <11.4309 -4.3931 -7.924>
          <12.8578 -3.9559 -10.7665>
          <12.1073 -3.3234 -11.9811>
          <10.0903 -0.5493 -14.1518>
          <10.0528 1.7838 -15.016>
          <-6.9942 -1.5206 7.1244>
         ]}}
       Random Seed=0
      )
      Index=$Particles.PartitionParticles.Index In Partition
      Num=::NewEnumerator2
     )
     3=&MergeHelper_VectorSwitch(
      0=&SelectVectorFromArray(
       Array Sampling Mode=::NewEnumerator0
       Direct Array Index=?
       Vector Selection Array=DataInterface() {propertyValues {FloatData [
          <-4.3516 1.8791 -2.5928>
          <0.2773 -0.1642 5.7218>
          <1.8137 -0.4164 5.672>
          <3.8953 -1.5305 -0.06>
          <1.1507 -0.0262 -3.7941>
          <3.0691 -0.196 -4.139>
          <2.3652 0.1254 -5.3059>
          <4.8556 -0.5573 -3.6843>
          <5.5916 -1.1 -2.3591>
          <5.7724 -0.3941 -4.4004>
          <7.686 -0.5557 -6.069>
          <8.2366 -0.6333 -4.2605>
          <12.2368 -0.5516 -5.8611>
          <12.9611 -1.4337 5.1475>
          <4.9374 -0.7863 4.5167>
          <8.4264 -1.9569 2.2314>
          <6.0278 -1.7344 -0.024>
         ]}}
       Random Seed=0
      )
      1=&SelectVectorFromArray(
       Array Sampling Mode=::NewEnumerator0
       Direct Array Index=?
       Vector Selection Array=DataInterface() {propertyValues {FloatData [
          <12.8365 -0.6775 -6.2608>
          <10.4235 -2.1158 7.7189>
          <11.5574 -3.0518 2.9138>
          <3.2858 -1.6914 9.5122>
          <4.1973 -1.7773 -3.3614>
          <-0.1739 -0.8037 -5.2139>
          <7.0392 -0.953 -6.4499>
          <12.936 -2.4902 -1.4546>
          <11.9473 -2.769 0.5989>
          <9.2288 -2.7886 -0.2875>
          <9.8262 -2.9326 2.1632>
          <2.0307 -1.928 5.6448>
          <-3.3671 -2.319 3.9217>
          <-1.4211 -2.7778 2.5841>
         ]}}
       Random Seed=0
      )
      2=&SelectVectorFromArray(
       Array Sampling Mode=::NewEnumerator0
       Direct Array Index=?
       Vector Selection Array=DataInterface() {propertyValues {FloatData [
          <-4.4711 -0.6173 -6.3131>
          <6.623 -1.2176 -7.0585>
          <11.743 -0.814 -6.7735>
          <4.2162 -0.3367 -5.3431>
          <10.8506 -1.6587 -2.4935>
          <9.0163 -1.6349 -2.6798>
          <0.6745 -1.7768 -2.8461>
          <7.6939 -2.5255 0.3525>
          <-0.6194 -2.6511 -0.4474>
          <-0.337 -3.2806 3.7298>
          <1.8761 -3.7015 4.998>
          <7.877 -2.7359 2.6185>
          <11.8046 -3.5613 6.071>
          <6.819 -2.2003 11.0462>
          <4.0493 -2.3355 8.9503>
          <-3.3545 -2.247 9.0877>
          <-4.1908 -2.2673 7.664>
          <-5.8859 -3.043 2.9149>
          <-8.0519 -2.3843 -1.6861>
         ]}}
       Random Seed=0
      )
      3=&SelectVectorFromArray(
       Array Sampling Mode=::NewEnumerator0
       Direct Array Index=?
       Vector Selection Array=DataInterface() {propertyValues {FloatData [
          <6.1685 0.3696 -6.2975>
          <-4.8639 -1.8373 -7.3626>
          <-2.0174 0.5331 -5.8391>
          <-3.1516 -0.5679 -4.1411>
          <-6.3018 -1.9311 -1.6156>
          <-6.5804 -2.3277 -0.403>
          <-6.8947 -3.7939 5.787>
          <-1.3032 -2.6788 3.5167>
          <-5.4932 -3.194 6.8793>
          <1.0961 -2.22 2.2465>
          <5.9961 -1.3285 -1.8281>
          <-0.7535 -3.7933 6.3305>
          <-6.039 -2.4903 7.7988>
          <1.4174 -2.0497 11.9384>
          <2.3615 -2.1639 10.0196>
          <12.9838 -1.8908 12.8261>
          <11.5712 -1.4234 -0.1103>
          <7.8089 -0.7844 -2.9083>
          <7.2847 -2.9679 5.2592>
          <7.6955 1.0736 -5.8073>
          <9.7901 0.0046 -6.5376>
          <12.407 -1.8103 3.5479>
          <9.2367 -2.6552 4.8918>
          <8.7347 -1.9103 9.029>
          <9.038 -1.89 9.3182>
          <11.4683 -1.813 9.8881>
          <12.4654 -1.8208 10.6203>
         ]}}
       Random Seed=0
      )
      Index=$Particles.PartitionParticles.Index In Partition
      Num=::NewEnumerator2
     )
     4=&MergeHelper_VectorSwitch(
      0=&SelectVectorFromArray(
       Array Sampling Mode=::NewEnumerator0
       Direct Array Index=?
       Vector Selection Array=DataInterface() {propertyValues {FloatData [
          <-3.6277 0.9295 -4.4257>
          <3.9914 -1.4836 -4.8789>
          <-1.7847 -0.9704 1.8692>
          <1.8967 -1.5917 3.09>
          <0.7674 -1.9866 -0.3333>
          <2.4262 -1.8776 2.4124>
          <1.3955 -2.3538 -2.6472>
          <2.5033 -2.6229 -0.6001>
          <6.5165 -2.7255 -3.9179>
          <8.7002 -3.0475 -0.5193>
          <11.975 -2.0392 -6.6812>
          <15.247 -4.4234 -2.9625>
          <12.8299 -2.2913 1.3619>
          <13.8333 -2.1719 5.3593>
          <10.0696 -1.6105 5.605>
         ]}}
       Random Seed=0
      )
      1=&SelectVectorFromArray(
       Array Sampling Mode=::NewEnumerator0
       Direct Array Index=?
       Vector Selection Array=DataInterface() {propertyValues {FloatData [
          <-4.3231 -2.7616 -0.3465>
          <-1.9158 -2.6951 -6.9306>
          <5.0894 -2.6172 -8.8124>
          <9.2933 -2.8244 -8.6546>
          <-0.8846 -4.5159 -4.6549>
          <9.8747 -5.3669 -5.0117>
          <7.2073 -5.3458 -3.9482>
          <-3.9114 -4.6073 -2.1205>
          <-0.819 -2.4114 -0.325>
          <11.0058 -1.3652 0.1338>
          <3.1952 -2.1787 2.4905>
          <-4.9162 -1.3256 6.8921>
          <6.0971 -2.4531 5.9588>
          <7.8905 -2.2874 3.966>
          <11.6165 -1.7322 2.1684>
          <13.2292 -1.7297 2.3598>
          <11.3897 -2.3399 4.7692>
          <9.786 -5.3439 -2.6604>
         ]}}
       Random Seed=0
      )
      2=&SelectVectorFromArray(
       Array Sampling Mode=::NewEnumerator0
       Direct Array Index=?
       Vector Selection Array=DataInterface() {propertyValues {FloatData [
          <-7.0278 -2.649 -9.9147>
          <-3.0515 -3.6446 -8.0444>
          <1.4428 -3.9307 -8.1461>
          <-3.8762 -5.0429 -6.6475>
          <-8.2772 -2.0958 8.0121>
          <-2.3431 -1.6196 9.0105>
          <1.0744 -2.4083 7.2567>
          <-0.6173 -1.7113 3.4053>
          <4.9128 -2.2958 8.7544>
          <3.1143 -1.9833 4.911>
          <-0.0756 -2.4849 -0.2038>
          <-0.8642 -4.9078 -1.4021>
          <-1.6859 -5.6979 -2.4735>
          <1.8346 -6.1346 -5.1335>
          <5.7576 -5.086 -1.4346>
          <11.2551 -3.028 -9.9363>
          <12.724 -3.6924 -9.032>
          <9.1609 -6.0913 -6.9122>
          <6.6065 -6.2863 -5.5805>
          <8.7485 -6.1861 -2.9457>
          <9.6759 -4.8302 -1.1682>
          <14.3157 -5.0739 -1.3502>
          <12.4968 -3.4863 -0.1716>
          <8.4883 -0.9547 1.8739>
          <12.4743 -0.9182 1.2302>
          <11.8195 -0.8785 2.2643>
          <11.2928 -1.1099 3.3139>
          <9.0939 -1.3232 3.6826>
          <12.3074 -1.3442 4.1329>
          <12.976 -1.8 5.6218>
          <7.6234 -2.3266 7.8405>
          <8.0079 -2.1671 9.6077>
          <5.2905 -1.3672 10.053>
          <8.1096 -1.1093 10.5365>
         ]}}
       Random Seed=0
      )
      3=&SelectVectorFromArray(
       Array Sampling Mode=::NewEnumerator0
       Direct Array Index=?
       Vector Selection Array=DataInterface() {propertyValues {FloatData [
          <-3.3902 -2.9333 -10.292>
          <-0.2034 -3.0304 -9.8568>
          <-6.3513 -6.2862 -2.8768>
          <-1.3116 -6.4448 -3.0494>
          <3.7604 -6.6673 -7.5697>
          <3.8266 -6.5441 -3.0932>
          <5.5951 -6.5434 -3.0294>
          <7.573 -6.575 -3.1117>
          <3.604 -5.1216 -1.0702>
          <4.3402 -2.3993 0.2925>
          <-2.35 -1.1909 4.29>
          <-6.9226 -1.943 6.4486>
          <-4.1577 -1.8481 10.6837>
          <1.2653 -0.7519 1.7511>
          <1.9979 -2.2531 9.1552>
          <5.7708 -1.7586 6.7871>
          <7.1062 -1.2791 5.4568>
          <6.447 -1.9008 7.7961>
          <-0.5993 -0.5178 12.0842>
          <11.6004 -1.6641 7.2494>
          <9.5539 -1.9651 8.8494>
          <7.6521 -1.4708 6.0838>
          <8.862 -1.8939 10.7816>
          <10.3346 -1.6254 11.7312>
          <7.8893 -1.6928 11.4459>
          <11.2463 -1.3228 12.5201>
          <10.5884 -0.0459 13.2321>
          <12.4944 -0.1564 13.4988>
          <10.9103 -0.5782 2.2549>
          <9.4402 -0.6447 2.9111>
          <8.8664 -1.104 0.6854>
          <9.4591 -3.5294 0.1145>
          <11.0986 -3.167 0.2352>
          <12.5312 -6.1735 -2.1822>
          <13.4159 -6.5529 -2.839>
          <7.9122 -7.0058 -5.6185>
          <8.4176 -5.909 -8.8174>
         ]}}
       Random Seed=0
      )
      Index=$Particles.PartitionParticles.Index In Partition
      Num=::NewEnumerator2
     )
     5=&MergeHelper_VectorSwitch(
      0=&SelectVectorFromArray(
       Array Sampling Mode=::NewEnumerator0
       Direct Array Index=?
       Vector Selection Array=DataInterface() {propertyValues {FloatData [
          <-5.1578 1.9054 -0.8273>
          <-3.9725 1.098 -1.5476>
          <-5.2462 1.5696 -0.4986>
          <-1.5438 -1.0028 1.1676>
          <-0.0204 -0.9908 1.6983>
          <0.4521 -1.3515 1.5526>
          <3.0397 -1.1901 2.3487>
          <-0.2117 -1.5702 0.1214>
          <-0.5793 -1.3477 -0.5654>
          <5.108 -1.1633 -1.288>
          <10.3958 -0.8387 -2.9407>
          <14.6994 -1.276 -1.7853>
          <8.6863 -3.0108 0.5665>
          <8.6114 -2.2845 2.4254>
          <12.8652 -1.2243 3.7266>
          <9.6086 -1.047 5.1863>
          <0.274 0.0091 4.8907>
          <12.5257 -2.6842 2.5819>
          <14.993 -2.6501 -0.483>
          <13.1035 -2.3405 -5.3676>
          <-3.856 1.4549 -3.5103>
         ]}}
       Random Seed=0
      )
      1=&SelectVectorFromArray(
       Array Sampling Mode=::NewEnumerator0
       Direct Array Index=?
       Vector Selection Array=DataInterface() {propertyValues {FloatData [
          <-1.7055 -1.1032 -2.859>
          <-0.1354 -0.8444 7.6557>
          <4.9793 -0.9123 7.8254>
          <13.0217 -1.0836 7.8683>
          <6.0195 -1.2375 5.4887>
          <3.285 -1.2344 6.2281>
          <4.5033 -1.5303 4.4788>
          <-3.8577 -3.1258 -0.2289>
          <-2.8189 -3.4562 0.0758>
          <11.6981 -2.5241 -1.2571>
          <12.5786 -5.1098 1.8912>
          <6.3832 -4.741 2.065>
          <10.2963 -3.0417 -5.9101>
          <5.8017 -2.7921 -5.5189>
          <-2.1148 -2.373 -4.7559>
         ]}}
       Random Seed=0
      )
      2=&SelectVectorFromArray(
       Array Sampling Mode=::NewEnumerator0
       Direct Array Index=?
       Vector Selection Array=DataInterface() {propertyValues {FloatData [
          <4.5869 -3.0303 -5.6283>
          <5.2636 -2.1312 -5.005>
          <-1.5778 -2.5643 -4.8688>
          <-3.1741 -1.4215 -4.5869>
          <6.8693 -0.8555 9.5713>
          <-7.2875 -0.8143 9.1467>
          <-3.6473 -1.2227 7.277>
          <-1.2155 -1.258 6.4993>
          <6.089 -1.5621 5.7697>
          <-1.8887 -1.7562 5.2235>
          <1.159 -4.8082 3.2881>
          <-2.6427 -4.3475 3.3985>
          <-5.6077 -2.4967 -1.298>
          <3.8034 -0.7952 -2.4906>
          <12.3866 -0.8882 -2.2261>
          <11.4616 -1.1098 6.4045>
          <6.8794 -2.8926 4.8565>
          <3.108 -5.126 3.1574>
          <9.9145 -4.7903 3.6518>
          <12.4826 -5.1769 3.4837>
          <12.464 -3.0153 -0.9904>
          <7.7653 -4.836 0.1871>
          <6.1273 -3.6699 -0.6122>
          <2.1146 -4.3199 -0.1632>
          <2.7544 -5.0259 0.5535>
          <3.3708 -3.4367 -0.7663>
          <0.6427 -5.1229 0.779>
          <-0.5874 -3.883 -0.4203>
          <-0.0345 -3.048 -1.0146>
          <-5.8764 -3.9505 -0.2313>
          <-5.4084 -2.2201 -4.6997>
         ]}}
       Random Seed=0
      )
      3=&SelectVectorFromArray(
       Array Sampling Mode=::NewEnumerator0
       Direct Array Index=?
       Vector Selection Array=DataInterface() {propertyValues {FloatData [
          <-4.7393 -0.9887 -5.0391>
          <-2.2928 -2.9104 -5.7796>
          <6.1548 -2.4149 -5.8731>
          <13.1554 -3.4873 -7.8814>
          <-4.7239 -1.3112 -1.9631>
          <12.5562 -1.2677 -2.4636>
          <12.6586 -5.3059 1.6287>
          <10.2859 -5.4994 2.8129>
          <-3.8334 -0.5616 10.9978>
          <-6.0557 -1.06 6.7307>
          <-0.6388 -0.9882 7.9205>
          <-7.2058 -2.0335 5.8789>
          <-0.8547 -1.1292 7.0045>
          <-5.1 -5.1761 3.6475>
          <2.3794 -0.9473 7.507>
          <3.6389 -0.9126 9.0675>
          <5.2491 -1.6964 6.9081>
          <9.3466 -0.3483 11.9851>
          <8.9132 0.446 13.3495>
          <8.7911 -0.7911 8.8764>
          <8.8193 -0.766 10.3132>
          <5.3365 -2.7805 5.8168>
          <4.5911 -4.0237 4.8133>
          <3.3158 -5.1599 3.9234>
          <5.4236 -5.0695 4.0445>
          <5.0902 -5.4027 3.4335>
          <9.9923 -3.8498 5.0754>
          <9.5982 -4.0249 -0.189>
          <10.9539 -4.5343 0.2528>
          <7.6953 -2.7814 -1.1807>
          <4.4857 -3.7829 -0.3758>
          <2.1466 -3.8906 -0.3099>
         ]}}
       Random Seed=0
      )
      Index=$Particles.PartitionParticles.Index In Partition
      Num=::NewEnumerator2
     )
     6=&MergeHelper_VectorSwitch(
      0=&SelectVectorFromArray(
       Array Sampling Mode=::NewEnumerator0
       Direct Array Index=?
       Vector Selection Array=DataInterface() {propertyValues {FloatData [
          <-3.761 1.1597 1.2311>
          <-1.3654 -1.3107 -1.3163>
          <-0.3763 -0.9872 -2.3599>
          <-0.1482 -0.4941 -2.8721>
          <2.3408 -1.271 -3.5664>
          <3.8169 -1.53 -3.3955>
          <6.2615 -2.3869 -4.8684>
          <10.9325 -2.7058 -3.5515>
          <9.8979 -3.0599 -5.3685>
          <8.8812 -2.7944 -3.0965>
          <8.9036 -3.8433 -1.3385>
          <7.2562 -1.2967 1.3498>
          <13.7167 -1.7698 1.5511>
          <6.5194 -0.4218 2.0331>
          <2.3818 -0.764 4.7861>
         ]}}
       Random Seed=0
      )
      1=&SelectVectorFromArray(
       Array Sampling Mode=::NewEnumerator0
       Direct Array Index=?
       Vector Selection Array=DataInterface() {propertyValues {FloatData [
          <-4.0817 -3.9455 -5.5213>
          <4.7583 -4.3306 -4.8289>
          <-0.3231 -5.0707 -3.1413>
          <5.5284 -4.4534 -5.1454>
          <4.9305 -4.4784 -4.028>
          <10.129 -4.5691 -4.9684>
          <12.0582 -5.5838 -3.9385>
          <-0.795 -2.8229 4.2663>
          <1.4047 -2.4699 3.5063>
          <-4.8 -2.3676 1.3135>
          <2.6311 -2.5668 3.5351>
          <2.4359 -3.345 5.3491>
          <8.0303 -3.4407 4.4223>
          <-3.8881 -3.5829 0.4695>
          <-3.7536 -3.9789 0.1879>
          <-0.2705 -5.2651 -1.1072>
          <3.7999 -4.9959 -0.3637>
          <5.1287 -2.5692 1.3638>
          <8.3426 -4.9458 -0.0875>
          <7.1737 -3.0868 0.9744>
          <11.0994 -2.2986 2.2758>
          <11.5646 -2.6167 2.9808>
          <7.656 -2.2563 2.3544>
          <6.7896 -2.2875 2.482>
          <8.0937 -2.6559 3.2572>
          <11.5693 -3.626 4.6252>
          <10.784 -3.5266 6.7521>
          <11.745 -3.4563 6.995>
         ]}}
       Random Seed=0
      )
      2=&SelectVectorFromArray(
       Array Sampling Mode=::NewEnumerator0
       Direct Array Index=?
       Vector Selection Array=DataInterface() {propertyValues {FloatData [
          <-4.303 -4.746 -5.2282>
          <4.6401 -5.2479 -7.2137>
          <11.3108 -5.7185 -5.6214>
          <1.4063 -5.1724 -5.1511>
          <-3.5512 -6.6473 -3.092>
          <4.0352 -6.1819 -1.0315>
          <2.5153 -3.5042 0.4908>
          <11.4815 -4.0918 0.2202>
          <7.1512 -2.5356 2.4442>
          <13.0211 -3.2721 3.7142>
          <-3.312 -2.6476 2.7436>
          <-3.6272 -3.5195 4.1478>
          <-5.7665 -4.0102 6.3304>
          <-4.1357 -3.6787 7.1776>
          <2.1676 -4.2937 6.2719>
          <7.021 -4.3164 5.6858>
          <-0.4263 -3.6142 7.6786>
          <7.7747 -4.4222 6.2278>
          <7.6336 -4.3318 7.2415>
          <10.4536 -4.5485 7.1658>
          <13.2536 -4.3625 5.2677>
          <11.9147 -3.9665 8.0727>
         ]}}
       Random Seed=0
      )
      3=&SelectVectorFromArray(
       Array Sampling Mode=::NewEnumerator0
       Direct Array Index=?
       Vector Selection Array=DataInterface() {propertyValues {FloatData [
          <-6.4844 -5.1564 -7.3043>
          <-2.2142 -5.4149 -6.3664>
          <3.6941 -6.4115 -6.3214>
          <0.6102 -7.5123 -4.7899>
          <-3.0589 -7.5836 -3.3224>
          <-6.3453 -3.4956 0.4339>
          <2.0756 -2.025 1.5667>
          <-3.0149 -3.2221 3.6485>
          <-6.9804 -4.0853 4.8624>
          <9.9745 -6.8582 -6.6178>
          <13.7093 -4.2371 -8.2069>
          <11.0953 -5.6976 -0.3027>
          <12.5871 -3.502 0.4424>
          <7.0921 -6.1476 -0.6477>
          <8.3702 -5.4546 -0.2939>
          <4.5976 -5.4959 -0.3346>
          <9.0181 -3.6333 0.4375>
          <7.6034 -4.0894 0.2985>
          <5.5929 -1.8237 1.9257>
          <2.5691 -3.8099 4.4581>
          <3.503 -4.6068 5.5355>
          <2.1185 -4.1599 8.1777>
          <2.151 -3.0839 9.1954>
         ]}}
       Random Seed=0
      )
      Index=$Particles.PartitionParticles.Index In Partition
      Num=::NewEnumerator2
     )
     7=&MergeHelper_VectorSwitch(
      0=&SelectVectorFromArray(
       Array Sampling Mode=::NewEnumerator0
       Direct Array Index=?
       Vector Selection Array=DataInterface() {propertyValues {FloatData [
          <-6.3138 2.6059 -0.8865>
          <1.4299 -1.4069 -3.0267>
          <-3.6126 1.4961 0.9269>
          <6.2303 -2.4933 -2.9358>
          <9.5919 -1.5987 -4.1948>
          <10.1598 -1.299 -4.8814>
          <8.8789 -3.4234 -1.6053>
          <13.6549 -1.3777 -5.2434>
          <14.6995 -2.3664 -3.8777>
          <14.6282 -1.7444 -4.6417>
          <12.732 -3.7277 -1.9665>
          <15.5552 -4.013 -1.8686>
          <8.3461 -2.6986 0.3931>
          <16.631 -4.1457 -0.1946>
          <10.5013 -1.5172 2.5817>
         ]}}
       Random Seed=0
      )
      1=&SelectVectorFromArray(
       Array Sampling Mode=::NewEnumerator0
       Direct Array Index=?
       Vector Selection Array=DataInterface() {propertyValues {FloatData [
          <-4.5714 -1.4207 -5.5158>
          <0.1081 -2.8225 -3.7909>
          <-4.4136 -4.2122 -0.1365>
          <-2.6189 -3.7005 1.0268>
          <1.7404 -4.8172 -0.1497>
          <1.0669 -4.0281 -2.843>
          <6.6757 -2.2236 -4.923>
          <8.1663 -3.6134 1.8617>
          <9.7084 -4.8378 -2.2276>
          <13.4025 -4.2282 -3.1273>
          <7.3098 -3.2729 5.3722>
         ]}}
       Random Seed=0
      )
      2=&SelectVectorFromArray(
       Array Sampling Mode=::NewEnumerator0
       Direct Array Index=?
       Vector Selection Array=DataInterface() {propertyValues {FloatData [
          <-7.0595 -4.9132 -2.3866>
          <-3.0653 -5.1886 -1.9869>
          <-2.8283 -5.6349 0.4002>
          <0.1439 -5.4999 1.1075>
          <-6.1169 -3.4131 4.3835>
          <1.9564 -3.9152 6.0011>
          <6.9929 -3.5731 3.842>
          <11.1577 -3.7518 4.9511>
          <12.6779 -4.7509 2.6513>
          <8.3524 -6.2053 -0.8529>
          <11.4717 -3.7291 -3.5226>
          <13.1501 -4.1712 -3.2218>
          <1.127 -2.9928 -4.3987>
          <-7.5842 -2.5293 -5.0678>
         ]}}
       Random Seed=0
      )
      3=&SelectVectorFromArray(
       Array Sampling Mode=::NewEnumerator0
       Direct Array Index=?
       Vector Selection Array=DataInterface() {propertyValues {FloatData [
          <-4.7817 -4.3752 3.3004>
          <0.9742 -4.366 3.6931>
          <1.5181 -4.8336 3.268>
          <2.8004 -6.1707 2.158>
          <13.3708 -6.012 3.2787>
          <12.2833 -5.7611 3.3446>
          <10.9562 -7.1959 1.0802>
          <6.6153 -7.0815 0.3367>
          <8.8292 -2.4325 -4.4459>
          <10.8354 -2.3201 -4.6793>
          <-0.3755 -6.8395 -0.9403>
          <-3.4378 -3.9636 -3.3355>
         ]}}
       Random Seed=0
      )
      Index=$Particles.PartitionParticles.Index In Partition
      Num=::NewEnumerator2
     )
     Index=$Particles.RandomInt
     Num=::NewEnumerator6
    )
    Skeletal Mesh=$System.Skeletal Mesh_Dyn
    Source Movement Scale:Vector3f=&Lerp_Vector(
     A=&VectorFromFloat(Value=&MakeCustomFloatFromBool(
       Bool=$User.IsSkydiving
       True Float=&MakeCustomFloatFromBool(
        Bool=&SetBoolByFloatComparison(
         A=&VectorLength(Vector=&Multiply_Vector(A=$Engine.Owner.Velocity, B=<1 1 0>))
         B=50
         Comparison Type=::NewEnumerator6
        )
        True Float=0.7
        False Float=0
       )
       False Float=0
      ))
     B=&VectorFromFloat(Value=&MakeCustomFloatFromBool(Bool=$User.IsSkydiving, True Float=0.9, False Float=0.7))
     Alpha=&RemapRange(
      Input Value To Remap=&VectorLength(Vector=&Multiply_Vector(A=$Engine.Owner.Velocity, B=<1 1 1>))
      Input Min=0
      Input Max=800
      Output Min=0
      Output Max=1
     )
    )
   )
   &ScaleSpriteSize(
    Uniform Curve Scale=&Lerp_Float(A=0.3, B=1, Alpha=$System.AudioReactivity)
    Uniform Curve Sprite Scale=DataInterface() {propertyValues {Curve {keys [
        {interpMode=RCIM_Cubic, tangentMode=RCTM_User, time=0, value=0}
        {interpMode=RCIM_Cubic, tangentMode=RCTM_User, time=0.5, value=1}
        {tangentMode=RCTM_User, time=1, value=0}
       ]}}}
   ) {module=ScaleSpriteSize001}
   &ScaleColor(
    Scale Alpha=&Multiply_Float(
     A=&Multiply_Float(
      A=$System.TODScalar
      B=&Multiply_Float(
       A=&FloatFromCurve(
        FloatCurve=DataInterface() {propertyValues {Curve {keys [
            {arriveTangent=6.5707, interpMode=RCIM_Cubic, leaveTangent=6.5707, tangentMode=RCTM_User, time=0, value=0}
            {interpMode=RCIM_Cubic, tangentMode=RCTM_User, time=0.3, value=1}
            {tangentMode=RCTM_User, time=1, value=0}
           ]}}}
        Scale Curve=&RandomRangeFloat(Minimum=0.1, Maximum=0.5)
       )
       B=1
      )
     )
     B=&Lerp_Float(A=0, B=1, Alpha=$System.AudioReactivity)
    )
    Scale RGB=&Multiply_Vector(
     A=&VectorFromFloat(Value=&Multiply_Float(
       A=&RandomRangeFloat(Minimum=0.75, Maximum=1)
       B=&SelectFloatFromArray(Direct Array Index=$Engine.QualityLevel, Float Selection Array=DataInterface() {propertyValues {FloatData [10,5,1,1,1]}})
      ))
     B=&MakeVectorFromLinearColorRGB(Color=$System.TODColor)
    )
    ScaleA=true
    ScaleRGB=true
   )
   /SetVariables(
    Particles.CenterFlareColor:LinearColor=&MultiplyLinearColor(
     Linear Color=$Particles.Color
     Scale Factor=&MakeVector4FromVectorAndFloat(
      "Vector (RGB)"=&MakeVectorFromLinearColorRGB(Color=<rgba 10 8.5224 3.7296 1>)
      "Float (Alpha)"=&Multiply_Float(
       A=&FloatFromCurve(
        FloatCurve=DataInterface() {propertyValues {Curve {
           defaultValue=1
           keys [
            {arriveTangent=4.0437, interpMode=RCIM_Cubic, leaveTangent=4.0437, tangentMode=RCTM_User, time=0, value=0.1}
            {arriveTangent=-0.8452, interpMode=RCIM_Cubic, leaveTangent=-0.8452, tangentMode=RCTM_User, time=0.1418, value=0.9847}
            {arriveTangent=-0.8544, interpMode=RCIM_Cubic, leaveTangent=-0.8544, tangentMode=RCTM_User, time=1, value=0}
           ]
          }}}
        Scale Curve=2
       )
       B=&RemapRange(
        Input Value To Remap=&DistanceBetweenPositions(A=$Engine.Emitter.SimulationPosition, B=$Particles.Position)
        Input Min=0
        Input Max=300
        Output Min=1
        Output Max=0
       )
      )
     )
    )
    Particles.CenterFlareSize:Vector2f=&Multiply_Vector2(A=$Particles.SpriteSize, B=&Vector2DFromFloat(Value=0.5))
    Particles.CoronaFlareColor:LinearColor=&MultiplyLinearColor(
     Linear Color=$Particles.Color
     Scale Factor=&MakeVector4FromVectorAndFloat(
      "Vector (RGB)"=&MakeVectorFromLinearColorRGB(Color=<rgba 0.15 0.0919 0.0339 1>)
      "Float (Alpha)"=&Multiply_Float(
       A=&FloatFromCurve(
        FloatCurve=DataInterface() {propertyValues {Curve {
           defaultValue=1
           keys [
            {arriveTangent=4.0437, interpMode=RCIM_Cubic, leaveTangent=4.0437, tangentMode=RCTM_User, time=0, value=0.1}
            {arriveTangent=-0.8452, interpMode=RCIM_Cubic, leaveTangent=-0.8452, tangentMode=RCTM_User, time=0.1418, value=0.9847}
            {arriveTangent=-0.8544, interpMode=RCIM_Cubic, leaveTangent=-0.8544, tangentMode=RCTM_User, time=1, value=0}
           ]
          }}}
        Scale Curve=2
       )
       B=&RemapRange(
        Input Value To Remap=&DistanceBetweenPositions(A=$Engine.Emitter.SimulationPosition, B=$Particles.Position)
        Input Min=0
        Input Max=300
        Output Min=1
        Output Max=0
       )
      )
     )
    )
    Particles.CoronaFlareSize:Vector2f=&Multiply_Vector2(
     A=&Multiply_Vector2D_ByFloat(Float=&Lerp_Float(A=0.3, B=1, Alpha=$System.AudioReactivity), Vector2D=$Particles.Initial.SpriteSize)
     B=&Vector2DFromFloat(Value=0.75)
    )
   ) {isSetParameters=true, module=SetVariables_#7}
   &Drag(Use Linear Drag=true, Use Rotational Drag=true)
   &SolveForcesAndVelocity(Clamp Velocity=false, Limit Acceleration=false)
  ]
  rendererClasses [&NiagaraSpriteRendererProperties]
  renderers [
   {rendererClass=&NiagaraSpriteRendererProperties}
   {rendererClass=&NiagaraSpriteRendererProperties, rendererIndex=1}
  ]
 }
 Stars_Packed_02 {
  emitterUpdateScript [
   &EmitterState(
    Inactive Response=::NewEnumerator1
    Life Cycle Mode=::NewEnumerator1
    Loop Behavior=::NewEnumerator1
    Loop Duration Mode=::NewEnumerator1
    UseLoopDelay=false
   )
   &SpawnRate(
    Spawn Probability=&Lerp_Float(A=0, B=0.7, Alpha=$System.AudioReactivity)
    SpawnRate=20
    Use Spawn Probability=true
   )
  ]
  particleSpawnScript [
   &InitializeParticle(
    AdjustAlpha=false
    AdjustHue=true
    AdjustSaturation=true
    AdjustValue=true
    Color=<rgba 5 4.0303 0.9457 1>
    Color Mode=::NewEnumerator3
    Hue Shift Range:Vector2f=<-0.05 0.05>
    Lifetime Max=1.2
    Lifetime Mode=::NewEnumerator1
    Sprite Size Mode=::NewEnumerator3
    Sprite UV Mode=::NewEnumerator3
    Uniform Sprite Size=&Multiply_Float(
     A=&RandomRangeFloat(Minimum=15, Maximum=20)
     B=&SelectFloatFromArray(Direct Array Index=$Engine.QualityLevel, Float Selection Array=DataInterface() {propertyValues {FloatData [0.5,0.6,0.8,1,1]}})
    )
    UsePositionOffset=false
    Write Material Random=true
   )
   &CameraOffset(Camera Offset Amount=5)
   &AddVelocity(Velocity=&FN_FrontEndVector3Modifier(Gameplay=<0 -20 30>, Front End=<0 0 30>))
   &ShapeLocation(MaskSpawnGroup=false, Override Seed=false, Sphere Radius=20, UseOwnerScale=true)
  ]
  particleUpdateScript [
   &ParticleState()
   &SkeletalMeshLocation(
    Bone / Socket Index=&SelectIntFromArray(
     Array Sampling Mode=::NewEnumerator1
     Direct Array Index=?
     Int Selection Array=DataInterface() {propertyValues {IntData [32,33,34,35,36]}}
     Random Seed=0
    )
    Bone Sampling Mode=::NewEnumerator4
    MaskSpawnGroup=false
    Position Sampling=::NewEnumerator1
    Sampled Position Offset=$Particles.Initial.ShapeLocation.ShapeVector
    Skeletal Mesh=$System.Skeletal Mesh_Dyn
    Source Movement Scale:Vector3f=&Lerp_Vector(
     A=&VectorFromFloat(Value=&MakeCustomFloatFromBool(
       Bool=$User.IsSkydiving
       True Float=&MakeCustomFloatFromBool(
        Bool=&SetBoolByFloatComparison(
         A=&VectorLength(Vector=&Multiply_Vector(A=$Engine.Owner.Velocity, B=<1 1 0>))
         B=50
         Comparison Type=::NewEnumerator6
        )
        True Float=0.7
        False Float=0
       )
       False Float=0
      ))
     B=&VectorFromFloat(Value=&MakeCustomFloatFromBool(Bool=$User.IsSkydiving, True Float=0.9, False Float=0.7))
     Alpha=&RemapRange(
      Input Value To Remap=&VectorLength(Vector=&Multiply_Vector(A=$Engine.Owner.Velocity, B=<1 1 1>))
      Input Min=0
      Input Max=800
      Output Min=0
      Output Max=1
     )
    )
   )
   &ScaleSpriteSize(
    Uniform Curve Scale=&Lerp_Float(A=0.3, B=1, Alpha=$System.AudioReactivity)
    Uniform Curve Sprite Scale=DataInterface() {propertyValues {Curve {keys [
        {interpMode=RCIM_Cubic, tangentMode=RCTM_User, time=0, value=0}
        {interpMode=RCIM_Cubic, tangentMode=RCTM_User, time=0.5, value=1}
        {tangentMode=RCTM_User, time=1, value=0}
       ]}}}
   ) {module=ScaleSpriteSize001}
   &ScaleColor(
    Scale Alpha=&Multiply_Float(
     A=&Multiply_Float(
      A=$System.TODScalar
      B=&Multiply_Float(
       A=&FloatFromCurve(
        FloatCurve=DataInterface() {propertyValues {Curve {keys [
            {arriveTangent=6.5707, interpMode=RCIM_Cubic, leaveTangent=6.5707, tangentMode=RCTM_User, time=0, value=0}
            {interpMode=RCIM_Cubic, tangentMode=RCTM_User, time=0.3, value=1}
            {tangentMode=RCTM_User, time=1, value=0}
           ]}}}
        Scale Curve=&RandomRangeFloat(Minimum=0.1, Maximum=0.5)
       )
       B=1
      )
     )
     B=&Lerp_Float(A=0, B=1, Alpha=$System.AudioReactivity)
    )
    Scale RGB=&Multiply_Vector(
     A=&VectorFromFloat(Value=&Multiply_Float(
       A=&RandomRangeFloat(Minimum=0.75, Maximum=1)
       B=&SelectFloatFromArray(Direct Array Index=$Engine.QualityLevel, Float Selection Array=DataInterface() {propertyValues {FloatData [10,5,1,1,1]}})
      ))
     B=&MakeVectorFromLinearColorRGB(Color=$System.TODColor)
    )
    ScaleA=true
    ScaleRGB=true
   )
   /SetVariables(
    Particles.CenterFlareColor:LinearColor=&MultiplyLinearColor(
     Linear Color=$Particles.Color
     Scale Factor=&MakeVector4FromVectorAndFloat(
      "Vector (RGB)"=&MakeVectorFromLinearColorRGB(Color=<rgba 10 8.5224 3.7296 1>)
      "Float (Alpha)"=&Multiply_Float(
       A=&FloatFromCurve(
        FloatCurve=DataInterface() {propertyValues {Curve {
           defaultValue=1
           keys [
            {arriveTangent=4.0437, interpMode=RCIM_Cubic, leaveTangent=4.0437, tangentMode=RCTM_User, time=0, value=0.1}
            {arriveTangent=-0.8452, interpMode=RCIM_Cubic, leaveTangent=-0.8452, tangentMode=RCTM_User, time=0.1418, value=0.9847}
            {arriveTangent=-0.8544, interpMode=RCIM_Cubic, leaveTangent=-0.8544, tangentMode=RCTM_User, time=1, value=0}
           ]
          }}}
        Scale Curve=2
       )
       B=&RemapRange(
        Input Value To Remap=&DistanceBetweenPositions(A=$Engine.Emitter.SimulationPosition, B=$Particles.Position)
        Input Min=0
        Input Max=300
        Output Min=1
        Output Max=0
       )
      )
     )
    )
    Particles.CenterFlareSize:Vector2f=&Multiply_Vector2(A=$Particles.SpriteSize, B=&Vector2DFromFloat(Value=0.5))
    Particles.CoronaFlareColor:LinearColor=&MultiplyLinearColor(
     Linear Color=$Particles.Color
     Scale Factor=&MakeVector4FromVectorAndFloat(
      "Vector (RGB)"=&MakeVectorFromLinearColorRGB(Color=<rgba 0.15 0.0919 0.0339 1>)
      "Float (Alpha)"=&Multiply_Float(
       A=&FloatFromCurve(
        FloatCurve=DataInterface() {propertyValues {Curve {
           defaultValue=1
           keys [
            {arriveTangent=4.0437, interpMode=RCIM_Cubic, leaveTangent=4.0437, tangentMode=RCTM_User, time=0, value=0.1}
            {arriveTangent=-0.8452, interpMode=RCIM_Cubic, leaveTangent=-0.8452, tangentMode=RCTM_User, time=0.1418, value=0.9847}
            {arriveTangent=-0.8544, interpMode=RCIM_Cubic, leaveTangent=-0.8544, tangentMode=RCTM_User, time=1, value=0}
           ]
          }}}
        Scale Curve=2
       )
       B=&RemapRange(
        Input Value To Remap=&DistanceBetweenPositions(A=$Engine.Emitter.SimulationPosition, B=$Particles.Position)
        Input Min=0
        Input Max=300
        Output Min=1
        Output Max=0
       )
      )
     )
    )
    Particles.CoronaFlareSize:Vector2f=&Multiply_Vector2(
     A=&Multiply_Vector2D_ByFloat(Float=&Lerp_Float(A=0.3, B=1, Alpha=$System.AudioReactivity), Vector2D=$Particles.Initial.SpriteSize)
     B=&Vector2DFromFloat(Value=0.75)
    )
   ) {isSetParameters=true, module=SetVariables_#7}
   &Drag(Use Linear Drag=true, Use Rotational Drag=true)
   &SolveForcesAndVelocity(Clamp Velocity=false, Limit Acceleration=false)
  ]
  rendererClasses [&NiagaraSpriteRendererProperties]
  renderers [
   {rendererClass=&NiagaraSpriteRendererProperties}
   {rendererClass=&NiagaraSpriteRendererProperties, rendererIndex=1}
  ]
 }
}
summary {
 systemName=NS_Character_IndoorLaceBeat
 emitters [
  {emitterName=Pixie_Dust, rendererClasses [&NiagaraSpriteRendererProperties]}
  {emitterName=PixieDust, rendererClasses [&NiagaraSpriteRendererProperties]}
  {emitterName=Stars_Packed_02, rendererClasses [&NiagaraSpriteRendererProperties]}
  {emitterName=Stars_Packed_01, rendererClasses [&NiagaraSpriteRendererProperties]}
 ]
 userVariables [
  User.bIsADS {type=NiagaraBool}
  User.bIsFrontEnd {type=NiagaraBool}
  User.bIsFrontEndPreview {type=NiagaraBool}
  User.DebugReactivity {type=NiagaraBool}
  User.EnableFrontendFakeAudioLoop {type=NiagaraFloat, defaultValue=0, description=""}
  User.IsSkydiving {type=NiagaraBool}
 ]
}
