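# Material function for the IndoorLaceBeat outfit that draws animated spinning circle symbols and outputs the masked result plus the rotated UVs.
# A pivot map supplies centers, a scale channel and an edge-fix channel, recentering UV1 so each circle rotates over time with sine-based variation.
# Flipbook IDs become column and row indices for cells in the grayscale symbols texture.
# A rounded-distance test clips each cell, and biased UVs feed texture streaming info.
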MaterialFunction /BRCosmetics/Characters/Player/Female/Medium/Bodies/F_MED_IndoorLaceBeat/Materials/MF_IndoorLaceBeat_SpinningUVs
graph {expressions {
  MaterialExpressionAdd_1 {
   class=MaterialExpressionAdd
   inputs [
    A {from=MaterialExpressionDivide_2}
    B {from=MaterialExpressionScalarParameter_7}
   ]
   properties {constB=0}
  }
  MaterialExpressionAdd_2 {
   class=MaterialExpressionAdd
   inputs [
    A {from=MaterialExpressionSubtract_2}
    B {from=MaterialExpressionConstant2Vector_4}
   ]
   properties {desc=Center frame perfectly}
  }
  MaterialExpressionAdd_3 {
   class=MaterialExpressionAdd
   inputs [
    A {from=MaterialExpressionAdd_2}
    B {from=MaterialExpressionDivide_1}
   ]
  }
  MaterialExpressionAdd_4 {
   class=MaterialExpressionAdd
   inputs [
    A {from=MaterialExpressionComponentMask_1}
    B {from=MaterialExpressionComponentMask_5}
   ]
  }
  MaterialExpressionAdd_5 {
   class=MaterialExpressionAdd
   inputs [
    A {from=MaterialExpressionTime_1}
    B {from=MaterialExpressionMultiply_11}
   ]
  }
  MaterialExpressionAdd_6 {
   class=MaterialExpressionAdd
   inputs [
    A {from=MaterialExpressionScalarParameter_8}
    B {from=MaterialExpressionMultiply_12}
   ]
  }
  MaterialExpressionAdd_9 {
   class=MaterialExpressionAdd
   inputs [
    A {from=MaterialExpressionLocalPosition_1, output=Z}
    B {from=MaterialExpressionScalarParameter_0}
   ]
  }
  MaterialExpressionAppendVector_1 {
   class=MaterialExpressionAppendVector
   inputs [
    A {from=MaterialExpressionNamedRerouteDeclaration_11, output=""}
    B {from=MaterialExpressionFloor_0, output=""}
   ]
  }
  MaterialExpressionArctangent2_0 {
   class=MaterialExpressionArctangent2
   inputs [
    Y {from=MaterialExpressionConvert_1, output=""}
    X {from=MaterialExpressionConvert_1, output=""}
   ]
   properties {}
  }
  MaterialExpressionCeil_1 {class=MaterialExpressionCeil, inputs [None {from=MaterialExpressionScalarParameter_3, output=""}], properties {}}
  MaterialExpressionCeil_2 {class=MaterialExpressionCeil, inputs [None {from=MaterialExpressionOneMinus_0, output=""}], properties {}}
  MaterialExpressionClamp_0 {
   class=MaterialExpressionClamp
   inputs [
    None {from=MaterialExpressionLinearInterpolate_2, output=""}
    Min {from="", output=""}
    Max {from="", output=""}
   ]
   properties {minDefault=-1}
  }
  MaterialExpressionComponentMask_0 {class=MaterialExpressionComponentMask, inputs [None {from=MaterialExpressionNamedRerouteUsage_10}], properties {b=true}}
  MaterialExpressionComponentMask_1 {class=MaterialExpressionComponentMask, inputs [None {from=MaterialExpressionNamedRerouteUsage_12}], properties {g=true, r=true}}
  MaterialExpressionComponentMask_2 {class=MaterialExpressionComponentMask, inputs [None {from=MaterialExpressionNamedRerouteUsage_12}], properties {a=true}}
  MaterialExpressionComponentMask_3 {class=MaterialExpressionComponentMask, inputs [None {from=MaterialExpressionTextureCoordinate_4}], properties {r=true}}
  MaterialExpressionComponentMask_4 {class=MaterialExpressionComponentMask, inputs [None {from=MaterialExpressionSine_0}], properties {r=true}}
  MaterialExpressionComponentMask_5 {class=MaterialExpressionComponentMask, inputs [None {from=MaterialExpressionLinearInterpolate_0}], properties {g=true, r=true}}
  MaterialExpressionConstant2Vector_1 {class=MaterialExpressionConstant2Vector, properties {g=1, r=1}}
  MaterialExpressionConstant2Vector_4 {class=MaterialExpressionConstant2Vector, properties {g=0.5, r=0.5}}
  MaterialExpressionConstant2Vector_5 {class=MaterialExpressionConstant2Vector}
  MaterialExpressionConstant4Vector_0 {class=MaterialExpressionConstant4Vector, inputs [], properties {constant=<rgba -1 0 0 1>}}
  MaterialExpressionConstant_0 {class=MaterialExpressionConstant}
  MaterialExpressionConstant_1 {class=MaterialExpressionConstant}
  MaterialExpressionConstant_2 {class=MaterialExpressionConstant, properties {r=2}}
  MaterialExpressionConstant_3 {class=MaterialExpressionConstant, properties {r=1}}
  MaterialExpressionConstant_4 {class=MaterialExpressionConstant}
  MaterialExpressionConstant_5 {class=MaterialExpressionConstant, properties {r=2}}
  MaterialExpressionConstant_6 {class=MaterialExpressionConstant, properties {r=1}}
  MaterialExpressionConstant_7 {class=MaterialExpressionConstant, properties {r=2}}
  MaterialExpressionConvert_0 {
   class=MaterialExpressionConvert
   inputs [
    None {from=MaterialExpressionTextureCoordinate_4}
    None {from=MaterialExpressionFrac_0}
   ]
   properties {
    convertInputs [
     {
      type=Vector2
      expressionInput {
       expression=&self:MaterialExpressionTextureCoordinate_4
       inputName=None
       mask=0
       maskA=0
       maskB=0
       maskG=0
       maskR=0
       outputIndex=0
      }
     }
     {
      type=Scalar
      expressionInput {
       expression=&self:MaterialExpressionFrac_0
       inputName=None
       mask=0
       maskA=0
       maskB=0
       maskG=0
       maskR=0
       outputIndex=0
      }
     }
    ]
    convertOutputs [{type=Vector2}]
   }
  }
  MaterialExpressionConvert_1 {
   class=MaterialExpressionConvert
   inputs [None {from=MaterialExpressionVertexInterpolator_1, output=PS}]
   properties {
    convertInputs [{
      type=Vector2
      expressionInput {
       expression=&self:MaterialExpressionVertexInterpolator_1
       inputName=None
       mask=0
       maskA=0
       maskB=0
       maskG=0
       maskR=0
       outputIndex=0
      }
     }]
    convertOutputs [{type=Scalar}, {type=Scalar}]
   }
  }
  MaterialExpressionDistance_0 {
   class=MaterialExpressionDistance
   inputs [
    A {from=MaterialExpressionRound_0, output=""}
    B {from=MaterialExpressionConstant2Vector_5, output=""}
   ]
   properties {}
  }
  MaterialExpressionDivide_0 {
   class=MaterialExpressionDivide
   inputs [
    A {from=MaterialExpressionNamedRerouteDeclaration_12}
    B {from=MaterialExpressionScalarParameter_5}
   ]
  }
  MaterialExpressionDivide_1 {
   class=MaterialExpressionDivide
   inputs [
    A {from=MaterialExpressionNamedRerouteUsage_5}
    B {from=""}
   ]
  }
  MaterialExpressionDivide_2 {
   class=MaterialExpressionDivide
   inputs [
    A {from=MaterialExpressionArctangent2_0}
    B {from=MaterialExpressionMaterialFunctionCall_2, output=Result}
   ]
  }
  MaterialExpressionFloor_0 {class=MaterialExpressionFloor, inputs [None {from=MaterialExpressionDivide_0, output=""}], properties {}}
  MaterialExpressionFrac_0 {class=MaterialExpressionFrac, inputs [None {from=MaterialExpressionAdd_1, output=""}], properties {}}
  MaterialExpressionFrac_1 {class=MaterialExpressionFrac, inputs [None {from=MaterialExpressionMultiply_1, output=""}], properties {}}
  MaterialExpressionFrac_2 {class=MaterialExpressionFrac, inputs [None {from=MaterialExpressionLinearInterpolate_2, output=""}], properties {}}
  MaterialExpressionFunctionInput_0 {
   class=MaterialExpressionFunctionInput
   inputs [Preview {from=MaterialExpressionTextureObject_0, output=""}]
   properties {inputName=PivotsMap, inputType=FunctionInput_Texture2D}
  }
  MaterialExpressionFunctionInput_1 {
   class=MaterialExpressionFunctionInput
   inputs [Preview {from=MaterialExpressionConstant_1, output=""}]
   properties {bUsePreviewValueAsDefault=true, inputName=InvertEdgeFix, inputType=FunctionInput_Scalar, sortPriority=3}
  }
  MaterialExpressionFunctionInput_2 {
   class=MaterialExpressionFunctionInput
   inputs [Preview {from="", output=""}]
   properties {bUsePreviewValueAsDefault=true, inputName=Mask, inputType=FunctionInput_Scalar, sortPriority=1, previewValue=<1 0 0 1>}
  }
  MaterialExpressionFunctionInput_3 {
   class=MaterialExpressionFunctionInput
   inputs [Preview {from="", output=""}]
   properties {bUsePreviewValueAsDefault=true, inputName=FlipbookIDs, inputType=FunctionInput_Scalar, sortPriority=3}
  }
  MaterialExpressionFunctionInput_4 {
   class=MaterialExpressionFunctionInput
   inputs [Preview {from=MaterialExpressionConstant2Vector_1, output=""}]
   properties {bUsePreviewValueAsDefault=true, inputName=ScaleCorrection, inputType=FunctionInput_Vector2, sortPriority=3}
  }
  MaterialExpressionFunctionInput_5 {
   class=MaterialExpressionFunctionInput
   inputs [Preview {from=MaterialExpressionConstant_6, output=""}]
   properties {bUsePreviewValueAsDefault=true, inputName=FixEdges, inputType=FunctionInput_Scalar, sortPriority=3}
  }
  MaterialExpressionFunctionInput_6 {
   class=MaterialExpressionFunctionInput
   inputs [Preview {from=MaterialExpressionConstant_7, output=""}]
   properties {bUsePreviewValueAsDefault=true, inputName=BaseScale, inputType=FunctionInput_Scalar, sortPriority=3}
  }
  MaterialExpressionFunctionOutput_0 {class=MaterialExpressionFunctionOutput, inputs [None {from=MaterialExpressionMultiply_7, output=""}]}
  MaterialExpressionFunctionOutput_1 {
   class=MaterialExpressionFunctionOutput
   inputs [None {from=MaterialExpressionMaterialFunctionCall_0, output=Rotated Values}]
   properties {outputName=UVs}
  }
  MaterialExpressionIf_0 {
   class=MaterialExpressionIf
   inputs [
    A {from=MaterialExpressionModulo_0, output=""}
    B {from="", output=""}
    "A > B" {from=MaterialExpressionModulo_0, output=""}
    "A == B" {from=MaterialExpressionConstant_4, output=""}
    "A < B" {from=MaterialExpressionConstant_3, output=""}
   ]
   properties {}
  }
  MaterialExpressionLinearInterpolate_0 {
   class=MaterialExpressionLinearInterpolate
   inputs [
    A {from=MaterialExpressionConstant4Vector_0, output=RGBA}
    B {from=MaterialExpressionMultiply_3}
    Alpha {from=MaterialExpressionNamedRerouteUsage_1}
   ]
  }
  MaterialExpressionLinearInterpolate_1 {
   class=MaterialExpressionLinearInterpolate
   inputs [
    A {from=MaterialExpressionComponentMask_3}
    B {from=MaterialExpressionFrac_0}
    Alpha {from=""}
   ]
  }
  MaterialExpressionLinearInterpolate_2 {
   class=MaterialExpressionLinearInterpolate
   inputs [
    A {from=MaterialExpressionComponentMask_1}
    B {from=MaterialExpressionAdd_4}
    Alpha {from=MaterialExpressionMultiply_2}
   ]
  }
  MaterialExpressionLocalPosition_0 {class=MaterialExpressionLocalPosition, inputs [], properties {localOrigin=InstancePreSkinning}}
  MaterialExpressionLocalPosition_1 {class=MaterialExpressionLocalPosition, inputs [], properties {localOrigin=InstancePreSkinning}}
  MaterialExpressionMaterialFunctionCall_0 {
   class=MaterialExpressionMaterialFunctionCall
   inputs [
    UVs {from=MaterialExpressionMultiply_4}
    Rotation Center {from=MaterialExpressionConstant_0}
    "Rotation Angle (0-1)" {from=MaterialExpressionFrac_1}
   ]
   properties {materialFunction=&CustomRotator}
  }
  MaterialExpressionMaterialFunctionCall_2 {
   class=MaterialExpressionMaterialFunctionCall
   inputs [Multiplier {from=MaterialExpressionConstant_2}]
   properties {materialFunction=&Pi}
  }
  MaterialExpressionMaterialFunctionCall_3 {
   class=MaterialExpressionMaterialFunctionCall
   inputs [
    Vector or UVs {from=""}
    Swizzle Coordinate Output {from=""}
   ]
   properties {materialFunction=&VectorToRadialValue}
  }
  MaterialExpressionModulo_0 {
   class=MaterialExpressionModulo
   inputs [
    A {from=MaterialExpressionNamedRerouteDeclaration_12, output=""}
    B {from=MaterialExpressionScalarParameter_4, output=""}
   ]
   properties {}
  }
  MaterialExpressionMultiply_0 {
   class=MaterialExpressionMultiply
   inputs [
    A {from=MaterialExpressionTextureSample_0, output=R}
    B {from=MaterialExpressionNamedRerouteUsage_8}
   ]
  }
  MaterialExpressionMultiply_1 {
   class=MaterialExpressionMultiply
   inputs [
    A {from=MaterialExpressionAdd_5}
    B {from=MaterialExpressionAdd_6}
   ]
   properties {constB=0.1}
  }
  MaterialExpressionMultiply_11 {
   class=MaterialExpressionMultiply
   inputs [
    A {from=MaterialExpressionComponentMask_4}
    B {from=""}
   ]
  }
  MaterialExpressionMultiply_12 {
   class=MaterialExpressionMultiply
   inputs [
    A {from=MaterialExpressionMultiply_11}
    B {from=""}
   ]
   properties {constB=0.05}
  }
  MaterialExpressionMultiply_13 {
   class=MaterialExpressionMultiply
   inputs [
    A {from=MaterialExpressionAdd_9}
    B {from=""}
   ]
   properties {constB=-0.01}
  }
  MaterialExpressionMultiply_2 {
   class=MaterialExpressionMultiply
   inputs [
    A {from=MaterialExpressionComponentMask_2}
    B {from=MaterialExpressionNamedRerouteUsage_13}
   ]
  }
  MaterialExpressionMultiply_21 {
   class=MaterialExpressionMultiply
   inputs [
    A {from=MaterialExpressionMultiply_13}
    B {from=MaterialExpressionScalarParameter_6}
   ]
  }
  MaterialExpressionMultiply_3 {
   class=MaterialExpressionMultiply
   inputs [
    A {from=MaterialExpressionConstant4Vector_0, output=RGBA}
    B {from=""}
   ]
   properties {constB=-1}
  }
  MaterialExpressionMultiply_4 {
   class=MaterialExpressionMultiply
   inputs [
    A {from=MaterialExpressionMultiply_6}
    B {from=MaterialExpressionSubtract_5}
   ]
   properties {constB=2}
  }
  MaterialExpressionMultiply_5 {
   class=MaterialExpressionMultiply
   inputs [
    A {from=MaterialExpressionAdd_3}
    B {from=MaterialExpressionPower_0}
   ]
  }
  MaterialExpressionMultiply_6 {
   class=MaterialExpressionMultiply
   inputs [
    A {from=MaterialExpressionNamedRerouteDeclaration_1}
    B {from=MaterialExpressionNamedRerouteUsage_2}
   ]
  }
  MaterialExpressionMultiply_7 {
   class=MaterialExpressionMultiply
   inputs [
    A {from=MaterialExpressionMultiply_0}
    B {from=MaterialExpressionCeil_2}
   ]
  }
  MaterialExpressionMultiply_8 {
   class=MaterialExpressionMultiply
   inputs [
    A {from=MaterialExpressionMaterialFunctionCall_0, output=Rotated Values}
    B {from=""}
   ]
   properties {constB=2}
  }
  MaterialExpressionNamedRerouteDeclaration_0 {
   class=MaterialExpressionNamedRerouteDeclaration
   inputs [None {from=MaterialExpressionFunctionInput_1}]
   properties {name=InvertEdgeFix, nodeColor=<rgba 0.4297 0 1 1>}
  }
  MaterialExpressionNamedRerouteDeclaration_1 {
   class=MaterialExpressionNamedRerouteDeclaration
   inputs [None {from=MaterialExpressionSubtract_1}]
   properties {name=Positions, nodeColor=<rgba 1 0.3281 0 1>}
  }
  MaterialExpressionNamedRerouteDeclaration_10 {
   class=MaterialExpressionNamedRerouteDeclaration
   inputs [None {from=MaterialExpressionFunctionInput_3}]
   properties {name=FlipbookIDs, nodeColor=<rgba 1 0 0.2812 1>}
  }
  MaterialExpressionNamedRerouteDeclaration_11 {
   class=MaterialExpressionNamedRerouteDeclaration
   inputs [None {from=MaterialExpressionModulo_0}]
   properties {name=CurrentColumn, nodeColor=<rgba 0.1016 1 0 1>}
  }
  MaterialExpressionNamedRerouteDeclaration_12 {
   class=MaterialExpressionNamedRerouteDeclaration
   inputs [None {from=MaterialExpressionNamedRerouteUsage_6}]
   properties {name=CurrentIndex, nodeColor=<rgba 0 1 0.0391 1>}
  }
  MaterialExpressionNamedRerouteDeclaration_13 {
   class=MaterialExpressionNamedRerouteDeclaration
   inputs [None {from=MaterialExpressionAppendVector_1}]
   properties {name=FlipbookIndex, nodeColor=<rgba 0 0.1094 1 1>}
  }
  MaterialExpressionNamedRerouteDeclaration_17 {
   class=MaterialExpressionNamedRerouteDeclaration
   inputs [None {from=MaterialExpressionTextureSample_1, output=RGBA}]
   properties {name=PivotsMap, nodeColor=<rgba 0 0.1094 1 1>}
  }
  MaterialExpressionNamedRerouteDeclaration_19 {
   class=MaterialExpressionNamedRerouteDeclaration
   inputs [None {from=MaterialExpressionFunctionInput_5}]
   properties {name=FixEdges, nodeColor=<rgba 0.4531 1 0 1>}
  }
  MaterialExpressionNamedRerouteDeclaration_2 {
   class=MaterialExpressionNamedRerouteDeclaration
   inputs [None {from=MaterialExpressionFunctionInput_4}]
   properties {name=ScaleCorrection, nodeColor=<rgba 0.9453 1 0 1>}
  }
  MaterialExpressionNamedRerouteDeclaration_21 {
   class=MaterialExpressionNamedRerouteDeclaration
   inputs [None {from=MaterialExpressionFunctionInput_6}]
   properties {name=BaseScale, nodeColor=<rgba 0 0.8594 1 1>}
  }
  MaterialExpressionNamedRerouteDeclaration_6 {
   class=MaterialExpressionNamedRerouteDeclaration
   inputs [None {from=MaterialExpressionLinearInterpolate_2}]
   properties {name=Pivots, nodeColor=<rgba 1 0 0.9844 1>}
  }
  MaterialExpressionNamedRerouteDeclaration_7 {
   class=MaterialExpressionNamedRerouteDeclaration
   inputs [None {from=MaterialExpressionComponentMask_0}]
   properties {name=ScaleMap, nodeColor=<rgba 0 0.0859 1 1>}
  }
  MaterialExpressionNamedRerouteDeclaration_8 {
   class=MaterialExpressionNamedRerouteDeclaration
   inputs [None {from=MaterialExpressionFunctionInput_2}]
   properties {name=Mask, nodeColor=<rgba 0.6641 1 0 1>}
  }
  MaterialExpressionNamedRerouteUsage_0 {class=MaterialExpressionNamedRerouteUsage}
  MaterialExpressionNamedRerouteUsage_1 {class=MaterialExpressionNamedRerouteUsage}
  MaterialExpressionNamedRerouteUsage_10 {class=MaterialExpressionNamedRerouteUsage}
  MaterialExpressionNamedRerouteUsage_12 {class=MaterialExpressionNamedRerouteUsage}
  MaterialExpressionNamedRerouteUsage_13 {class=MaterialExpressionNamedRerouteUsage}
  MaterialExpressionNamedRerouteUsage_14 {class=MaterialExpressionNamedRerouteUsage}
  MaterialExpressionNamedRerouteUsage_2 {class=MaterialExpressionNamedRerouteUsage}
  MaterialExpressionNamedRerouteUsage_3 {class=MaterialExpressionNamedRerouteUsage}
  MaterialExpressionNamedRerouteUsage_4 {class=MaterialExpressionNamedRerouteUsage}
  MaterialExpressionNamedRerouteUsage_5 {class=MaterialExpressionNamedRerouteUsage}
  MaterialExpressionNamedRerouteUsage_6 {class=MaterialExpressionNamedRerouteUsage}
  MaterialExpressionNamedRerouteUsage_8 {class=MaterialExpressionNamedRerouteUsage}
  MaterialExpressionOneMinus_0 {class=MaterialExpressionOneMinus, inputs [None {from=MaterialExpressionDistance_0}]}
  MaterialExpressionPower_0 {
   class=MaterialExpressionPower
   inputs [
    Base {from=MaterialExpressionConstant_5}
    Exp {from=MaterialExpressionScalarParameter_1}
   ]
  }
  MaterialExpressionRecordTextureStreamingInfo_0 {
   class=MaterialExpressionRecordTextureStreamingInfo
   inputs [
    Tex {from=MaterialExpressionTextureObjectParameter_0, output=""}
    UVs {from=MaterialExpressionMultiply_5, output=""}
   ]
   properties {}
  }
  MaterialExpressionRound_0 {class=MaterialExpressionRound, inputs [None {from=MaterialExpressionMultiply_8, output=""}], properties {}}
  MaterialExpressionScalarParameter_0 {class=MaterialExpressionScalarParameter, properties {defaultValue=-111, parameterName=Param_1}}
  MaterialExpressionScalarParameter_1 {
   class=MaterialExpressionScalarParameter
   properties {defaultValue=-2, group=IndoorLaceBeat|Animated Circles, parameterName=MipBias for Texture Loading}
  }
  MaterialExpressionScalarParameter_3 {class=MaterialExpressionScalarParameter, properties {defaultValue=3, group=IndoorLaceBeat|Animated Circles, parameterName=DebugIndex}}
  MaterialExpressionScalarParameter_4 {class=MaterialExpressionScalarParameter, properties {defaultValue=2, group=IndoorLaceBeat|Animated Circles, parameterName=FlipbookColumnAmt}}
  MaterialExpressionScalarParameter_5 {class=MaterialExpressionScalarParameter, properties {defaultValue=2, group=IndoorLaceBeat|Animated Circles, parameterName=FlipbookRowAmt}}
  MaterialExpressionScalarParameter_6 {class=MaterialExpressionScalarParameter, properties {defaultValue=0.9, parameterName=B}}
  MaterialExpressionScalarParameter_7 {class=MaterialExpressionScalarParameter, properties {defaultValue=0.1}}
  MaterialExpressionScalarParameter_8 {class=MaterialExpressionScalarParameter, properties {defaultValue=0.1, group=IndoorLaceBeat|Animated Circles, parameterName=RotationSpeed}}
  MaterialExpressionSine_0 {class=MaterialExpressionSine, inputs [None {from=MaterialExpressionNamedRerouteUsage_4}]}
  MaterialExpressionSubtract_1 {
   class=MaterialExpressionSubtract
   inputs [
    A {from=MaterialExpressionTextureCoordinate_1}
    B {from=MaterialExpressionNamedRerouteUsage_0}
   ]
  }
  MaterialExpressionSubtract_2 {
   class=MaterialExpressionSubtract
   inputs [
    A {from=MaterialExpressionMaterialFunctionCall_0, output=Rotated Values}
    B {from=""}
   ]
   properties {constB=0.25}
  }
  MaterialExpressionSubtract_5 {
   class=MaterialExpressionSubtract
   inputs [
    A {from=MaterialExpressionNamedRerouteUsage_14}
    B {from=MaterialExpressionNamedRerouteUsage_3}
   ]
  }
  MaterialExpressionTextureCoordinate_1 {class=MaterialExpressionTextureCoordinate, properties {coordinateIndex=1}}
  MaterialExpressionTextureCoordinate_4 {class=MaterialExpressionTextureCoordinate, properties {coordinateIndex=1}}
  MaterialExpressionTextureObjectParameter_0 {
   class=MaterialExpressionTextureObjectParameter
   inputs [
    UVs {from="", output=""}
    Apply View MipBias {from="", output=""}
   ]
   properties {
    group=IndoorLaceBeat|Animated Circles
    parameterName=Circles Texture
    samplerType=SAMPLERTYPE_LinearGrayscale
    texture=&/T_IndoorLace_symbols
   }
  }
  MaterialExpressionTextureObject_0 {
   class=MaterialExpressionTextureObject
   inputs []
   properties {
    samplerType=SAMPLERTYPE_LinearColor
    texture=&/T_IndoorLace_Positions_1
   }
  }
  MaterialExpressionTextureSample_0 {
   class=MaterialExpressionTextureSample
   inputs [
    UVs {from=MaterialExpressionAdd_3}
    Tex {from=MaterialExpressionRecordTextureStreamingInfo_0, output=Tex}
    Apply View MipBias {from=""}
   ]
   properties {samplerSource=SSM_Wrap_WorldGroupSettings}
  }
  MaterialExpressionTextureSample_1 {
   class=MaterialExpressionTextureSample
   inputs [
    UVs {from=""}
    Tex {from=MaterialExpressionFunctionInput_0}
    Apply View MipBias {from=""}
   ]
   properties {
    constCoordinate=1
    samplerSource=SSM_Clamp_WorldGroupSettings
    samplerType=SAMPLERTYPE_LinearColor
    texture=&T_IndoorLace_Positions_1
   }
  }
  MaterialExpressionTime_1 {class=MaterialExpressionTime}
  MaterialExpressionVertexInterpolator_0 {class=MaterialExpressionVertexInterpolator, inputs [VS {from=MaterialExpressionMultiply_21, output=""}], properties {}}
  MaterialExpressionVertexInterpolator_1 {class=MaterialExpressionVertexInterpolator, inputs [VS {from=MaterialExpressionLocalPosition_0, output=XY}], properties {}}
 }}
