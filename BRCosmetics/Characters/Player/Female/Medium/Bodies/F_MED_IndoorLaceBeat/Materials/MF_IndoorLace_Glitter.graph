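# Material function that layers animated, twinkling glitter onto the IndoorLace female body cosmetic's material attributes, behind a Use Glitter static switch.
# Samples a sparkle mask twice, once with bump-offset parallax, using tiling blended by camera depth fade.
# A time-driven sine mask, color, power and intensity inputs, and inverse Fresnel shape the glow.
# A channel of the SkinFX mask gates it into emissive and lerps base color, opacity and subsurface color.
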MaterialFunction /BRCosmetics/Characters/Player/Female/Medium/Bodies/F_MED_IndoorLaceBeat/Materials/MF_IndoorLace_Glitter
graph {
 properties {bExposeToLibrary=true}
 nodes [
  FunctionOutput(None=StaticSwitchParameter(True=NamedRerouteUsage(), False=NamedRerouteUsage()) {defaultValue=true, group=Glitter, parameterName=Use Glitter})
  n1 = FunctionInput(Preview=Constant4Vector() {constant=<rgba 1 0.9767 0.8802 1>}.RGB) {bUsePreviewValueAsDefault=true, inputName=GlitterColor, sortPriority=1, previewValue=<100 1 0 1>}
  NamedRerouteDeclaration(None=@n1) {name=GlitterColor, nodeColor=<rgba 0.6875 1 0 1>}
  n2 = FunctionInput() {bUsePreviewValueAsDefault=true, inputName=SpeedMultiplier, inputType=FunctionInput_Scalar, sortPriority=1, previewValue=<1 1 0 1>}
  NamedRerouteDeclaration(None=@n2) {name=SpeedMultiplier, nodeColor=<rgba 0 1 0.5078 1>}
  n3 = ScalarParameter() {defaultValue=16, group=Glitter, parameterName=Sparkle Detail Tiling, sliderMax=64, sliderMin=8, sortPriority=2}
  NamedRerouteDeclaration(None=@n3) {name=Tiling, nodeColor=<rgba 0.2656 0 1 1>}
  n4 = NamedRerouteUsage()
  n5 = Subtract(A=NamedRerouteUsage(), B=Frac(None=NamedRerouteUsage()))
  n6 = NamedRerouteUsage()
  n7 = Multiply(
   A=Add(A=Divide(A=@n5, B=@n6), B=Time())
   B=Multiply(A=ScalarParameter() {defaultValue=0.2, group=Glitter, parameterName=Time Scale, sliderMax=2, sortPriority=3}, B=NamedRerouteUsage())
  ) {constB=0.5}
  n8 = MaterialFunctionCall(UV Channel=Frac(None=Add(A=@n4, B=@n7))) {materialFunction=&LinearGradient}
  NamedRerouteDeclaration(None=ConstantBiasScale(None=Sine(None=@n8.VGradient))) {name=GlitterAnimatedMask, nodeColor=<rgba 0.1016 0 1 1>}
  NamedRerouteDeclaration(None=FunctionInput() {inputType=FunctionInput_MaterialAttributes}) {name=InMaterial, nodeColor=<rgba 0 1 0.2266 1>}
  n9 = GetMaterialAttributes(None=NamedRerouteUsage()) {attributeGetTypes [
    #1
    #2
    #3
    #4
   ]}
  n10 = Multiply(
   A=ChannelMaskParameter(None=NamedRerouteUsage()) {group=Glitter, parameterName=Glitter Channel Mask, sortPriority=0}.R
   B=NamedRerouteUsage()
  )
  n11 = Saturate(None=@n10)
  n12 = SetMaterialAttributes!(
   MaterialAttributes=@n9.MaterialAttributes
   EmissiveColor=Add(A=@n9.EmissiveColor, B=@n10)
   BaseColor=LinearInterpolate(A=@n9.BaseColor, Alpha=@n11)
   Opacity=LinearInterpolate(A=@n9.Opacity, Alpha=@n11)
   SubsurfaceColor=LinearInterpolate(A=@n9.SubsurfaceColor, Alpha=@n11)
  ) {attributeSetTypes [
    #1
    #2
    #3
    #4
   ]}
  NamedRerouteDeclaration(None=@n12) {name=OutGlitter, nodeColor=<rgba 1 0 0.0937 1>}
  n13 = FunctionInput() {bUsePreviewValueAsDefault=true, inputName=GlitterIntensity, inputType=FunctionInput_Scalar, sortPriority=1, previewValue=<100 1 0 1>}
  NamedRerouteDeclaration(None=@n13) {name=GlitterIntensity, nodeColor=<rgba 1 0 0.7734 1>}
  n14 = FunctionInput() {bUsePreviewValueAsDefault=true, inputName=SkinFX_Mask, inputType=FunctionInput_Vector4, sortPriority=1, previewValue=<1 1 0 1>}
  NamedRerouteDeclaration(None=@n14) {name=FX Map, nodeColor=<rgba 0 1 0.0625 1>}
  n15 = FunctionInput() {bUsePreviewValueAsDefault=true, inputName=GlitterPower, inputType=FunctionInput_Scalar, sortPriority=1, previewValue=<4 1 0 1>}
  NamedRerouteDeclaration(None=@n15) {name=GlitterPower, nodeColor=<rgba 0 0.1328 1 1>}
  n16 = Power(Base=Multiply(A=NamedRerouteUsage(), B=NamedRerouteUsage()), Exp=NamedRerouteUsage()) {constExponent=4}
  n17 = NamedRerouteUsage()
  n18 = Multiply(A=NamedRerouteUsage(), B=Saturate(None=Multiply(A=@n16, B=@n17) {constB=4}))
  n19 = ScalarParameter() {defaultValue=10, group=Glitter, parameterName=Emissive Intensity, sliderMax=20, sortPriority=5}
  n20 = OneMinus(None=Fresnel() {baseReflectFraction=0, exponent=1})
  n21 = ScalarParameter() {defaultValue=0.5, group=Glitter, parameterName=GlitterFresnelExp}
  NamedRerouteDeclaration(None=Multiply(A=Multiply(A=@n18, B=@n19), B=Power(Base=@n20, Exp=@n21) {constExponent=3})) {name=Glitter Animated Emissive, nodeColor=<rgba 1 0.2874 0 1>}
  n22 = TextureCoordinate()
  n23 = ScalarParameter() {defaultValue=1, group=Glitter, parameterName=Glitter Tiling Close}
  n24 = ScalarParameter() {defaultValue=0.5, group=Glitter, parameterName=Glitter Tiling Far}
  n25 = MaterialFunctionCall(
   Fade Length=ScalarParameter() {defaultValue=2, group=Glitter, parameterName=Glitter Fade Length}
   Fade Offset=ScalarParameter() {defaultValue=100, group=Glitter, parameterName=Glitter Fade Offset}
  ) {materialFunction=&CameraDepthFade}
  n26 = NamedRerouteDeclaration(None=Multiply(A=Multiply(A=@n22, B=@n3) {constB=3}, B=LinearInterpolate(A=@n23, B=@n24, Alpha=@n25.Result))) {name=UVs, nodeColor=<rgba 1 0.5859 0 1>}
  n27 = TextureSampleParameter2D(UVs=@n26) {
   group=Glitter
   parameterName=Sparkle Mask
   samplerSource=SSM_Wrap_WorldGroupSettings
   samplerType=SAMPLERTYPE_Masks
   sortPriority=0
   texture=/Game/Weapons/WeaponSkins/Textures/T_SparkleMask
  }
  n28 = TextureSampleParameter2D(UVs=BumpOffset(Coordinate=@n26, Height=ScalarParameter() {defaultValue=-3, group=Glitter, parameterName=Glitter Depth})) {
   group=Glitter
   parameterName=Sparkle Mask
   samplerSource=SSM_Wrap_WorldGroupSettings
   samplerType=SAMPLERTYPE_Masks
   sortPriority=0
   texture=&T_SparkleMask
  }
  NamedRerouteDeclaration(None=Multiply(A=@n27.R, B=@n28.R)) {name=Glitter, nodeColor=<rgba 1 0 0.7734 1>}
  NamedRerouteDeclaration(None=@n28.G) {name=GlitterMask, nodeColor=<rgba 0 1 0.9063 1>}
  ScalarParameter() {defaultValue=4, group=Glitter, parameterName=GlitterPower}
  ScalarParameter() {defaultValue=100, group=Glitter, parameterName=GlitterIntensity}
 ]
}
