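# Parent material for the IndoorLaceBeat character outfit, adding audio-reactive glow, animated circles and glitter between the shared character PreFX and PostFX functions.
# Frontend-loop audio amplitude provides a reactivity signal, alongside an optionally inverted mask for limiting the effect.
# Two spinning-UV functions use pivot textures, masks and flipbook IDs to build rotating symbol disks and noisy emissive circles.
# Reactive attributes blend in recolored cloth and time-of-day-adjusted emissive, then glitter is applied.
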Material /BRCosmetics/Characters/Player/Female/Medium/Bodies/F_MED_IndoorLaceBeat/Materials/M_IndoorLace_Parent
graph {
 properties {
  bAutomaticallySetUsageInEditor=false
  bHasPixelAnimation=true
  bUseMaterialAttributes=true
  bUsedWithMeshDeformer=true
  bUsedWithMorphTargets=true
  bUsedWithSkeletalMesh=true
  pixelDepthOffsetMode=PDOM_Legacy
  shadingModel=MSM_FromMaterialExpression
  usedShadingModels=Default Lit | Subsurface | Clear Coat | Subsurface Profile | Cloth
 }
 nodes [
  Constant()
  Constant() {r=1}
  n1 = TextureObject() {
   samplerType=SAMPLERTYPE_LinearColor
   texture=&/T_IndoorLace_Positions_1
  }
  n2 = MaterialFunctionCall!(
   PivotsMap=@n1
   Mask=NamedRerouteUsage()
   FixEdges=Constant() {r=1}
   BaseScale
   InvertEdgeFix=Constant()
   ScaleCorrection=Constant2Vector() {g=0.85, r=1}
   FlipbookIDs=NamedRerouteUsage()
  ) {materialFunction=&/MF_IndoorLaceBeat_SpinningUVs}
  n3 = Multiply(A=Multiply(A=@n2.UVs) {constB=-1}) {constB=2}
  n4 = TextureObject() {
   samplerType=SAMPLERTYPE_LinearColor
   texture=&/T_IndoorLace_Positions_2
  }
  n5 = MaterialFunctionCall(
   PivotsMap=@n4
   Mask=NamedRerouteUsage()
   FixEdges=Constant() {r=1}
   BaseScale=Constant() {r=3.5}
   InvertEdgeFix=Constant() {r=1}
   ScaleCorrection=Constant2Vector() {g=0.8, r=1}
   FlipbookIDs=NamedRerouteUsage()
  ) {materialFunction=&MF_IndoorLaceBeat_SpinningUVs}
  n6 = Multiply(A=Multiply(A=@n5.UVs) {constB=-1}) {constB=2}
  n7 = NamedRerouteUsage()
  LinearInterpolate(A=@n3, B=@n6, Alpha=@n7)
  n8 = MaterialFunctionCall() {materialFunction=&MF_CharacterParent_2023_PreFX}
  MaterialFunctionCall(Material=NamedRerouteUsage(), M=@n8.M, Pre Skinned Local Position=@n8.M, Distance Blend=@n8.M, Fresnel=@n8.M) {materialFunction=&MF_CharacterParent_2023_PostFX}
  n9 = TextureSample() {
   constCoordinate=1
   samplerSource=SSM_Clamp_WorldGroupSettings
   samplerType=SAMPLERTYPE_LinearColor
   texture=&/T_IndoorLace_Mask
  }
  NamedRerouteDeclaration(None=@n9.R) {name=Mask_1, nodeColor=<rgba 0.6641 1 0 1>}
  NamedRerouteDeclaration(None=LinearInterpolate(A=@n2.Result, B=@n5.Result, Alpha=@n7)) {name=RotatingDisksMask, nodeColor=<rgba 1 0.9375 0 1>}
  NamedRerouteDeclaration(None=@n9.G) {name=Mask_2, nodeColor=<rgba 0.9922 0 1 1>}
  n10 = NamedRerouteUsage()
  n11 = GetMaterialAttributes(None=@n10) {attributeGetTypes [#1, #2, #3]}
  n12 = Add(A=@n11.Emissive Color, B=Add(A=NamedRerouteUsage(), B=NamedRerouteUsage()))
  n13 = NamedRerouteUsage()
  n14 = Multiply(A=Add(A=@n12, B=Multiply(A=@n13) {constB=0.25})) {constB=0.75}
  n15 = TextureSample() {
   samplerSource=SSM_Wrap_WorldGroupSettings
   texture=/BRCosmetics/Characters/Player/Female/Medium/Bodies/F_MED_IndoorLaceBeat_Tick/Textures/T_IndoorLaceBeat_Tick_Body_D
  }
  n16 = VectorParameter() {group=IndoorLaceScore|Reactivity, parameterName=EmissiveInnerColor, defaultValue=<rgba 0.9478 0.8175 0.6121 1>}
  n17 = Multiply(
   A=@n16.RGB
   B=LinearInterpolate(
    A=ScalarParameter() {defaultValue=1, group=IndoorLaceScore|Reactivity, parameterName=Reactive Color Boost}
    B=ScalarParameter() {defaultValue=2, group=IndoorLaceScore|Reactivity, parameterName=Reactive Color Max}
    Alpha=NamedRerouteUsage()
   )
  )
  n18 = OneMinus(None=Fresnel(ExponentIn=ScalarParameter() {defaultValue=0.75, group=IndoorLaceScore|Reactivity, parameterName=Reactive Fresnel Exp}) {baseReflectFraction=0})
  n19 = NamedRerouteDeclaration(None=Multiply(A=LinearInterpolate(A=@n11.Base Color, B=@n15.RGB) {constAlpha=0.75}, B=Multiply(A=@n17, B=@n18))) {name=Coloured Reactivity - Animation, nodeColor=<rgba 0.1719 0 1 1>}
  n20 = SetMaterialAttributes!(
   MaterialAttributes=@n11.MaterialAttributes
   Emissive Color=MaterialFunctionCall(Color_Morning=@n12, Color_Day=@n12, Color_Evening=@n14, Color_Night=@n14) {materialFunction=&Color_Time_of_Day}.Result
   Base Color=NamedRerouteUsage()
   Opacity=Constant()
   Fuzz Color or Subsurface Color=Multiply(A=Multiply(A=@n11.Fuzz Color or Subsurface Color, B=@n19)) {constB=0.4}
  ) {attributeSetTypes [
    #1
    #2
    #4
    #3
   ]}
  n21 = NamedRerouteUsage()
  n22 = NamedRerouteUsage()
  NamedRerouteDeclaration(None=BlendMaterialAttributes(A=@n10, B=@n20, Alpha=Multiply(A=@n21, B=@n22))) {name=Reactive Color, nodeColor=<rgba 0.1484 1 0 1>}
  n23 = Ceil(None=NamedRerouteUsage())
  NamedRerouteDeclaration(None=LinearInterpolate(Alpha=@n23) {constA=4, constB=2.5}) {name=GlitterPower, nodeColor=<rgba 0.1953 1 0 1>}
  NamedRerouteDeclaration(None=@n8.M) {name=InMask, nodeColor=<rgba 0 0.1562 1 1>}
  NamedRerouteDeclaration(None=@n8.Result) {name=InMaterial, nodeColor=<rgba 0.6172 0 1 1>}
  NamedRerouteDeclaration(None=MaterialFunctionCall(Color=NamedRerouteUsage()) {materialFunction=&SplitComponents}.R) {name=AmbientOcclusion, nodeColor=<rgba 1 0 0.4453 1>}
  n24 = NamedRerouteUsage()
  n25 = NamedRerouteUsage()
  n26 = NamedRerouteUsage()
  n27 = Saturate(None=Power(Base=NamedRerouteUsage(), Exp=ScalarParameter() {defaultValue=2, group=IndoorLaceScore|Reactivity, parameterName=Glitter AO Mask - Exp}))
  NamedRerouteDeclaration(None=MaterialFunctionCall!(
    In=StaticSwitchParameter(True=@n24, False=@n25) {group=IndoorLaceScore|Reactivity, parameterName=IsReactive}
    SpeedMultiplier=NamedRerouteUsage()
    GlitterPower=NamedRerouteUsage()
    GlitterIntensity
    GlitterColor=NamedRerouteUsage()
    SkinFX_Mask=Multiply(A=@n26, B=@n27)
   ) {materialFunction=&/MF_IndoorLace_Glitter}.Result) {name=MaterialOut, nodeColor=<rgba 0 0.5547 1 1>}
  n28 = VectorParameter() {group=IndoorLaceScore|Reactivity, parameterName=GlitterColor - Reactive, defaultValue=<rgba 0.2552 0.1215 0 1>}
  NamedRerouteDeclaration(None=LinearInterpolate(
    A=VectorParameter() {group=IndoorLaceScore|Reactivity, parameterName=Glitter Color, defaultValue=<rgba 1 0.9767 0.8802 1>}.RGB
    B=@n28.RGB
    Alpha=@n23
   )) {name=GlitterColor, nodeColor=<rgba 0 0.7891 1 1>}
  n29 = TextureCoordinate() {coordinateIndex=1}
  n30 = Constant() {r=3}
  n31 = TextureSample(UVs=Panner(Coordinate=Multiply(A=@n29, B=@n30)) {speedX=0.1, speedY=0.1}) {samplerSource=SSM_Wrap_WorldGroupSettings, samplerType=SAMPLERTYPE_Grayscale, texture=&LowResBlurredNoise}
  n32 = TextureSample(UVs=@n3) {samplerSource=SSM_Wrap_WorldGroupSettings, samplerType=SAMPLERTYPE_Grayscale, texture=&LowResBlurredNoise}
  n33 = TextureSample(UVs=@n6) {samplerSource=SSM_Wrap_WorldGroupSettings, samplerType=SAMPLERTYPE_Grayscale, texture=&LowResBlurredNoise}
  NamedRerouteDeclaration(None=Multiply(
    A=NamedRerouteDeclaration(None=@n31.R) {name=LowFrequencyNoise, nodeColor=<rgba 0.4297 0 1 1>}
    B=LinearInterpolate(A=@n32.RGB, B=@n33.RGB, Alpha=@n7)
   )) {name=CircleNoise, nodeColor=<rgba 0.9453 0 1 1>}
  n34 = Constant3Vector() {constant=<rgba 0 0 1 0>}
  n35 = MaterialFunctionCall() {materialFunction=&CameraDirectionVector}
  n36 = Saturate(None=DotProduct(A=Transform(None=@n34) {transformSourceType=TRANSFORMSOURCE_Local}, B=Multiply(A=@n35.Result) {constB=-1}))
  NamedRerouteDeclaration(None=Saturate(None=Power(Base=@n36))) {name=Fresnel modulation, nodeColor=<rgba 1 0.1172 0 1>}
  n37 = NamedRerouteUsage()
  n38 = LinearInterpolate(Alpha=NamedRerouteUsage()) {constA=-0.02}
  n39 = ScalarParameter() {defaultValue=5, group=IndoorLaceScore|Animated Circles, parameterName=CircleNoiseIntensity}
  n40 = Multiply(A=@n37, B=Add(A=@n37, B=Multiply(A=@n38, B=@n39)))
  n41 = VectorParameter() {group=IndoorLaceScore|Animated Circles, parameterName=CircleEmissiveColor, defaultValue=<rgba 1 0.5536 0.0729 1>}
  n42 = Power(Base=NamedRerouteUsage(), Exp=ScalarParameter() {defaultValue=4, group=IndoorLaceScore|Animated Circles, parameterName=CircleReactivityExp}) {constExponent=0.8}
  n43 = LinearInterpolate(
   A=ScalarParameter() {defaultValue=0.75, group=IndoorLaceScore|Animated Circles, parameterName=CircleIntencity_Min}
   B=ScalarParameter() {defaultValue=8, group=IndoorLaceScore|Animated Circles, parameterName=CircleIntencity_Max}
   Alpha=@n42
  ) {constA=0.5, constB=5}
  n44 = Ceil(None=@n42)
  NamedRerouteDeclaration(None=Multiply(A=Multiply(A=@n40, B=@n41.RGB), B=Multiply(A=@n43, B=@n44))) {name=EmissiveCircles, nodeColor=<rgba 1 0.0703 0 1>}
  NamedRerouteDeclaration(None=AppendVector(A=@n8.Tech Art Mask, B=@n8.Tech Art Mask)) {name=FXMask2, nodeColor=<rgba 0.7109 0 1 1>}
  n45 = Power(Base=NamedRerouteUsage(), Exp=ScalarParameter() {defaultValue=0.6, group=IndoorLaceScore|Reactivity, parameterName=Reactivity Color - Exp}) {constExponent=0.4}
  n46 = LinearInterpolate(
   A=ScalarParameter() {defaultValue=0.1, group=IndoorLaceScore|Reactivity, parameterName=Fresnel Color - Exp}
   B=ScalarParameter() {defaultValue=1.5, group=IndoorLaceScore|Reactivity, parameterName=Fresnel Color - Exp Reactive}
   Alpha=@n45
  ) {constB=2}
  n47 = LinearInterpolate(A=VertexNormalWS(), B=PixelNormalWS())
  n48 = LinearInterpolate(
   A=OneMinus(None=MaterialFunctionCall(Exponent=@n46, Normal=@n47) {materialFunction=&AxisAlignedFresnel}.Fresnel)
   B=Fresnel(Normal=@n47) {baseReflectFraction=0, exponent=2}
   Alpha=NamedRerouteUsage()
  )
  n49 = NamedRerouteUsage()
  n50 = ScalarParameter() {defaultValue=2, group=IndoorLaceScore|Reactivity, parameterName=Noise Exp}
  n51 = Multiply(A=@n48, B=Add(A=@n48, B=Power(Base=@n49, Exp=@n50)))
  n52 = VectorParameter() {group=IndoorLaceScore|Reactivity, parameterName=Reactive Color, defaultValue=<rgba 0.6198 0.3578 0.0742 1>}
  n53 = ScalarParameter() {defaultValue=1.5, group=IndoorLaceScore|Reactivity, parameterName=Fresnel Color - Min}
  n54 = ScalarParameter() {defaultValue=3, group=IndoorLaceScore|Reactivity, parameterName=Fresnel Color - Max}
  NamedRerouteDeclaration(None=Multiply(A=Multiply(A=@n51, B=@n52.RGB), B=LinearInterpolate(A=@n53, B=@n54, Alpha=@n45))) {name=EmissiveFresnel, nodeColor=<rgba 0 1 0.2734 1>}
  n55 = Multiply(A=@n9.B) {constB=4}
  NamedRerouteDeclaration(None=Round(None=Subtract(A=@n55))) {name=Symbols_1_IDs, nodeColor=<rgba 1 0 0.2812 1>}
  n56 = Multiply(A=@n9.A) {constB=4}
  NamedRerouteDeclaration(None=Round(None=Subtract(A=@n56))) {name=Symbols_2_IDs, nodeColor=<rgba 0 0.6016 1 1>}
  n57 = ChannelMaskParameter(None=NamedRerouteUsage()) {group=IndoorLaceScore|Reactivity, maskChannel=Green, parameterName=ReactivityMask, defaultValue=<rgba 0 1 0 0>}
  NamedRerouteDeclaration(None=LinearInterpolate(A=@n57.R, B=OneMinus(None=@n57.R), Alpha=ScalarParameter() {defaultValue=1, parameterName=InvertReactivityMask})) {name=ReactivityMask, nodeColor=<rgba 0 0.2266 1 1>}
  n58 = Multiply(
   A=Power(
    Base=MaterialFunctionCall!(InAmplitude, InSpectrum, CustomFakeSpectrum, CustomFakeAmplitude) {materialFunction=/BRCosmetics/Global/MaterialFunctions/AudioReactivity/MF_AudioReactivity_FrontendLoop}.Amplitude
    Exp=ScalarParameter() {defaultValue=1, group=IndoorLaceScore|Reactivity, parameterName=Amplitude Power}
   )
   B=ScalarParameter() {defaultValue=1, group=IndoorLaceScore|Reactivity, parameterName=Amplitude Boost}
  )
  NamedRerouteDeclaration(None=LinearInterpolate(B=Saturate(None=@n58), Alpha=ScalarParameter() {defaultValue=1, group=IndoorLaceScore|Reactivity, parameterName=IsReactive})) {name=Audio Reactivity, nodeColor=<rgba 0 0.2969 1 1>}
  NamedRerouteDeclaration(None=LinearInterpolate(Alpha=@n23) {constA=1, constB=2}) {name=GlitterSpeed, nodeColor=<rgba 0.125 0 1 1>}
 ]
}
