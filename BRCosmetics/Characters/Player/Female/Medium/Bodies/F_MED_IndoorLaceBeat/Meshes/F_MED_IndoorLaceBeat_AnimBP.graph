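# Animation Blueprint for the female medium IndoorLaceBeat body cosmetic, following the attached parent mesh and adding secondary skirt motion.
# The copied parent pose runs through the outfit's Control Rig, fed crouching and not-crouching alphas set from bIsCrouching on each update.
# The cached rig result passes through a serial chain of identically tuned, root-relative, spring-driven Anim Dynamics nodes for the second and third bones of eight skirt strands.
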AnimBlueprint /BRCosmetics/Characters/Player/Female/Medium/Bodies/F_MED_IndoorLaceBeat/Meshes/F_MED_IndoorLaceBeat_AnimBP
target_skeleton=&/F_MED_IndoorLaceBeat_Skeleton
graphs [
 AnimGraph {nodes [
   Output Pose(Result=Use cached pose 'dynamicsCache'()) {layerGroup=None}
   n1 = Control Rig!(
    Source=Copy Pose From Mesh() {bUseAttachedParent=true}
    Alpha
    Is Crouching:double=Get Is Crouching Alpha() -> (Is Crouching Alpha:double)
    Is Not Crouching:double=Get Is Not Crouching Alpha() -> (Is Not Crouching Alpha:double)
   ) -> (Pose) {controlRigAssetReference {blueprintRigClass=&/F_MED_IndoorLaceBeat_CtrlRig.F_MED_IndoorLaceBeat_CtrlRig_C}}
   Save cached pose 'controlRigCache'(Pose=@n1) {cacheName=controlRigCache}
   n2 = Anim Dynamics - Bone: dyn_skirt_back_2(ComponentPose=Local To Component(LocalPose=Use cached pose 'controlRigCache'()), Alpha=1, ExternalForce="0,0,0") {
    angularBiasOverride=1
    angularDampingOverride=1
    angularSpringConstant=200
    bAngularSpring=true
    bOverrideAngularBias=true
    bOverrideAngularDamping=true
    bUseGravityOverride=true
    gravityScale=0.35
    simulationSpace=RootRelative
    alphaBoolBlend {blendOption=Linear}
    boundBone {boneName=dyn_skirt_back_2}
    componentAppliedLinearAccClamp=<3000 3000 3000>
    gravityOverride {z=-470}
    physicsBodyDefinitions [{
      boundBone {boneName=dyn_skirt_back_2}
      boxExtents=<20 20 20>
      constraintSetup {angularLimitsMax=<1 1 0>, angularLimitsMin=<-1 -1 -45>, angularTarget=<1 0 0>}
      localJointOffset=<7 0 0>
     }]
    simSpaceSettings {maxAngularAcceleration=45, maxAngularVelocity=8, simSpaceAngularAlpha=0.25}
   }
   n3 = Anim Dynamics - Bone: dyn_skirt_front_2(ComponentPose=@n2, Alpha=1, ExternalForce="0,0,0") {
    angularBiasOverride=1
    angularDampingOverride=1
    angularSpringConstant=200
    bAngularSpring=true
    bOverrideAngularBias=true
    bOverrideAngularDamping=true
    bUseGravityOverride=true
    gravityScale=0.35
    simulationSpace=RootRelative
    alphaBoolBlend {blendOption=Linear}
    boundBone {boneName=dyn_skirt_front_2}
    componentAppliedLinearAccClamp=<3000 3000 3000>
    gravityOverride {z=-470}
    physicsBodyDefinitions [{
      boundBone {boneName=dyn_skirt_front_2}
      boxExtents=<20 20 20>
      constraintSetup {angularLimitsMax=<1 1 0>, angularLimitsMin=<-1 -1 -45>, angularTarget=<1 0 0>}
      localJointOffset=<7 0 0>
     }]
    simSpaceSettings {maxAngularAcceleration=45, maxAngularVelocity=8, simSpaceAngularAlpha=0.25}
   }
   n4 = Anim Dynamics - Bone: dyn_skirt_side_l_2(ComponentPose=@n3, Alpha=1, ExternalForce="0,0,0") {
    angularBiasOverride=1
    angularDampingOverride=1
    angularSpringConstant=200
    bAngularSpring=true
    bOverrideAngularBias=true
    bOverrideAngularDamping=true
    bUseGravityOverride=true
    gravityScale=0.35
    simulationSpace=RootRelative
    alphaBoolBlend {blendOption=Linear}
    boundBone {boneName=dyn_skirt_side_l_2}
    componentAppliedLinearAccClamp=<3000 3000 3000>
    gravityOverride {z=-470}
    physicsBodyDefinitions [{
      boundBone {boneName=dyn_skirt_side_l_2}
      boxExtents=<20 20 20>
      constraintSetup {angularLimitsMax=<1 1 0>, angularLimitsMin=<-1 -1 -45>, angularTarget=<1 0 0>}
      localJointOffset=<7 0 0>
     }]
    simSpaceSettings {maxAngularAcceleration=45, maxAngularVelocity=8, simSpaceAngularAlpha=0.25}
   }
   n5 = Anim Dynamics - Bone: dyn_skirt_back_side_l_2(ComponentPose=@n4, Alpha=1, ExternalForce="0,0,0") {
    angularBiasOverride=1
    angularDampingOverride=1
    angularSpringConstant=200
    bAngularSpring=true
    bOverrideAngularBias=true
    bOverrideAngularDamping=true
    bUseGravityOverride=true
    gravityScale=0.35
    simulationSpace=RootRelative
    alphaBoolBlend {blendOption=Linear}
    boundBone {boneName=dyn_skirt_back_side_l_2}
    componentAppliedLinearAccClamp=<3000 3000 3000>
    gravityOverride {z=-470}
    physicsBodyDefinitions [{
      boundBone {boneName=dyn_skirt_back_side_l_2}
      boxExtents=<20 20 20>
      constraintSetup {angularLimitsMax=<1 1 0>, angularLimitsMin=<-1 -1 -45>, angularTarget=<1 0 0>}
      localJointOffset=<7 0 0>
     }]
    simSpaceSettings {maxAngularAcceleration=45, maxAngularVelocity=8, simSpaceAngularAlpha=0.25}
   }
   n6 = Anim Dynamics - Bone: dyn_skirt_front_l_2(ComponentPose=@n5, Alpha=1, ExternalForce="0,0,0") {
    angularBiasOverride=1
    angularDampingOverride=1
    angularSpringConstant=200
    bAngularSpring=true
    bOverrideAngularBias=true
    bOverrideAngularDamping=true
    bUseGravityOverride=true
    gravityScale=0.35
    simulationSpace=RootRelative
    alphaBoolBlend {blendOption=Linear}
    boundBone {boneName=dyn_skirt_front_l_2}
    componentAppliedLinearAccClamp=<3000 3000 3000>
    gravityOverride {z=-470}
    physicsBodyDefinitions [{
      boundBone {boneName=dyn_skirt_front_l_2}
      boxExtents=<20 20 20>
      constraintSetup {angularLimitsMax=<1 1 0>, angularLimitsMin=<-1 -1 -45>, angularTarget=<1 0 0>}
      localJointOffset=<7 0 0>
     }]
    simSpaceSettings {maxAngularAcceleration=45, maxAngularVelocity=8, simSpaceAngularAlpha=0.25}
   }
   n7 = Anim Dynamics - Bone: dyn_skirt_side_r_2(ComponentPose=@n6, Alpha=1, ExternalForce="0,0,0") {
    angularBiasOverride=1
    angularDampingOverride=1
    angularSpringConstant=200
    bAngularSpring=true
    bOverrideAngularBias=true
    bOverrideAngularDamping=true
    bUseGravityOverride=true
    gravityScale=0.35
    simulationSpace=RootRelative
    alphaBoolBlend {blendOption=Linear}
    boundBone {boneName=dyn_skirt_side_r_2}
    componentAppliedLinearAccClamp=<3000 3000 3000>
    gravityOverride {z=-470}
    physicsBodyDefinitions [{
      boundBone {boneName=dyn_skirt_side_r_2}
      boxExtents=<20 20 20>
      constraintSetup {angularLimitsMax=<1 1 0>, angularLimitsMin=<-1 -1 -45>, angularTarget=<1 0 0>}
      localJointOffset=<7 0 0>
     }]
    simSpaceSettings {maxAngularAcceleration=45, maxAngularVelocity=8, simSpaceAngularAlpha=0.25}
   }
   n8 = Anim Dynamics - Bone: dyn_skirt_back_side_r_2(ComponentPose=@n7, Alpha=1, ExternalForce="0,0,0") {
    angularBiasOverride=1
    angularDampingOverride=1
    angularSpringConstant=200
    bAngularSpring=true
    bOverrideAngularBias=true
    bOverrideAngularDamping=true
    bUseGravityOverride=true
    gravityScale=0.35
    simulationSpace=RootRelative
    alphaBoolBlend {blendOption=Linear}
    boundBone {boneName=dyn_skirt_back_side_r_2}
    componentAppliedLinearAccClamp=<3000 3000 3000>
    gravityOverride {z=-470}
    physicsBodyDefinitions [{
      boundBone {boneName=dyn_skirt_back_side_r_2}
      boxExtents=<20 20 20>
      constraintSetup {angularLimitsMax=<1 1 0>, angularLimitsMin=<-1 -1 -45>, angularTarget=<1 0 0>}
      localJointOffset=<7 0 0>
     }]
    simSpaceSettings {maxAngularAcceleration=45, maxAngularVelocity=8, simSpaceAngularAlpha=0.25}
   }
   n9 = Anim Dynamics - Bone: dyn_skirt_front_r_2(ComponentPose=@n8, Alpha=1, ExternalForce="0,0,0") {
    angularBiasOverride=1
    angularDampingOverride=1
    angularSpringConstant=200
    bAngularSpring=true
    bOverrideAngularBias=true
    bOverrideAngularDamping=true
    bUseGravityOverride=true
    gravityScale=0.35
    simulationSpace=RootRelative
    alphaBoolBlend {blendOption=Linear}
    boundBone {boneName=dyn_skirt_front_r_2}
    componentAppliedLinearAccClamp=<3000 3000 3000>
    gravityOverride {z=-470}
    physicsBodyDefinitions [{
      boundBone {boneName=dyn_skirt_front_r_2}
      boxExtents=<20 20 20>
      constraintSetup {angularLimitsMax=<1 1 0>, angularLimitsMin=<-1 -1 -45>, angularTarget=<1 0 0>}
      localJointOffset=<7 0 0>
     }]
    simSpaceSettings {maxAngularAcceleration=45, maxAngularVelocity=8, simSpaceAngularAlpha=0.25}
   }
   n10 = Anim Dynamics - Bone: dyn_skirt_back_3(ComponentPose=@n9, Alpha=1, ExternalForce="0,0,0") {
    angularBiasOverride=1
    angularDampingOverride=1
    angularSpringConstant=200
    bAngularSpring=true
    bOverrideAngularBias=true
    bOverrideAngularDamping=true
    bUseGravityOverride=true
    gravityScale=0.35
    simulationSpace=RootRelative
    alphaBoolBlend {blendOption=Linear}
    boundBone {boneName=dyn_skirt_back_3}
    componentAppliedLinearAccClamp=<3000 3000 3000>
    gravityOverride {z=-470}
    physicsBodyDefinitions [{
      boundBone {boneName=dyn_skirt_back_3}
      boxExtents=<20 20 20>
      constraintSetup {angularLimitsMax=<1 1 0>, angularLimitsMin=<-1 -1 -45>, angularTarget=<1 0 0>}
      localJointOffset=<7 0 0>
     }]
    simSpaceSettings {maxAngularAcceleration=45, maxAngularVelocity=8, simSpaceAngularAlpha=0.25}
   }
   n11 = Anim Dynamics - Bone: dyn_skirt_front_3(ComponentPose=@n10, Alpha=1, ExternalForce="0,0,0") {
    angularBiasOverride=1
    angularDampingOverride=1
    angularSpringConstant=200
    bAngularSpring=true
    bOverrideAngularBias=true
    bOverrideAngularDamping=true
    bUseGravityOverride=true
    gravityScale=0.35
    simulationSpace=RootRelative
    alphaBoolBlend {blendOption=Linear}
    boundBone {boneName=dyn_skirt_front_3}
    componentAppliedLinearAccClamp=<3000 3000 3000>
    gravityOverride {z=-470}
    physicsBodyDefinitions [{
      boundBone {boneName=dyn_skirt_front_3}
      boxExtents=<20 20 20>
      constraintSetup {angularLimitsMax=<1 1 0>, angularLimitsMin=<-1 -1 -45>, angularTarget=<1 0 0>}
      localJointOffset=<7 0 0>
     }]
    simSpaceSettings {maxAngularAcceleration=45, maxAngularVelocity=8, simSpaceAngularAlpha=0.25}
   }
   n12 = Anim Dynamics - Bone: dyn_skirt_side_l_3(ComponentPose=@n11, Alpha=1, ExternalForce="0,0,0") {
    angularBiasOverride=1
    angularDampingOverride=1
    angularSpringConstant=200
    bAngularSpring=true
    bOverrideAngularBias=true
    bOverrideAngularDamping=true
    bUseGravityOverride=true
    gravityScale=0.35
    simulationSpace=RootRelative
    alphaBoolBlend {blendOption=Linear}
    boundBone {boneName=dyn_skirt_side_l_3}
    componentAppliedLinearAccClamp=<3000 3000 3000>
    gravityOverride {z=-470}
    physicsBodyDefinitions [{
      boundBone {boneName=dyn_skirt_side_l_3}
      boxExtents=<20 20 20>
      constraintSetup {angularLimitsMax=<1 1 0>, angularLimitsMin=<-1 -1 -45>, angularTarget=<1 0 0>}
      localJointOffset=<7 0 0>
     }]
    simSpaceSettings {maxAngularAcceleration=45, maxAngularVelocity=8, simSpaceAngularAlpha=0.25}
   }
   n13 = Anim Dynamics - Bone: dyn_skirt_back_side_l_3(ComponentPose=@n12, Alpha=1, ExternalForce="0,0,0") {
    angularBiasOverride=1
    angularDampingOverride=1
    angularSpringConstant=200
    bAngularSpring=true
    bOverrideAngularBias=true
    bOverrideAngularDamping=true
    bUseGravityOverride=true
    gravityScale=0.35
    simulationSpace=RootRelative
    alphaBoolBlend {blendOption=Linear}
    boundBone {boneName=dyn_skirt_back_side_l_3}
    componentAppliedLinearAccClamp=<3000 3000 3000>
    gravityOverride {z=-470}
    physicsBodyDefinitions [{
      boundBone {boneName=dyn_skirt_back_side_l_3}
      boxExtents=<20 20 20>
      constraintSetup {angularLimitsMax=<1 1 0>, angularLimitsMin=<-1 -1 -45>, angularTarget=<1 0 0>}
      localJointOffset=<7 0 0>
     }]
    simSpaceSettings {maxAngularAcceleration=45, maxAngularVelocity=8, simSpaceAngularAlpha=0.25}
   }
   n14 = Anim Dynamics - Bone: dyn_skirt_front_l_3(ComponentPose=@n13, Alpha=1, ExternalForce="0,0,0") {
    angularBiasOverride=1
    angularDampingOverride=1
    angularSpringConstant=200
    bAngularSpring=true
    bOverrideAngularBias=true
    bOverrideAngularDamping=true
    bUseGravityOverride=true
    gravityScale=0.35
    simulationSpace=RootRelative
    alphaBoolBlend {blendOption=Linear}
    boundBone {boneName=dyn_skirt_front_l_3}
    componentAppliedLinearAccClamp=<3000 3000 3000>
    gravityOverride {z=-470}
    physicsBodyDefinitions [{
      boundBone {boneName=dyn_skirt_front_l_3}
      boxExtents=<20 20 20>
      constraintSetup {angularLimitsMax=<1 1 0>, angularLimitsMin=<-1 -1 -45>, angularTarget=<1 0 0>}
      localJointOffset=<7 0 0>
     }]
    simSpaceSettings {maxAngularAcceleration=45, maxAngularVelocity=8, simSpaceAngularAlpha=0.25}
   }
   n15 = Anim Dynamics - Bone: dyn_skirt_side_r_3(ComponentPose=@n14, Alpha=1, ExternalForce="0,0,0") {
    angularBiasOverride=1
    angularDampingOverride=1
    angularSpringConstant=200
    bAngularSpring=true
    bOverrideAngularBias=true
    bOverrideAngularDamping=true
    bUseGravityOverride=true
    gravityScale=0.35
    simulationSpace=RootRelative
    alphaBoolBlend {blendOption=Linear}
    boundBone {boneName=dyn_skirt_side_r_3}
    componentAppliedLinearAccClamp=<3000 3000 3000>
    gravityOverride {z=-470}
    physicsBodyDefinitions [{
      boundBone {boneName=dyn_skirt_side_r_3}
      boxExtents=<20 20 20>
      constraintSetup {angularLimitsMax=<1 1 0>, angularLimitsMin=<-1 -1 -45>, angularTarget=<1 0 0>}
      localJointOffset=<7 0 0>
     }]
    simSpaceSettings {maxAngularAcceleration=45, maxAngularVelocity=8, simSpaceAngularAlpha=0.25}
   }
   n16 = Anim Dynamics - Bone: dyn_skirt_back_side_r_3(ComponentPose=@n15, Alpha=1, ExternalForce="0,0,0") {
    angularBiasOverride=1
    angularDampingOverride=1
    angularSpringConstant=200
    bAngularSpring=true
    bOverrideAngularBias=true
    bOverrideAngularDamping=true
    bUseGravityOverride=true
    gravityScale=0.35
    simulationSpace=RootRelative
    alphaBoolBlend {blendOption=Linear}
    boundBone {boneName=dyn_skirt_back_side_r_3}
    componentAppliedLinearAccClamp=<3000 3000 3000>
    gravityOverride {z=-470}
    physicsBodyDefinitions [{
      boundBone {boneName=dyn_skirt_back_side_r_3}
      boxExtents=<20 20 20>
      constraintSetup {angularLimitsMax=<1 1 0>, angularLimitsMin=<-1 -1 -45>, angularTarget=<1 0 0>}
      localJointOffset=<7 0 0>
     }]
    simSpaceSettings {maxAngularAcceleration=45, maxAngularVelocity=8, simSpaceAngularAlpha=0.25}
   }
   n17 = Anim Dynamics - Bone: dyn_skirt_front_r_3(ComponentPose=@n16, Alpha=1, ExternalForce="0,0,0") {
    angularBiasOverride=1
    angularDampingOverride=1
    angularSpringConstant=200
    bAngularSpring=true
    bOverrideAngularBias=true
    bOverrideAngularDamping=true
    bUseGravityOverride=true
    gravityScale=0.35
    simulationSpace=RootRelative
    alphaBoolBlend {blendOption=Linear}
    boundBone {boneName=dyn_skirt_front_r_3}
    componentAppliedLinearAccClamp=<3000 3000 3000>
    gravityOverride {z=-470}
    physicsBodyDefinitions [{
      boundBone {boneName=dyn_skirt_front_r_3}
      boxExtents=<20 20 20>
      constraintSetup {angularLimitsMax=<1 1 0>, angularLimitsMin=<-1 -1 -45>, angularTarget=<1 0 0>}
      localJointOffset=<7 0 0>
     }]
    simSpaceSettings {maxAngularAcceleration=45, maxAngularVelocity=8, simSpaceAngularAlpha=0.25}
   }
   Save cached pose 'dynamicsCache'(Pose=Component To Local(ComponentPose=@n17)) {cacheName=dynamicsCache}
  ]}
 EventGraph {nodes [
   Event BlueprintUpdateAnimation()
   n1 = Get bIsCrouching()
   > Set Is Crouching Alpha(Is Crouching Alpha:double="To Float (Boolean)"(InBool=@n1)) -> (Output_Get:double)
   > Set Is Not Crouching Alpha(Is Not Crouching Alpha:double="To Float (Boolean)"(InBool=NOT Boolean(A=@n1))) -> (Output_Get:double)
  ]}
]
variables [
 Is Crouching Alpha {type=number}
 Is Not Crouching Alpha {type=number}
]
