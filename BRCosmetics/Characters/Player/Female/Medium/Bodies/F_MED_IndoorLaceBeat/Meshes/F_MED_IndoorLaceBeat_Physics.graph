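# Physics asset for the female medium IndoorLaceBeat body cosmetic, set up for a dynamic skirt.
# Eight skirt bone chains use capsules on their lower segments, while base and root bodies, plus the upper front segments, have zero-size shapes.
# Twist-locked constraints link segments within each chain, starting one segment lower on the three front chains, with no links between chains.
# Thigh, calf and foot capsules plus a huge floor capsule provide collision shapes.
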PhysicsAsset /BRCosmetics/Characters/Player/Female/Medium/Bodies/F_MED_IndoorLaceBeat/Meshes/F_MED_IndoorLaceBeat_Physics
bodies {
 base_skirt_back_0 [{shapeType=Capsule, length=0, radius=0, shapeName=base_skirt_back_0_capsule, rotation=<rot 0 -90 0>}]
 base_skirt_back_side_l_0 [{shapeType=Capsule, length=0, radius=0, shapeName=base_skirt_back_side_l_0_capsule, rotation=<rot 0 -90 0>}]
 base_skirt_back_side_r_0 [{shapeType=Capsule, length=0, radius=0, shapeName=base_skirt_back_side_r_0_capsule, rotation=<rot 0 -90 0>}]
 base_skirt_front_0 [{shapeType=Capsule, length=0, radius=0, shapeName=base_skirt_front_0_capsule, rotation=<rot 0 -90 0>}]
 base_skirt_front_l_0 [{shapeType=Capsule, length=0, radius=0, shapeName=base_skirt_front_l_0_capsule, rotation=<rot 0 -90 0>}]
 base_skirt_front_r_0 [{shapeType=Capsule, length=0, radius=0, shapeName=base_skirt_front_r_0_capsule, rotation=<rot 0 -90 0>}]
 base_skirt_side_l_0 [{shapeType=Capsule, length=0, radius=0, shapeName=base_skirt_side_l_0_capsule, rotation=<rot 0 -90 0>}]
 base_skirt_side_r_0 [{shapeType=Capsule, length=0, radius=0, shapeName=base_skirt_side_r_0_capsule, rotation=<rot 0 -90 0>}]
 calf_l [{shapeType=Capsule, length=52.8464, radius=6.1535, shapeName=calf_l_capsule, center=<-15 -2 0>, rotation=<rot 90 0 0>}]
 calf_r [{shapeType=Capsule, length=52.8464, radius=6.1535, shapeName=calf_l_capsule, center=<15 2 0>, rotation=<rot 90 0 0>}]
 dyn_skirt_back_0 [{shapeType=Capsule, length=0, radius=0, shapeName=dyn_skirt_back_0_capsule, center=<5 -8.2 0>}]
 dyn_skirt_back_1 [{shapeType=Capsule, length=34, radius=9, shapeName=dyn_skirt_back_1_capsule, center=<10 -8.2 0>}]
 dyn_skirt_back_2 [{shapeType=Capsule, length=34, radius=9, shapeName=dyn_skirt_back_2_capsule, center=<13 0 0>}]
 dyn_skirt_back_3 [{shapeType=Capsule, length=34, radius=9, shapeName=dyn_skirt_back_3_capsule, center=<15 0 0>}]
 dyn_skirt_back_side_l_0 [{shapeType=Capsule, length=0, radius=0, shapeName=dyn_skirt_back_side_l_0_capsule, center=<5 -8.2 0>}]
 dyn_skirt_back_side_l_1 [{shapeType=Capsule, length=34, radius=9, shapeName=dyn_skirt_back_side_l_1_capsule, center=<10 -8.2 0>}]
 dyn_skirt_back_side_l_2 [{shapeType=Capsule, length=34, radius=9, shapeName=dyn_skirt_back_side_l_2_capsule, center=<13 -5 0>}]
 dyn_skirt_back_side_l_3 [{shapeType=Capsule, length=34, radius=9, shapeName=dyn_skirt_back_side_l_3_capsule, center=<15 -1 0>}]
 dyn_skirt_back_side_r_0 [{shapeType=Capsule, length=0, radius=0, shapeName=dyn_skirt_back_side_r_0_capsule, center=<5 -8.2 0>}]
 dyn_skirt_back_side_r_1 [{shapeType=Capsule, length=34, radius=9, shapeName=dyn_skirt_back_side_r_1_capsule, center=<10 -8.2 0>}]
 dyn_skirt_back_side_r_2 [{shapeType=Capsule, length=34, radius=9, shapeName=dyn_skirt_back_side_r_2_capsule, center=<13 -5 0>}]
 dyn_skirt_back_side_r_3 [{shapeType=Capsule, length=34, radius=9, shapeName=dyn_skirt_back_side_r_3_capsule, center=<15 -1 0>}]
 dyn_skirt_front_0 [{shapeType=Capsule, length=0, radius=0, shapeName=dyn_skirt_front_0_capsule, center=<5 -8.2 0>}]
 dyn_skirt_front_1 [{shapeType=Capsule, length=0, radius=0, shapeName=dyn_skirt_front_1_capsule, center=<10 -8.2 0>}]
 dyn_skirt_front_2 [{shapeType=Capsule, length=34, radius=9, shapeName=dyn_skirt_front_2_capsule, center=<13 0 0>}]
 dyn_skirt_front_3 [{shapeType=Capsule, length=34, radius=9, shapeName=dyn_skirt_front_3_capsule, center=<15 0 0>}]
 dyn_skirt_front_l_0 [{shapeType=Capsule, length=0, radius=0, shapeName=dyn_skirt_front_l_0_capsule, center=<5 -8.3 0>}]
 dyn_skirt_front_l_1 [{shapeType=Capsule, length=0, radius=0, shapeName=dyn_skirt_front_l_1_capsule, center=<10 -8.7 0>}]
 dyn_skirt_front_l_2 [{shapeType=Capsule, length=34, radius=9, shapeName=dyn_skirt_front_l_2_capsule, center=<13 -4 0>}]
 dyn_skirt_front_l_3 [{shapeType=Capsule, length=34, radius=9, shapeName=dyn_skirt_front_l_3_capsule, center=<15 0 0>}]
 dyn_skirt_front_r_0 [{shapeType=Capsule, length=0, radius=0, shapeName=dyn_skirt_front_r_0_capsule, center=<5 -8.3 0>}]
 dyn_skirt_front_r_1 [{shapeType=Capsule, length=0, radius=0, shapeName=dyn_skirt_front_r_1_capsule, center=<10 -8.7 0>}]
 dyn_skirt_front_r_2 [{shapeType=Capsule, length=34, radius=9, shapeName=dyn_skirt_front_r_2_capsule, center=<13 -4 0>}]
 dyn_skirt_front_r_3 [{shapeType=Capsule, length=34, radius=9, shapeName=dyn_skirt_front_r_3_capsule, center=<15 0 0>}]
 dyn_skirt_side_l_0 [{shapeType=Capsule, length=0, radius=0, shapeName=dyn_skirt_side_l_0_capsule, center=<5 -9.2 0>}]
 dyn_skirt_side_l_1 [{shapeType=Capsule, length=34, radius=9, shapeName=dyn_skirt_side_l_1_capsule, center=<10 -8.5 0>, rotation=<rot 0.0001 0 7.7827>}]
 dyn_skirt_side_l_2 [{shapeType=Capsule, length=34, radius=9, shapeName=dyn_skirt_side_l_2_capsule, center=<13 -6 0>, rotation=<rot 0.0001 0 7.7827>}]
 dyn_skirt_side_l_3 [{shapeType=Capsule, length=34, radius=9, shapeName=dyn_skirt_side_l_3_capsule, center=<15 -2.5 0>, rotation=<rot 0.0001 0 7.7827>}]
 dyn_skirt_side_r_0 [{shapeType=Capsule, length=0, radius=0, shapeName=dyn_skirt_side_r_0_capsule, center=<5 -9.2 0>}]
 dyn_skirt_side_r_1 [{shapeType=Capsule, length=34, radius=9, shapeName=dyn_skirt_side_r_1_capsule, center=<10 -10 0>}]
 dyn_skirt_side_r_2 [{shapeType=Capsule, length=34, radius=9, shapeName=dyn_skirt_side_r_2_capsule, center=<13 -8.9 0>}]
 dyn_skirt_side_r_3 [{shapeType=Capsule, length=34, radius=9, shapeName=dyn_skirt_side_r_3_capsule, center=<15 -2.5 0>}]
 floor [{shapeType=Capsule, length=10, radius=510, shapeName=floor_capsule, center=<0 0 -520>}]
 foot_l [{shapeType=Capsule, length=15.4733, radius=9.8213, shapeName=foot_l_capsule, center=<-4.5 6 0>, rotation=<rot 0 0 90>}]
 foot_r [{shapeType=Capsule, length=15.4733, radius=9.8213, shapeName=foot_l_capsule, center=<4.5 -6 0>, rotation=<rot 0 0 90>}]
 thigh_l [{shapeType=Capsule, length=39.5, radius=7.7, shapeName=thigh_l_capsule, center=<-25 0 0>, rotation=<rot 90 5 0>}]
 thigh_r [{shapeType=Capsule, length=39.5, radius=7.7, shapeName=thigh_l_capsule, center=<25 0 0>, rotation=<rot 90 5 0>}]
}
constraints [
 {bone1Name=dyn_skirt_back_1, bone2Name=dyn_skirt_back_0, swing1LimitDegrees=15, swing2LimitDegrees=0, twistLimitDegrees=0, twistMotion=Locked}
 {bone1Name=dyn_skirt_back_2, bone2Name=dyn_skirt_back_1, swing1LimitDegrees=15, swing2LimitDegrees=0, twistLimitDegrees=0, twistMotion=Locked}
 {bone1Name=dyn_skirt_back_3, bone2Name=dyn_skirt_back_2, swing1LimitDegrees=15, swing2LimitDegrees=0, twistLimitDegrees=0, twistMotion=Locked}
 {
  bone1Name=dyn_skirt_back_side_l_1
  bone2Name=dyn_skirt_back_side_l_0
  swing1LimitDegrees=15
  swing2LimitDegrees=15
  twistLimitDegrees=0
  twistMotion=Locked
 }
 {
  bone1Name=dyn_skirt_back_side_l_2
  bone2Name=dyn_skirt_back_side_l_1
  swing1LimitDegrees=15
  swing2LimitDegrees=15
  twistLimitDegrees=0
  twistMotion=Locked
 }
 {
  bone1Name=dyn_skirt_back_side_l_3
  bone2Name=dyn_skirt_back_side_l_2
  swing1LimitDegrees=15
  swing2LimitDegrees=15
  twistLimitDegrees=0
  twistMotion=Locked
 }
 {
  bone1Name=dyn_skirt_back_side_r_1
  bone2Name=dyn_skirt_back_side_r_0
  swing1LimitDegrees=15
  swing2LimitDegrees=15
  twistLimitDegrees=0
  twistMotion=Locked
 }
 {
  bone1Name=dyn_skirt_back_side_r_2
  bone2Name=dyn_skirt_back_side_r_1
  swing1LimitDegrees=15
  swing2LimitDegrees=15
  twistLimitDegrees=0
  twistMotion=Locked
 }
 {
  bone1Name=dyn_skirt_back_side_r_3
  bone2Name=dyn_skirt_back_side_r_2
  swing1LimitDegrees=15
  swing2LimitDegrees=15
  twistLimitDegrees=0
  twistMotion=Locked
 }
 {
  bone1Name=dyn_skirt_front_2
  bone2Name=dyn_skirt_front_1
  swing1LimitDegrees=15
  swing2LimitDegrees=0
  twistLimitDegrees=0
  twistMotion=Locked
 }
 {
  bone1Name=dyn_skirt_front_3
  bone2Name=dyn_skirt_front_2
  swing1LimitDegrees=15
  swing2LimitDegrees=0
  twistLimitDegrees=0
  twistMotion=Locked
 }
 {
  bone1Name=dyn_skirt_front_l_2
  bone2Name=dyn_skirt_front_l_1
  swing1LimitDegrees=15
  swing2LimitDegrees=0
  twistLimitDegrees=0
  twistMotion=Locked
 }
 {
  bone1Name=dyn_skirt_front_l_3
  bone2Name=dyn_skirt_front_l_2
  swing1LimitDegrees=15
  swing2LimitDegrees=0
  twistLimitDegrees=0
  twistMotion=Locked
 }
 {
  bone1Name=dyn_skirt_front_r_2
  bone2Name=dyn_skirt_front_r_1
  swing1LimitDegrees=15
  swing2LimitDegrees=15
  twistLimitDegrees=0
  twistMotion=Locked
 }
 {
  bone1Name=dyn_skirt_front_r_3
  bone2Name=dyn_skirt_front_r_2
  swing1LimitDegrees=15
  swing2LimitDegrees=15
  twistLimitDegrees=0
  twistMotion=Locked
 }
 {
  bone1Name=dyn_skirt_side_l_1
  bone2Name=dyn_skirt_side_l_0
  swing1LimitDegrees=15
  swing2LimitDegrees=15
  twistLimitDegrees=0
  twistMotion=Locked
 }
 {
  bone1Name=dyn_skirt_side_l_2
  bone2Name=dyn_skirt_side_l_1
  swing1LimitDegrees=15
  swing2LimitDegrees=15
  twistLimitDegrees=0
  twistMotion=Locked
 }
 {
  bone1Name=dyn_skirt_side_l_3
  bone2Name=dyn_skirt_side_l_2
  swing1LimitDegrees=15
  swing2LimitDegrees=15
  twistLimitDegrees=0
  twistMotion=Locked
 }
 {
  bone1Name=dyn_skirt_side_r_1
  bone2Name=dyn_skirt_side_r_0
  swing1LimitDegrees=15
  swing2LimitDegrees=15
  twistLimitDegrees=0
  twistMotion=Locked
 }
 {
  bone1Name=dyn_skirt_side_r_2
  bone2Name=dyn_skirt_side_r_1
  swing1LimitDegrees=15
  swing2LimitDegrees=15
  twistLimitDegrees=0
  twistMotion=Locked
 }
 {
  bone1Name=dyn_skirt_side_r_3
  bone2Name=dyn_skirt_side_r_2
  swing1LimitDegrees=15
  swing2LimitDegrees=15
  twistLimitDegrees=0
  twistMotion=Locked
 }
]
