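# Physics asset for the IndoorLaceBeat female FaceAcc Backpack mesh part, set up for dynamic hair with capsule bodies.
# Center, left and right five-bone chains and a four-bone front-right chain each hang from an offset_hair anchor.
# Side and center roots barely swing, lower joints generally loosen, and twist is locked except at the front tip.
# Spine, neck, clavicle and upper-arm capsules provide body shapes, including one large offset spine_05 capsule.
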PhysicsAsset /BRCosmetics/Characters/Player/Female/Medium/Bodies/F_MED_IndoorLaceBeat/Meshes/Parts/F_MED_IndoorLaceBeat_FaceAcc_Backpack_Physics
bodies {
 clavicle_l [{
   shapeType=Capsule
   length=7.4228
   radius=5.7886
   shapeName=clavicle_l_capsule
   center=<6.3703 -0.6819 -1.6719>
   rotation=<rot 90 -179.9999 0>
  }]
 clavicle_r [{
   shapeType=Capsule
   length=7.2873
   radius=5.4241
   shapeName=clavicle_r_capsule
   center=<-6.1862 1.4322 1.2036>
   rotation=<rot 87.5685 -179.9942 -179.9944>
  }]
 dyn_hair_01_cr [{shapeType=Capsule, length=6.9364, radius=2, shapeName=dyn_hair_01_rt_capsule, center=<5 0 0>, rotation=<rot 81.7452 0 0>}]
 dyn_hair_01_lf [{
   shapeType=Capsule
   length=18.334
   radius=3.673
   shapeName=dyn_hair_01_rt_capsule
   center=<13.6229 3.0355 2.3879>
   rotation=<rot 84.2217 179.9994 179.9995>
  }]
 dyn_hair_01_rt [{
   shapeType=Capsule
   length=18.1145
   radius=2.1222
   shapeName=dyn_hair_01_rt_capsule
   center=<13.7563 0 0.4645>
   rotation=<rot 76.8299 180 180>
  }]
 dyn_hair_02_cr [{shapeType=Capsule, length=5.8123, radius=1.941, shapeName=dyn_hair_02_rt_capsule, center=<2 0 1.0253>, rotation=<rot 83.9553 180 180>}]
 dyn_hair_02_lf [{
   shapeType=Capsule
   length=6.9023
   radius=3.2418
   shapeName=dyn_hair_02_rt_capsule
   center=<5 4.649 -0.415>
   rotation=<rot 71.5532 179.9996 179.9997>
  }]
 dyn_hair_02_rt [{shapeType=Capsule, length=5.82, radius=2.816, shapeName=dyn_hair_02_rt_capsule, center=<6.2045 0 -0.3549>, rotation=<rot 78.449 180 180>}]
 dyn_hair_03_cr [{shapeType=Capsule, length=5.9471, radius=2.0529, shapeName=dyn_hair_03_cr_capsule, center=<5.077 0 0.9926>, rotation=<rot 87.9195 0 0>}]
 dyn_hair_03_lf [{
   shapeType=Capsule
   length=5.8131
   radius=2.1869
   shapeName=dyn_hair_03_lf_capsule
   center=<9.0139 4.6407 -1.2551>
   rotation=<rot 70.4787 0 0>
  }]
 dyn_hair_03_rt [{shapeType=Capsule, length=5.342, radius=2.8685, shapeName=dyn_hair_03_rt_capsule, center=<3.6215 0 0.7576>, rotation=<rot 88.1838 180 180>}]
 dyn_hair_04_cr [{shapeType=Capsule, length=8.7239, radius=2.1494, shapeName=dyn_hair_04_cr_capsule, center=<7.4287 0 0.8046>, rotation=<rot 88.5975 0 0>}]
 dyn_hair_04_lf [{
   shapeType=Capsule
   length=5.3061
   radius=2.0852
   shapeName=dyn_hair_04_lf_capsule
   center=<11.9747 4.0936 -0.6028>
   rotation=<rot 87.5939 0 0>
  }]
 dyn_hair_04_rt [{
   shapeType=Capsule
   length=8.0312
   radius=2.8421
   shapeName=dyn_hair_04_rt_capsule
   center=<7.3479 0 0.12>
   rotation=<rot 89.3926 0.0003 0.0003>
  }]
 dyn_hair_05_cr [{shapeType=Capsule, length=3, radius=4, shapeName=dyn_hair_05_cr_capsule, center=<5 0 1.2654>, rotation=<rot 79.4625 180 -180>}]
 dyn_hair_05_lf [{shapeType=Capsule, length=3.8098, radius=3.1902, shapeName=dyn_hair_05_lf_capsule, center=<4.9006 0 0.5958>, rotation=<rot 80.5288 180 180>}]
 dyn_hair_05_rt [{shapeType=Capsule, length=5.3607, radius=2.6393, shapeName=dyn_hair_05_rt_capsule, center=<4.8976 0 0.494>, rotation=<rot 78.2878 180 180>}]
 dyn_hair_front_01_rt [{
   shapeType=Capsule
   length=7.334
   radius=3.173
   shapeName=dyn_hair_01_rt_capsule
   center=<11.97 0 1.0112>
   rotation=<rot 49.9452 179.9998 179.9999>
  }]
 dyn_hair_front_02_rt [{
   shapeType=Capsule
   length=5.5171
   radius=2.5108
   shapeName=dyn_hair_front_02_lf_capsule
   center=<3.2614 0 0.6618>
   rotation=<rot 79.9996 180 0>
  }]
 dyn_hair_front_03_rt [{
   shapeType=Capsule
   length=3.8667
   radius=4.4615
   shapeName=dyn_hair_front_03_lf_capsule
   center=<4.4266 -0.0939 1.7459>
   rotation=<rot 85.6531 -179.9984 0>
  }]
 dyn_hair_front_04_rt [{
   shapeType=Capsule
   length=2.8667
   radius=2.9615
   shapeName=dyn_hair_front_03_lf_capsule
   center=<3.8924 -0.7889 1.1291>
   rotation=<rot 85.6532 -179.9984 -90>
  }]
 neck_01 [{shapeType=Capsule, length=8, radius=4, shapeName=neck_01_capsule, center=<0 1.0001 0>, rotation=<rot 0 -90.0001 -90>}]
 neck_02 [{shapeType=Capsule, length=38.3892, radius=12.3653, shapeName=neck_02_capsule, center=<-1.5348 14.5958 0>, rotation=<rot 0 -90 0>}]
 offset_hair_cr [{shapeType=Capsule, length=0.505, radius=0.505, shapeName=offset_hair_cr_capsule, rotation=<rot 0 -90 0>}]
 offset_hair_front_rt [{shapeType=Capsule, length=0.505, radius=0.505, shapeName=offset_hair_front_rt_capsule, rotation=<rot 0 -90 0>}]
 offset_hair_lf [{shapeType=Capsule, length=0.505, radius=0.505, shapeName=offset_hair_lf_capsule, rotation=<rot 0 -90 0>}]
 offset_hair_rt [{shapeType=Capsule, length=0.505, radius=0.505, shapeName=offset_hair_rt_capsule, rotation=<rot 0 -90 0>}]
 spine_04 [
  {shapeType=Capsule, length=16.0862, radius=8.9138, shapeName=spine_05_capsule, center=<-6.204 2.8253 0>, rotation=<rot 0 -90 0>}
  {shapeType=Capsule, length=18.2664, radius=7.9449, shapeName=spine_04_capsule, center=<1.9344 0.3306 0>}
  {shapeType=Capsule, length=7.8481, radius=5.8535, shapeName=spine_04_capsule, center=<0.6187 7.2479 0>}
 ]
 spine_05 [
  {shapeType=Capsule, length=18.4157, radius=7.0551, shapeName=spine_05_capsule, center=<1.9647 -1.4698 0>, rotation=<rot 0 -90 0>}
  {shapeType=Capsule, length=18, radius=33, shapeName=spine_05_capsule, center=<9.8726 -48.2846 0>}
  {shapeType=Capsule, length=14, radius=6.5, shapeName=spine_05_capsule, center=<-2.0386 2.5338 0>}
 ]
 upperarm_l [{
   shapeType=Capsule
   length=26.8995
   radius=5.4721
   shapeName=upperarm_l_capsule
   center=<12.95 -0.7645 -0.5337>
   rotation=<rot -0.0001 -89.9999 -92.7028>
  }]
 upperarm_r [{shapeType=Capsule, length=25.0453, radius=4.0231, shapeName=upperarm_l_capsule, center=<-13.9807 0.2449 0.5342>, rotation=<rot 0 -90 -90>}]
}
constraints [
 {
  bone1Name=dyn_hair_01_cr
  bone2Name=offset_hair_cr
  swing1LimitDegrees=0
  swing1Motion=Locked
  swing2LimitDegrees=3
  twistLimitDegrees=0
  twistMotion=Locked
 }
 {bone1Name=dyn_hair_03_cr, bone2Name=dyn_hair_02_cr, swing1LimitDegrees=10, swing2LimitDegrees=30, twistLimitDegrees=0, twistMotion=Locked}
 {bone1Name=dyn_hair_04_cr, bone2Name=dyn_hair_03_cr, swing1LimitDegrees=10, swing2LimitDegrees=20, twistLimitDegrees=0, twistMotion=Locked}
 {bone1Name=dyn_hair_05_cr, bone2Name=dyn_hair_04_cr, swing1LimitDegrees=10, swing2LimitDegrees=40, twistLimitDegrees=45, twistMotion=Locked}
 {
  bone1Name=dyn_hair_01_lf
  bone2Name=offset_hair_lf
  swing1LimitDegrees=0
  swing1Motion=Locked
  swing2LimitDegrees=3
  twistLimitDegrees=0
  twistMotion=Locked
 }
 {
  bone1Name=dyn_hair_02_lf
  bone2Name=dyn_hair_01_lf
  swing1LimitDegrees=0
  swing1Motion=Locked
  swing2LimitDegrees=15
  twistLimitDegrees=0
  twistMotion=Locked
 }
 {bone1Name=dyn_hair_03_lf, bone2Name=dyn_hair_02_lf, swing1LimitDegrees=10, swing2LimitDegrees=30, twistLimitDegrees=0, twistMotion=Locked}
 {bone1Name=dyn_hair_04_lf, bone2Name=dyn_hair_03_lf, swing1LimitDegrees=10, swing2LimitDegrees=20, twistLimitDegrees=0, twistMotion=Locked}
 {bone1Name=dyn_hair_05_lf, bone2Name=dyn_hair_04_lf, swing1LimitDegrees=10, swing2LimitDegrees=40, twistLimitDegrees=45, twistMotion=Locked}
 {
  bone1Name=dyn_hair_01_rt
  bone2Name=offset_hair_rt
  swing1LimitDegrees=0
  swing1Motion=Locked
  swing2LimitDegrees=3
  twistLimitDegrees=0
  twistMotion=Locked
 }
 {
  bone1Name=dyn_hair_02_rt
  bone2Name=dyn_hair_01_rt
  swing1LimitDegrees=0
  swing1Motion=Locked
  swing2LimitDegrees=15
  twistLimitDegrees=0
  twistMotion=Locked
 }
 {bone1Name=dyn_hair_03_rt, bone2Name=dyn_hair_02_rt, swing1LimitDegrees=10, swing2LimitDegrees=30, twistLimitDegrees=0, twistMotion=Locked}
 {bone1Name=dyn_hair_04_rt, bone2Name=dyn_hair_03_rt, swing1LimitDegrees=10, swing2LimitDegrees=20, twistLimitDegrees=0, twistMotion=Locked}
 {bone1Name=dyn_hair_05_rt, bone2Name=dyn_hair_04_rt, swing1LimitDegrees=10, swing2LimitDegrees=40, twistLimitDegrees=45, twistMotion=Locked}
 {
  bone1Name=dyn_hair_02_cr
  bone2Name=dyn_hair_01_cr
  swing1LimitDegrees=0
  swing1Motion=Locked
  swing2LimitDegrees=15
  twistLimitDegrees=0
  twistMotion=Locked
 }
 {bone1Name=dyn_hair_front_04_rt, bone2Name=dyn_hair_front_03_rt, swing1LimitDegrees=10, swing2LimitDegrees=10, twistLimitDegrees=10}
 {
  bone1Name=dyn_hair_front_01_rt
  bone2Name=offset_hair_front_rt
  swing1LimitDegrees=5
  swing2LimitDegrees=15
  twistLimitDegrees=0
  twistMotion=Locked
 }
 {
  bone1Name=dyn_hair_front_02_rt
  bone2Name=dyn_hair_front_01_rt
  swing1LimitDegrees=10
  swing2LimitDegrees=15
  twistLimitDegrees=0
  twistMotion=Locked
 }
 {
  bone1Name=dyn_hair_front_03_rt
  bone2Name=dyn_hair_front_02_rt
  swing1LimitDegrees=15
  swing2LimitDegrees=10
  twistLimitDegrees=10
  twistMotion=Locked
 }
]
