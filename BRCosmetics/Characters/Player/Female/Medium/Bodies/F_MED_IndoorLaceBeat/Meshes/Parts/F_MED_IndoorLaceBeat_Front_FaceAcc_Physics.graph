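# Physics asset for the front face accessory of the IndoorLaceBeat medium female body, setting up physics for a right-side front hair strand.
# Four capsule hair bodies form a constrained chain hanging from a small offset_hair_front_rt anchor.
# Tight swing limits restrict each link, with twist locked on the upper three joints.
# Unconstrained capsules on the upper spine, right clavicle and right upper arm give the strand body shapes to collide against.
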PhysicsAsset /BRCosmetics/Characters/Player/Female/Medium/Bodies/F_MED_IndoorLaceBeat/Meshes/Parts/F_MED_IndoorLaceBeat_Front_FaceAcc_Physics
bodies {
 clavicle_r [{
   shapeType=Capsule
   length=6.7873
   radius=5.9241
   shapeName=clavicle_r_capsule
   center=<-6.2251 1.0912 2.1208>
   rotation=<rot 87.5685 -179.9942 -179.9944>
  }]
 dyn_hair_front_01_rt [{
   shapeType=Capsule
   length=6.834
   radius=3.673
   shapeName=dyn_hair_01_rt_capsule
   center=<4.6928 0 0>
   rotation=<rot 84.2217 179.9994 179.9995>
  }]
 dyn_hair_front_02_rt [{
   shapeType=Capsule
   length=6.0171
   radius=2.5108
   shapeName=dyn_hair_front_02_lf_capsule
   center=<5.7081 -0.3034 -0.1667>
   rotation=<rot 84.9999 180 180>
  }]
 dyn_hair_front_03_rt [{
   shapeType=Capsule
   length=4.8667
   radius=2.9615
   shapeName=dyn_hair_front_03_lf_capsule
   center=<1.9997 0 0.4921>
   rotation=<rot 85.6532 -179.9984 -180>
  }]
 dyn_hair_front_04_rt [{
   shapeType=Capsule
   length=4.8667
   radius=2.9615
   shapeName=dyn_hair_front_03_lf_capsule
   center=<1.9997 0 0.4921>
   rotation=<rot 85.6532 -179.9984 -180>
  }]
 offset_hair_front_rt [{shapeType=Capsule, length=0.505, radius=0.505, shapeName=offset_hair_front_rt_capsule, rotation=<rot 0 -90 0>}]
 spine_04 [
  {shapeType=Capsule, length=17.2664, radius=8.9449, shapeName=spine_04_capsule, center=<3.398 -0.1366 0>}
  {shapeType=Capsule, length=7.8481, radius=5.8535, shapeName=spine_04_capsule, center=<0.6187 7.2479 0>}
 ]
 spine_05 [
  {shapeType=Capsule, length=17.9157, radius=7.5551, shapeName=spine_05_capsule, center=<1.7618 -2.5383 0>, rotation=<rot 0 -90 0>}
  {shapeType=Capsule, length=17.9157, radius=7.5551, shapeName=spine_05_capsule, center=<1.7249 -0.1335 0>, rotation=<rot 0 -90 0>}
 ]
 upperarm_r [{shapeType=Capsule, length=24.0453, radius=5.0231, shapeName=upperarm_l_capsule, center=<-13.9807 0.352 0.5342>, rotation=<rot 0 -90 -90>}]
}
constraints [
 {
  bone1Name=dyn_hair_front_01_rt
  bone2Name=offset_hair_front_rt
  swing1LimitDegrees=5
  swing2LimitDegrees=10
  twistLimitDegrees=0
  twistMotion=Locked
 }
 {
  bone1Name=dyn_hair_front_02_rt
  bone2Name=dyn_hair_front_01_rt
  swing1LimitDegrees=5
  swing2LimitDegrees=25
  twistLimitDegrees=0
  twistMotion=Locked
 }
 {
  bone1Name=dyn_hair_front_03_rt
  bone2Name=dyn_hair_front_02_rt
  swing1LimitDegrees=5
  swing2LimitDegrees=20
  twistLimitDegrees=0
  twistMotion=Locked
 }
 {bone1Name=dyn_hair_front_04_rt, bone2Name=dyn_hair_front_03_rt, swing1LimitDegrees=10, swing2LimitDegrees=10, twistLimitDegrees=0}
]
